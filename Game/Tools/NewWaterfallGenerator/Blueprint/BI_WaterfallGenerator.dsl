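; BI_WaterfallGenerator is a Blueprint interface for the waterfall generator tool, declaring calls without implementations.
; It covers mesh generation, merging and baking, plus custom water body generation with an optional forced update.
; Helper calls create a named static mesh with a default template path, delay execution with a force-update flag, and show a titled message box.
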(fn BakeWaterfallMesh)

(fn CreateNewStaticMesh
  (params (Name :type String) (TemplatePath :type String :default "/Game/Tools/WaterfallGenerator/Template"))
  (return 0))

(fn DelayExecute
  (params (ForceUpdateStaticMesh :type Boolean)))

(fn GenerationBI_BakeMesh)

(fn GenerationBI_GenerateCustomWaterBody
  (params (ForceUpdate :type Boolean)))

(fn GenerationBI_MeshGen)

(fn GenerationBI_MeshMerge)

(fn MessageBox
  (params (Title :type Text) (Message :type Text)))
