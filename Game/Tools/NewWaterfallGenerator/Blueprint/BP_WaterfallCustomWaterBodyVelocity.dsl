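; BP_WaterfallCustomWaterBodyVelocity is an editor helper actor for the waterfall generator tool, destroying itself on BeginPlay when not in the editor.
; Its construction script finds every BP_WaterfallGenerator in the level and sets a short timer on each one's editor component to call DelayExecuteDispatcherWithForceUpdate.
; Constructing or editing this actor thus triggers a delayed forced update on all generators.
; The overlap and tick events are empty, and no velocity logic appears here.
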(event EventBeginPlay
  (if (not (Editor|InEditor))
    (Actor|DestroyActor self)))

(event Collision|EventActorBeginOverlap
  (params (OtherActor :type Actor%20Object%20Reference)))

(event EventTick
  (params (DeltaSeconds :type Float%20%28single-precision%29)))

(fn ConstructionScript
  (bind _outactors (Actor|GetAllActorsOfClass "/Game/Tools/NewWaterfallGenerator/Blueprint/BP_WaterfallGenerator.BP_WaterfallGenerator_C"))
  (for _array_element _outactors
    (Utilities|Time|SetTimerbyFunctionName (Variables|/Game/Tools/NewWaterfallGenerator/Blueprint/BP_WaterfallGenerator.BP_WaterfallGenerator_C|GetBP_WaterfallGeneratorEditorComp _array_element) "DelayExecuteDispatcherWithForceUpdate" 0.333)))
