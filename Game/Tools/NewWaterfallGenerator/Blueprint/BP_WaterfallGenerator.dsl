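; Waterfall generator actor that builds waterfall meshes in the editor and handles their in-game water linking and knockback.
; Widget points shape a spline, sampled at three LODs, while an editor component handles mesh generation and baking.
; It configures a custom top water body, passes clip data to linked water bodies and sets material parameters, with delayed client retries.
; Its collision box knocks back pawns via an ability and launches vehicles.
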(event Custom|DelaySetupMaterial
  (bind _bottomlinktolake (Variables|Initial|GetBottomLinkToLake))
  (bind _output (Utilities|ResolveSoftReference _bottomlinktolake))
  (bind _dynamicsplashmaterial (Variables|TmpMaterial|GetDynamicSplashMaterial))
  (Utilities|FlowControl|Delay 0.5)
  (Utilities|IsValid _output
    (:Is%20Valid
      (bind _returnvalue (Rendering|GetWaterMaterialInstance (Variables|/Script/Water.WaterBody|GetWaterBodyComponent _output)))
      (Rendering|Material|CopyMaterialInstanceParameters _dynamicsplashmaterial _returnvalue false)
      (Rendering|Material|SetScalarParameterValue _dynamicsplashmaterial "LinkToWaterBodyIndex" (Math|Conversions|ToFloat(Integer) (Variables|/Script/Water.WaterBody|GetWaterBodyIndex _output)) :!class /Script/Engine.MaterialInstanceDynamic))
    (:Is%20Not%20Valid
      (CallFunction|DelaySetupMaterial))))

(event EventBeginPlay
  (bind _debughlodtest (Variables|##Debug|GetDebugHLODTest))
  (bind _iterationcounter (Variables|Default|GetIterationCounter))
  (bind _bottomlinktolake (Variables|Initial|GetBottomLinkToLake))
  (Utilities|FlowControl|Sequence
    (:then_0
      (Networking|SwitchHasAuthority
        (:Remote
          (if (Utilities|IsValidSoftObjectReference _bottomlinktolake)
            (Utilities|FlowControl|Delay 5.0)
            (if (not (Utilities|IsValid (Utilities|ResolveSoftReference _bottomlinktolake) :!class /Script/Engine.KismetSystemLibrary))
              (Math|Integer|IncrementInt _iterationcounter
                (:%20%20
                  (if (< _iterationcounter 6)
                    _))))
            (CallFunction|DotheWaterbodyClip)
            (Default|FixCustomWaterBodyWaveSource)))))
    (:then_1
      (Utilities|IsValid _debughlodtest
        (:Is%20Valid
          (Actor|DestroyActor _debughlodtest))))))

(event EventBakeWaterfallMesh
  (Generation|GenerationBakeMesh))

(event EventDelayExecute
  (params (ForceUpdateStaticMesh :type Boolean))
  (Generation|GenerationPoolSplineGen)
  (Generation|GenerationGenerateCustomWaterBody ForceUpdateStaticMesh)
  (CallFunction|DotheWaterbodyClip)
  (Default|FixCustomWaterBodyWaveSource)
  (CallFunction|CustomWaterBodyLowPlatformSet))

(event OnComponentBeginOverlap(Collision)
  (params (OverlappedComponent :type Primitive%20Component%20Object%20Reference) (OtherActor :type Actor%20Object%20Reference) (OtherComp :type Primitive%20Component%20Object%20Reference) (OtherBodyIndex :type Integer) (bFromSweep :type Boolean) (SweepResult :type Hit%20Result%20Structure%20%28by%20ref%29))
  (bind _gp_current_actor (Variables|Default|GetGP%20Current%20Actor))
  (bind _returnvalue (Ability|GetAbilitySystemComponent _gp_current_actor))
  (bind _returnvalue_2 (Transformation|GetActorForwardVector))
  (Utilities|FlowControl|Sequence
    (:then_0
      (bind _asfort_pawn (Utilities|Casting|CastToFortPawn OtherActor))
      (Variables|Default|SetGP%20Current%20Actor _asfort_pawn)
      (if (not (or (GameplayTags|HasMatchingGameplayTag (Ability|GetAbilitySystemComponent (Variables|Default|GetGP%20Current%20Actor)) "(TagName=\"Gameplay.Action.Player.Launched\")") (Variables|/Script/FortniteGame.FortPawn|GetbIsDBNO (Variables|Default|GetGP%20Current%20Actor))))
        ;[0.1 0.1 0.75] Handle pawn knockback
        (bind _asfort_player_pawn (Utilities|Casting|CastToFortPlayerPawn (Variables|Default|GetGP%20Current%20Actor))
          (:then
            (RemoteControlPawn|ForceDetachfromRemoteControlledPawn _asfort_player_pawn))
          (:CastFailed
            _))
        (Utilities|IsValid _returnvalue
          (:Is%20Valid
            (Ability|Advanced|GrantandActivateAbilityonSelf _returnvalue (Variables|Default|GetPawnKnockbackAbility) self)
            (Variables|/Script/FortniteGame.FortPawn|SetbIgnoreNextFallingDamage true _gp_current_actor)
            (bind _asfort_player_controller (Utilities|Casting|CastToFortPlayerController (Pawn|GetController _gp_current_actor)))
            (Game|Feedback|LocalPlayForceFeedback _asfort_player_controller (Variables|Default|GetLaunchForceFeedback) false "TireLaunch"))
          (:Is%20Not%20Valid
            ;;[1 0 0] Warning: ASC is not valid
            (Fortnite|FortShippingLog self "Failed to knockback pawn. AbilitySystemComponent is not valid" true)))))
    (:then_1
      (bind _asfort_pickup (Utilities|Casting|CastToFortPickup OtherActor)))
    (:then_2
      ;[1 1 1] JIRA FORT-627177
      (bind _asfort_athena_vehicle (Utilities|Casting|CastToFortAthenaVehicle OtherActor))
      (Development|PrintString (Utilities|String|ToString(Boolean) (> (Math|Vector|DotProduct (Vehicle|GetVehicleVelocity _asfort_athena_vehicle) _returnvalue_2) 0)) true true "(R=1.000000,G=0.685931,B=0.000000,A=1.000000)")
      (bind (_returnvalue_x _returnvalue_y _returnvalue_z) (Math|Vector|BreakVector (Vehicle|GetVehicleVelocity _asfort_athena_vehicle)))
      (Audio|BroadcastAbilityImpactatLocation 0 _asfort_athena_vehicle (Math|Vector|MakeVector (+ _returnvalue_x (.x (Math|Vector|SelectVector (+ (* _returnvalue_2 "-400.000000,-400.000000,-400.000000") (* (Math|Vector|MakeVector (.x (Vehicle|GetVehicleVelocity _asfort_athena_vehicle)) (.y (Vehicle|GetVehicleVelocity _asfort_athena_vehicle))) "-1.500000,-1.500000,-1.500000")) (* _returnvalue_2 "-200.000000,-200.000000,-200.000000") (> (Math|Vector|DotProduct (Vehicle|GetVehicleVelocity _asfort_athena_vehicle) _returnvalue_2) 0)))) (+ (.y (Math|Vector|SelectVector (+ (* _returnvalue_2 "-400.000000,-400.000000,-400.000000") (* (Math|Vector|MakeVector (.x (Vehicle|GetVehicleVelocity _asfort_athena_vehicle)) (.y (Vehicle|GetVehicleVelocity _asfort_athena_vehicle))) "-1.500000,-1.500000,-1.500000")) (* _returnvalue_2 "-200.000000,-200.000000,-200.000000") (> (Math|Vector|DotProduct (Vehicle|GetVehicleVelocity _asfort_athena_vehicle) _returnvalue_2) 0))) _returnvalue_y) (+ _returnvalue_z (Math|Float|SelectFloat 3e+02 7e+02 (> (Math|Vector|DotProduct (Vehicle|GetVehicleVelocity _asfort_athena_vehicle) _returnvalue_2) 0)))) 777.0 "None" false) ; Launch vehicle up and retain forward speed
    )))

(fn AudioSetting
  (bind _toploopsound (Variables|Default|GetTopLoopSound))
  (bind _bottomloopsound (Variables|Default|GetBottomLoopSound))
  (if (Variables|Audio|GetEnableAudio)
    (Audio|Components|Audio|SetSound _bottomloopsound (Variables|Audio|GetBottomSound))
    (Audio|Components|Audio|SetSound _toploopsound (Variables|Audio|GetTopSound))
    (Transformation|SetRelativeLocation _bottomloopsound (+ (Variables|Widget|Get◉Bottom) (Variables|Audio|GetBottomSoundOffset)) false false)
    (Transformation|SetRelativeLocation _toploopsound (+ (Variables|Widget|Get◉Top) (Variables|Audio|GetTopSoundOffset)) false false)
    (else
      (Components|DestroyComponent (Variables|Default|GetBottomLoopSound)))))

(fn CreateDynamicMaterial
  (bind _returnvalue (Rendering|Material|CreateDynamicMaterialInstance (Variables|OverrideMaterialandFx|GetOverrideFallMaterial) "None" "None" :!class /Script/Engine.KismetMaterialLibrary))
  (Variables|TmpMaterial|SetDynamicFallMaterial _returnvalue)
  (bind _returnvalue_2 (Rendering|Material|CreateDynamicMaterialInstance (Variables|OverrideMaterialandFx|GetOverrideSplashMaterial) "None" "None" :!class /Script/Engine.KismetMaterialLibrary))
  (Variables|TmpMaterial|SetDynamicSplashMaterial _returnvalue_2))

(fn CreateNewStaticMesh
  (params (Name :type String) (TemplatePath :type String))
  (return 0))

(fn CreateNewStaticMeshButton
  (bind _createstaticmeshname (Variables|Initial|GetCreateStaticMeshName))
  (bind _waterfallmesh (Variables|Default|GetWaterfallMesh))
  (bind _bp_waterfallgeneratoreditorcomp (Variables|Default|GetBP_WaterfallGeneratorEditorComp))
  (if (not (Utilities|String|IsEmpty _createstaticmeshname))
    (Utilities|FlowControl|Sequence
      (:then_0
        (bind _returnsm (Class|BIWaterfallGenerator|CreateNewStaticMesh _bp_waterfallgeneratoreditorcomp _createstaticmeshname "/Game/NeverCook/TechArt/WaterfallGenerator/Template"))
        (Utilities|IsValid _returnsm
          (:Is%20Valid
            (Variables|Initial|SetTargetStaticMesh _returnsm))))
      (:then_1
        (bind _returnsm_2 (Class|BIWaterfallGenerator|CreateNewStaticMesh _bp_waterfallgeneratoreditorcomp (Utilities|String|Append _createstaticmeshname "_CWB") "/Game/NeverCook/TechArt/WaterfallGenerator/Template_CWB_LOD"))
        (Utilities|IsValid _returnsm_2
          (:Is%20Valid
            (Variables|Initial|SetTargetStaticMesh_CustomWaterBody _returnsm_2))))
      (:then_2
        (CallFunction|UserGuider))
      (:then_3
        (Utilities|IsValid (Variables|Initial|GetTargetStaticMesh)
          (:Is%20Valid
            (Rendering|SetVisibility (Variables|Default|GetDynamicMesh) false false)
            (Components|StaticMesh|SetStaticMesh (Variables|Default|GetWaterfallMesh) (Variables|Initial|GetTargetStaticMesh))
            (Material|CreateDynamicMaterial)
            (Rendering|Material|SetMaterial _waterfallmesh 0 (Variables|TmpMaterial|GetDynamicFallMaterial))
            (Rendering|Material|SetMaterial (Variables|Default|GetWaterfallMesh) 1 (Variables|TmpMaterial|GetDynamicSplashMaterial))
            (Rendering|Material|SetMaterial (Variables|Default|GetWaterfallMesh) 2 0)
            (Material|InitialMaterialonMesh _waterfallmesh)))))))

(fn CustomWaterBodyLowPlatformSet
  (bind _bottom (Variables|Widget|Get◉Bottom))
  (bind _top (Variables|Widget|Get◉Top))
  (bind _returnvalue (+ _top _bottom))
  (bind _returnvalue_2 (/ _returnvalue "2.000000,2.000000,2.000000"))
  (bind _z (.z _returnvalue_2))
  (bind _length (Variables|TmpVarButBake|GetLength))
  (bind _returnvalue_3 (* _length 0.1))
  (bind _extend (Variables|Widget|Get◉Extend))
  (bind _x (.x _extend))
  (bind _returnvalue_4 (- _x _returnvalue_3))
  (bind _returnvalue_5 (/ _returnvalue_4 2.0))
  (bind _returnvalue_6 (Math|Vector|MakeVector _returnvalue_5 0.0 _z))
  (bind _returnvalue_7 (Transformation|GetActorTransform))
  (bind _returnvalue_8 (Math|Transform|TransformLocation _returnvalue_7 _returnvalue_6))
  (Utilities|IsValid (Variables|Initial|GetTargetStaticMesh_CustomWaterBody)
    (:Is%20Valid
      (bind _returnvalue_9 (AddComponent|NODE_AddStaticMeshComponent-0 :RelativeTransform (Math|Transform|MakeTransform (+ (Variables|Widget|Get◉Top) (Math|Vector|MakeVector 0.0 0.0 (Utilities|Operators|Add (* (Variables|TmpVarButBake|GetThickness) 0.5) (Variables|TopConnectionWaterBody|GetCustomWaterBodyActorZOffset) 5.0))))))
      (Utilities|IsValid _returnvalue_9
        (:Is%20Valid
          (Components|StaticMesh|SetStaticMesh _returnvalue_9 (Variables|Initial|GetTargetStaticMesh_CustomWaterBody))
          (Rendering|Material|SetMaterial _returnvalue_9 0 (Variables|OverrideMaterialandFx|GetCustomWaterBodyMaterial%20Mobile))
          (Rendering|SetTranslucentSortPriority _returnvalue_9 0)
          (bind _output_get (Variables|Default|SetLowPlatformCustomWater _returnvalue_9))
          (Rendering|Material|SetColorParameterValueonMaterials _output_get "FoamColor" (Variables|MatAndFXParameters|GetFoamColor))
          (Rendering|Material|SetScalarParameterValueonMaterials _output_get "TopFoamZOffset" (Variables|MatAndFXParameters|GetTopFoamZOffset))
          (Development|PrintString "InitCustomWaterBodyLowPlatformVisual")
          (Development|PrintString (Utilities|String|ToString(Vector) _returnvalue_8))
          (Rendering|Material|SetVectorParameterValueonMaterials _output_get "BoxMaskCenterPos" _returnvalue_8)
          (Rendering|Material|SetColorParameterValueonMaterials _output_get "BoxMaskExtendAndFalloff" (Utilities|Struct|MakeLinearColor (- _returnvalue_4 4e+02) (* (.y (Variables|Widget|Get◉WdithPoint)) 2.0) 1e+05 2e+02)))))))

(fn CustomWaterBodySet
  (bind _customwaterbodyactor_soft (Variables|TopConnectionWaterBody|GetCustomWaterBodyActor_Soft))
  (bind _output (Utilities|ResolveSoftReference _customwaterbodyactor_soft))
  (bind _returnvalue (Water|GetWaterBodyComponent _output))
  (bind _returnvalue_2 (Water|GetWaterSpline _output :!class /Script/Water.WaterBody))
  (bind _waterbodycomponent (Variables|/Script/Water.WaterBody|GetWaterBodyComponent _output))
  (bind _targetstaticmesh_customwaterbody (Variables|Initial|GetTargetStaticMesh_CustomWaterBody))
  (bind _customwaterbodymaterial (Variables|OverrideMaterialandFx|GetCustomWaterBodyMaterial))
  (bind _customwaterbodyunderwatermaterial (Variables|OverrideMaterialandFx|GetCustomWaterBodyUnderWaterMaterial))
  (Utilities|IsValid _output
    (:Is%20Valid
      (Transformation|SetActorTransform _output (Math|Transform|ComposeTransforms (Transformation|GetActorTransform) (Math|Transform|MakeTransform (+ (Variables|Widget|Get◉Top) (Math|Vector|MakeVector 0.0 0.0 (+ (* (Variables|TmpVarButBake|GetThickness) 0.5) (Variables|TopConnectionWaterBody|GetCustomWaterBodyActorZOffset)))))) false false)
      (Utilities|IsValid _targetstaticmesh_customwaterbody
        (:Is%20Valid
          (Utilities|FlowControl|Sequence
            (:then_0
              (Spline|RemoveSplinePoint _returnvalue_2 2 false)
              (Spline|SetLocationatSplinePoint _returnvalue_2 0 (Math|Vector|MakeVector 1e+02) "Local" false)
              (Spline|SetLocationatSplinePoint _returnvalue_2 1 (Math|Vector|MakeVector 0.0) "Local" true))
            (:then_1
              (Variables|/Script/Water.WaterBodyComponent|SetWaterMeshOverride _targetstaticmesh_customwaterbody _returnvalue)
              (Utilities|IsValid _customwaterbodymaterial
                (:Is%20Valid
                  (Utilities|IsValid _customwaterbodyunderwatermaterial
                    (:Is%20Valid
                      (Rendering|SetWaterandUnderWaterPostProcessMaterial _returnvalue _customwaterbodymaterial _customwaterbodyunderwatermaterial))))))
            (:then_2
              (WaterBody|SetWaterVelocityatSplineInputKey _waterbodycomponent 0.0 3e+03)
              (WaterBody|SetWaterVelocityatSplineInputKey _waterbodycomponent 1.0 3e+03))))))))

(fn DoTheWaterbodyClip
  (bind _toplinktowaterbody (Variables|Initial|GetTopLinkToWaterBody))
  (bind _bottomlinktolake (Variables|Initial|GetBottomLinkToLake))
  (if (Variables|ClipWaterBody|GetWaterBodyClip)
    (Utilities|FlowControl|Sequence
      (:then_0
        (if (Utilities|IsValidSoftObjectReference _bottomlinktolake)
          (Base|TransitWaterbodyClipInformation (Utilities|ResolveSoftReference _bottomlinktolake))))
      (:then_1
        (if (Utilities|IsValidSoftObjectReference _toplinktowaterbody)
          (Base|TransitWaterbodyClipInformation (Utilities|ResolveSoftReference _toplinktowaterbody))))
      (:then_2
        (for _array_element (Variables|ClipWaterBody|GetAdditionalClipWaterBodys)
          (if (Utilities|IsValidSoftObjectReference _array_element)
            (Base|TransitWaterbodyClipInformation (Utilities|ResolveSoftReference _array_element))))))))

(fn DuplicateWaterfall
  (Class|BIWaterfallGenerator|MessageBox (Variables|Default|GetBP_WaterfallGeneratorEditorComp) (Utilities|Text|ToText(String) "How to duplicate waterfall?") (Utilities|Text|ToText(String) "Right click waterfall actor , then click Sprited Actor Actions -> Duplicate waterfall")))

(fn EditorInitialCustomParameters
  (bind _customparameterstable (Variables|OverrideMaterialandFx|GetCustomParametersTable))
  (bind _materialparameters_float (Variables|/Game/Tools/NewWaterfallGenerator/Blueprint/WaterfallGeneratorCustomParameter.WaterfallGeneratorCustomParameter_C|GetMaterialParameters_Float _customparameterstable))
  (bind (_key _value) (Base|GetMapKeyandValueFloat _materialparameters_float))
  (bind _fxparameters_bool (Variables|/Game/Tools/NewWaterfallGenerator/Blueprint/WaterfallGeneratorCustomParameter.WaterfallGeneratorCustomParameter_C|GetFXParameters_Bool _customparameterstable))
  (bind (_key_2 _value_2) (Base|GetMapKeyandValueBool _fxparameters_bool))
  (bind _materialparameters_color (Variables|/Game/Tools/NewWaterfallGenerator/Blueprint/WaterfallGeneratorCustomParameter.WaterfallGeneratorCustomParameter_C|GetMaterialParameters_Color _customparameterstable))
  (bind (_key_3 _value_3) (Base|GetMapKeyandValueColor _materialparameters_color))
  (bind _fxparameters_float (Variables|/Game/Tools/NewWaterfallGenerator/Blueprint/WaterfallGeneratorCustomParameter.WaterfallGeneratorCustomParameter_C|GetFXParameters_Float _customparameterstable))
  (bind (_key_4 _value_4) (Base|GetMapKeyandValueFloat _fxparameters_float))
  (bind _fxparameters_color (Variables|/Game/Tools/NewWaterfallGenerator/Blueprint/WaterfallGeneratorCustomParameter.WaterfallGeneratorCustomParameter_C|GetFXParameters_Color _customparameterstable))
  (bind (_key_5 _value_5) (Base|GetMapKeyandValueColor _fxparameters_color))
  (bind _fxcustomparameters_float (Variables|MatAndFXParameters|GetFXCustomParameters_Float))
  (bind _fxcustomparameters_bool (Variables|MatAndFXParameters|GetFXCustomParameters_Bool))
  (bind _fxcustomparameters_color (Variables|MatAndFXParameters|GetFXCustomParameters_Color))
  (bind _matcustomparameters_float (Variables|MatAndFXParameters|GetMatCustomParameters_Float))
  (bind _matcustomparameters_color (Variables|MatAndFXParameters|GetMatCustomParameters_Color))
  (Utilities|IsValid _customparameterstable
    (:Is%20Valid
      (Utilities|FlowControl|Sequence
        (:then_0
          (bind _keys (Utilities|Map|Keys _fxcustomparameters_float))
          (for (_array_element _array_index) _key_4
            (if (not (>= (Utilities|Array|FindItem _keys _array_element) 0))
              (Utilities|Map|Add _fxcustomparameters_float _array_element (Utilities|Array|Get(acopy) _value_4 _array_index)))))
        (:then_1
          (bind _keys_2 (Utilities|Map|Keys _fxcustomparameters_bool))
          (for (_array_element_2 _array_index_2) _key_2
            (if (not (>= (Utilities|Array|FindItem _keys_2 _array_element_2) 0))
              (Utilities|Map|Add _fxcustomparameters_bool _array_element_2 (Utilities|Array|Get(acopy) _value_2 _array_index_2)))))
        (:then_2
          (bind _keys_3 (Utilities|Map|Keys _fxcustomparameters_color))
          (for (_array_element_3 _array_index_3) _key_5
            (if (not (>= (Utilities|Array|FindItem _keys_3 _array_element_3) 0))
              (Utilities|Map|Add _fxcustomparameters_color _array_element_3 (Utilities|Array|Get(acopy) _value_5 _array_index_3)))))
        (:then_3
          (bind _keys_4 (Utilities|Map|Keys _matcustomparameters_float))
          (for (_array_element_4 _array_index_4) _key
            (if (not (>= (Utilities|Array|FindItem _keys_4 _array_element_4) 0))
              (Utilities|Map|Add _matcustomparameters_float _array_element_4 (Utilities|Array|Get(acopy) _value _array_index_4)))))
        (:then_4
          (bind _keys_5 (Utilities|Map|Keys _matcustomparameters_color))
          (for (_array_element_5 _array_index_5) _key_3
            (if (not (>= (Utilities|Array|FindItem _keys_5 _array_element_5) 0))
              (Utilities|Map|Add _matcustomparameters_color _array_element_5 (Utilities|Array|Get(acopy) _value_3 _array_index_5)))))))))

(fn FixCustomWaterBodyWaveSource
  (bind _waveattributesourceforcustomwaterbody (Variables|TopConnectionWaterBody|GetWaveAttributeSourceForCustomWaterBody))
  (bind _output (Utilities|ResolveSoftReference _waveattributesourceforcustomwaterbody))
  (bind _customwaterbodyactor_soft (Variables|TopConnectionWaterBody|GetCustomWaterBodyActor_Soft))
  (bind _output_2 (Utilities|ResolveSoftReference _customwaterbodyactor_soft))
  (bind _toplinktowaterbody (Variables|Initial|GetTopLinkToWaterBody))
  (bind _output_3 (Utilities|ResolveSoftReference _toplinktowaterbody))
  (bind _waterbodycomponent (Variables|/Script/Water.WaterBody|GetWaterBodyComponent _output_2))
  (if (Asset|IsAssetNull _customwaterbodyactor_soft)
    _
    (elif (Utilities|IsValidSoftObjectReference _customwaterbodyactor_soft)
      (Utilities|IsValid _output_2
        (:Is%20Valid
          (bind _returnvalue (Rendering|GetWaterMaterialInstance _waterbodycomponent))
          ; Todo:MoveThisToStandaloneFunction
          (if (Variables|TopConnectionWaterBody|GetOverrideCustomWaterBodyColor)
            (Rendering|Material|SetVectorParameterValue _returnvalue "Absorption" (Variables|TopConnectionWaterBody|GetCustomWaterBodyColor) :!class /Script/Engine.MaterialInstanceDynamic)
            (Rendering|Material|SetVectorParameterValue _returnvalue "OceanAbsorption" (Variables|TopConnectionWaterBody|GetCustomWaterBodyColor_Ocean) :!class /Script/Engine.MaterialInstanceDynamic)
            (Rendering|Material|SetVectorParameterValue _returnvalue "RiverAbsorption" (Variables|TopConnectionWaterBody|GetCustomWaterBodyColor_River) :!class /Script/Engine.MaterialInstanceDynamic)
            (Rendering|Material|SetScalarParameterValue _returnvalue "Roughness" (Variables|TopConnectionWaterBody|GetCustomWaterBodyRoughness) :!class /Script/Engine.MaterialInstanceDynamic)
            (Rendering|Material|SetScalarParameterValue _returnvalue "WavelNormalIntensity" (Variables|TopConnectionWaterBody|GetCustomWaterBodyWaveNormalStrength) :!class /Script/Engine.MaterialInstanceDynamic))
          (bind _returnvalue_2 (Rendering|GetUnderwaterPostProcessMaterialInstance _waterbodycomponent))
          (if (Asset|IsAssetNull _waveattributesourceforcustomwaterbody)
            _
            (elif (Utilities|IsValidSoftObjectReference _waveattributesourceforcustomwaterbody)
              (Utilities|IsValid _output
                (:Is%20Valid
                  (Rendering|Material|SetScalarParameterValue _returnvalue "WaterfallOverrideWaterBodyIndex" (Math|Conversions|ToFloat(Integer) (Variables|/Script/Water.WaterBody|GetWaterBodyIndex _output)) :!class /Script/Engine.MaterialInstanceDynamic))
                (:Is%20Not%20Valid
                  _))))
          (if (Asset|IsAssetNull _toplinktowaterbody)
            _
            (elif (Utilities|IsValidSoftObjectReference _toplinktowaterbody)
              (Utilities|IsValid _output_3
                (:Is%20Valid
                  (Rendering|Material|SetScalarParameterValue _returnvalue_2 "WaterfallOverrideWaterBodyIndex" (Math|Conversions|ToFloat(Integer) (Variables|/Script/Water.WaterBody|GetWaterBodyIndex _output_3)) :!class /Script/Engine.MaterialInstanceDynamic))
                (:Is%20Not%20Valid
                  _))))
          (Rendering|Material|SetScalarParameterValue _returnvalue "WaterfallOverrideWaterBodyIndex" (Math|Conversions|ToFloat(Integer) (Variables|/Script/Water.WaterBody|GetWaterBodyIndex _output_2)) :!class /Script/Engine.MaterialInstanceDynamic))))))

(fn GameplayThingsRefresh
  (bind _thickness (Variables|TmpVarButBake|GetThickness))
  (bind _returnvalue (Math|Float|Min(Float) _thickness 1.5e+02))
  (bind _bottomsplinepointposition (Variables|TmpVarButBake|GetBottomSplinePointPosition))
  (bind _topsplinepointposition (Variables|TmpVarButBake|GetTopSplinePointPosition))
  (bind _returnvalue_2 (- _topsplinepointposition _bottomsplinepointposition))
  (bind _returnvalue_3 (/ _returnvalue_2 "2.000000,2.000000,2.000000"))
  (bind _collision (Variables|Default|GetCollision))
  (Utilities|FlowControl|Sequence
    (:then_0
      (Transformation|SetWorldLocation _collision (Math|Vector|MakeVector (.x (Utilities|Operators|Add (Utilities|Operators|Multiply (Transformation|GetActorForwardVector) "1.000000,1.000000,1.000000" _returnvalue) _bottomsplinepointposition (Utilities|Operators|Multiply (Transformation|GetActorForwardVector) "-0.500000,-0.500000,-0.500000" (Math|Float|Max(Float) (- _thickness 1.5e+02))))) (.y (Utilities|Operators|Add (Utilities|Operators|Multiply (Transformation|GetActorForwardVector) "1.000000,1.000000,1.000000" _returnvalue) _bottomsplinepointposition (Utilities|Operators|Multiply (Transformation|GetActorForwardVector) "-0.500000,-0.500000,-0.500000" (Math|Float|Max(Float) (- _thickness 1.5e+02))))) (.z (+ _bottomsplinepointposition _returnvalue_3))) false false)
      (Components|Box|SetBoxExtent _collision (Math|Vector|MakeVector (* _returnvalue 1.5) (* (Variables|TmpVarButBake|GetWidth) 0.5) (+ (.z _returnvalue_3) (* (Variables|TmpVarButBake|GetThickness) 0.5))) true))))

(fn Generation_BakeMesh
  (Class|BIWaterfallGenerator|GenerationBIBakeMesh (Variables|Default|GetBP_WaterfallGeneratorEditorComp)))

(fn Generation_GenerateCustomWaterBody
  (params (ForceUpdate :type Boolean))
  (Class|BIWaterfallGenerator|GenerationBIGenerateCustomWaterBody (Variables|Default|GetBP_WaterfallGeneratorEditorComp) ForceUpdate))

(fn Generation_MeshGen
  (Class|BIWaterfallGenerator|GenerationBIMeshGen (Variables|Default|GetBP_WaterfallGeneratorEditorComp)))

(fn Generation_MeshMerge
  (Class|BIWaterfallGenerator|GenerationBIMeshMerge (Variables|Default|GetBP_WaterfallGeneratorEditorComp)))

(fn Generation_PoolSplineGen
  (bind _returnvalue (Math|Vector|MakeVector 0.0))
  (bind _thickness (Variables|TmpVarButBake|GetThickness))
  (bind _returnvalue_2 (* _thickness 0.5))
  (bind _startpointedgeoffset (Variables|TopConnectionWaterBody|GetStartPointEdgeOffset))
  (bind _bevel_horizontal (Variables|Widget|Get◉Bevel_Horizontal))
  (bind _spline_customwaterbody (Variables|Default|GetSpline_CustomWaterBody))
  (bind _width (Variables|TmpVarButBake|GetWidth))
  (Utilities|FlowControl|Sequence
    (:then_1
      (Spline|SetLocationatSplinePoint _spline_customwaterbody 0 (+ _bevel_horizontal (Math|Vector|MakeVector _startpointedgeoffset (+ (/ _width 2.0) (+ _returnvalue_2 (Variables|TopConnectionWaterBody|GetStartPointWidthOffset))) _returnvalue_2)) "Local" false)
      (Spline|SetLocationatSplinePoint _spline_customwaterbody 1 (+ _bevel_horizontal (Math|Vector|MakeVector _startpointedgeoffset (+ (/ _width -2.0) (+ (* (Variables|TmpVarButBake|GetThickness) -0.5) (* (Variables|TopConnectionWaterBody|GetStartPointWidthOffset) -1.0))) _returnvalue_2)) "Local" false)
      (Spline|SetSplinePointType _spline_customwaterbody 0 "Linear" false)
      (Spline|SetSplinePointType _spline_customwaterbody 1 "Linear" true)
      (Spline|SetTangentsatSplinePoint _spline_customwaterbody 0 _returnvalue _returnvalue "Local" true)
      (Spline|SetTangentsatSplinePoint _spline_customwaterbody 1 _returnvalue _returnvalue "Local" true))
    (:then_2
      (for _index (range (Spline|GetNumberOfSplinePoints (Variables|Default|GetSpline_CustomWaterBody)))
        (Spline|SetLocationatSplinePoint (Variables|Default|GetSpline_CustomWaterBody) _index (Math|Vector|MakeVector (.x (Spline|GetLocationatSplinePoint (Variables|Default|GetSpline_CustomWaterBody) _index "Local")) (.y (Spline|GetLocationatSplinePoint (Variables|Default|GetSpline_CustomWaterBody) _index "Local")) (+ (.z (Variables|Widget|Get◉Bevel_Horizontal)) (* (Variables|TmpVarButBake|GetThickness) 0.5))) "Local" true)))))

(fn Generation_SampleSpline
  ;[1 1 1 1 128] Collect Generation data
  (Utilities|FlowControl|Sequence
    (:then_0
      (Generation|GenerationSubFuncSampleBAF (Math|Integer|SelectInt (Variables|SampleSetting|GetSampleCount_RoundFall) (Variables|SampleSetting|GetSampleCount_Fall) (Variables|Initial|GetbRoundFall)) (Variables|TmpSampleVar|GetTmpSampleSP_BAF) (Variables|SampleSetting|GetSampleCount_Bevel) (Variables|SampleSetting|GetSampleCount_Extend)))
    (:then_1
      (Generation|GenerationSubFuncSampleBAF (Math|Integer|SelectInt (Variables|SampleSetting|GetSampleCount_RoundFall_LOD1) (Variables|SampleSetting|GetSampleCount_Fall_LOD1) (Variables|Initial|GetbRoundFall)) (Variables|TmpSampleVar|GetTmpSampleSP_BAF_LOD1) (Variables|SampleSetting|GetSampleCount_Bevel_LOD1) (Variables|SampleSetting|GetSampleCount_Extend_LOD1)))
    (:then_2
      (Generation|GenerationSubFuncSampleBAF (Math|Integer|SelectInt (Variables|SampleSetting|GetSampleCount_RoundFall_LOD2) (Variables|SampleSetting|GetSampleCount_Fall_LOD2) (Variables|Initial|GetbRoundFall)) (Variables|TmpSampleVar|GetTmpSampleSP_BAF_LOD2) (Variables|SampleSetting|GetSampleCount_Bevel_LOD2) (Variables|SampleSetting|GetSampleCount_Extend_LOD2)))))

(fn Generation_SaveMaterialAndFXData
  (bind _spline_bevelandfall (Variables|Default|GetSpline_BevelAndFall))
  (bind _spline_bevelandfall_2 (Variables|Default|GetSpline_BevelAndFall))
  (Variables|TmpVarButBake|SetLength (Spline|GetSplineLength _spline_bevelandfall))
  (Variables|TmpVarButBake|SetFallLength (Spline|GetDistanceAlongSplineatSplinePoint _spline_bevelandfall 1))
  (Variables|TmpVarButBake|SetBevelLength (- (Spline|GetDistanceAlongSplineatSplinePoint _spline_bevelandfall 2) (Spline|GetDistanceAlongSplineatSplinePoint _spline_bevelandfall 1)))
  (Variables|TmpVarButBake|SetBottomSplinePointPosition (Spline|GetLocationatSplinePoint _spline_bevelandfall_2 0 "World"))
  (Variables|TmpVarButBake|SetTopSplinePointPosition (Spline|GetLocationatSplinePoint _spline_bevelandfall_2 2 "World")))

(fn Generation_SplineGen
  (bind _bottom (Variables|Widget|Get◉Bottom))
  (bind _x (.x _bottom))
  (bind _bevel_vertical (Variables|Widget|Get➤Bevel_Vertical))
  (bind _y (.y _bevel_vertical))
  (bind _returnvalue (Math|Float|Absolute(Float) _y))
  (bind _top (Variables|Widget|Get◉Top))
  (bind _z (.z _top))
  (bind _returnvalue_2 (- _z _returnvalue))
  (bind _spline_bevelandfall (Variables|Default|GetSpline_BevelAndFall))
  (bind _spline_bevelandfall_2 (Variables|Default|GetSpline_BevelAndFall))
  (bind _bevel_horizontal (Variables|Widget|Get◉Bevel_Horizontal))
  (bind _spline_bevelandfall_3 (Variables|Default|GetSpline_BevelAndFall))
  (bind _spline_bevelandfall_4 (Variables|Default|GetSpline_BevelAndFall))
  ;[1 1 1 1 128] SplineGen
  (Spline|ClearSplinePoints (Variables|Default|GetSpline_BevelAndFall) true)
  (Utilities|FlowControl|Sequence
    (:then_0
      (Utilities|FlowControl|Sequence
        (:then_0
          (Spline|AddSplinePoint _spline_bevelandfall_3 (Variables|Widget|Get◉Bottom) "Local" false)
          (Spline|SetTangentsatSplinePoint _spline_bevelandfall_3 0 (Math|Vector|MakeVector 0.0) (Math|Vector|MakeVector 0.0 0.0 (Math|Float|SelectFloat (* (* (- (.z (Variables|Widget|Get◉Top)) (.z (Variables|Widget|Get◉Bottom))) 2.0) (+ (/ (* (Math|Float|Sign(Float) (.y (Variables|Widget|Get➤Bevel_Vertical))) (Math|Float|Max(Float) (- (Math|Float|Absolute(Float) (.y (Variables|Widget|Get➤Bevel_Vertical))) 1e+02))) 3e+02) 1.0)) 0.0 (Variables|Initial|GetbRoundFall))) "Local" false))
        (:then_1
          (Spline|AddSplinePoint _spline_bevelandfall_2 (Math|Vector|MakeVector (Math|Float|Lerp _x (.x _top) (/ _returnvalue_2 (- _z (.z _bottom)))) (.y _top) _returnvalue_2) "Local" false)
          (Spline|SetTangentsatSplinePoint _spline_bevelandfall_2 1 (Math|Vector|MakeVector 0.0) (Math|Vector|MakeVector (* _x -0.5) 0.0 (* (- _z _returnvalue_2) 2.0)) "Local" false))
        (:then_2
          (Spline|AddSplinePoint _spline_bevelandfall _bevel_horizontal "Local" false)
          (Spline|SetTangentsatSplinePoint _spline_bevelandfall 2 (Math|Vector|MakeVector (* (.x _bevel_horizontal) (Variables|TmpVarButBake|GetBevel_H_ArriveTangentScale))) (Math|Vector|MakeVector 0.0) "Local" false))
        (:then_3
          (Spline|AddSplinePoint _spline_bevelandfall_4 (Variables|Widget|Get◉Extend) "Local" false)
          (Spline|SetTangentsatSplinePoint _spline_bevelandfall_4 3 (Math|Vector|MakeVector 0.0) (Math|Vector|MakeVector 0.0) "Local" false))
        (:then_4
          (if (Variables|Initial|GetbRoundFall)
            (Spline|RemoveSplinePoint (Variables|Default|GetSpline_BevelAndFall) 1 true)
            (else
              (Spline|UpdateSpline (Variables|Default|GetSpline_BevelAndFall)))))))))

(fn Generation_SubFunc_SampleBAF
  (params (SampleCount_Fall :type Integer) (TargetArray :type Array%20of%20Vectors) (SampleCount_Bevel :type Integer) (SampleCount_Extend :type Integer))
  (Utilities|FlowControl|Sequence
    (:then_0
      ;[1 1 1] Comment
      (for _index (range (+ SampleCount_Fall 1))
        (Utilities|Array|Add (LocalVariables|GetTmpArray) (Spline|GetLocationatTime (Variables|Default|GetSpline_BevelAndFall) (* (/ (Math|Conversions|ToFloat(Integer) _index) (Math|Conversions|ToFloat(Integer) SampleCount_Fall)) (/ (Spline|GetDistanceAlongSplineatSplinePoint (Variables|Default|GetSpline_BevelAndFall) 1) (Spline|GetSplineLength (Variables|Default|GetSpline_BevelAndFall)))) "Local" true))))
    (:then_1
      ;[1 1 1] Comment
      (for _index_2 (range 1 (+ SampleCount_Bevel 1))
        (Utilities|Array|Add (LocalVariables|GetTmpArray) (Spline|GetLocationatTime (Variables|Default|GetSpline_BevelAndFall) (+ (* (/ (Math|Conversions|ToFloat(Integer) _index_2) (Math|Conversions|ToFloat(Integer) SampleCount_Bevel)) (- (/ (Spline|GetDistanceAlongSplineatSplinePoint (Variables|Default|GetSpline_BevelAndFall) 2) (Spline|GetSplineLength (Variables|Default|GetSpline_BevelAndFall))) (/ (Spline|GetDistanceAlongSplineatSplinePoint (Variables|Default|GetSpline_BevelAndFall) 1) (Spline|GetSplineLength (Variables|Default|GetSpline_BevelAndFall))))) (/ (Spline|GetDistanceAlongSplineatSplinePoint (Variables|Default|GetSpline_BevelAndFall) 1) (Spline|GetSplineLength (Variables|Default|GetSpline_BevelAndFall)))) "Local" true))))
    (:then_2
      ;[1 1 1] Comment
      (for _index_3 (range 1 (+ SampleCount_Extend 1))
        (Utilities|Array|Add (LocalVariables|GetTmpArray) (Spline|GetLocationatTime (Variables|Default|GetSpline_BevelAndFall) (+ (* (/ (Math|Conversions|ToFloat(Integer) _index_3) (Math|Conversions|ToFloat(Integer) SampleCount_Extend)) (- 1.0 (/ (Spline|GetDistanceAlongSplineatSplinePoint (Variables|Default|GetSpline_BevelAndFall) 2) (Spline|GetSplineLength (Variables|Default|GetSpline_BevelAndFall))))) (/ (Spline|GetDistanceAlongSplineatSplinePoint (Variables|Default|GetSpline_BevelAndFall) 2) (Spline|GetSplineLength (Variables|Default|GetSpline_BevelAndFall)))) "Local" true))))
    (:then_3
      (Variables|SetBy-RefVar TargetArray (LocalVariables|GetTmpArray)))))

(fn GetClipMaterialParameterNames
  (params (ChannelIndex :type Integer))
  (return (Utilities|Select :Index ChannelIndex :Option%200 "WaterfallClipBound_A_Position" :Option%201 "WaterfallClipBound_B_Position" :Option%202 "WaterfallClipBound_C_Position" :Option%203 "WaterfallClipBound_D_Position") (Utilities|Select :Index ChannelIndex :Option%200 "WaterfallClipBound_A_Rotation" :Option%201 "WaterfallClipBound_B_Rotation" :Option%202 "WaterfallClipBound_C_Rotation" :Option%203 "WaterfallClipBound_D_Rotation") (Utilities|Select :Index ChannelIndex :Option%200 "WaterfallClipBound_A_Extend" :Option%201 "WaterfallClipBound_B_Extend" :Option%202 "WaterfallClipBound_C_Extend" :Option%203 "WaterfallClipBound_D_Extend")))

(fn GetMapKeyAndValue_Bool
  (params (In :type Map%20of%20Names%20to%20Booleans))
  (bind _keys (Utilities|Map|Keys In))
  (bind _values (Utilities|Map|Values In))
  (return _keys _values))

(fn GetMapKeyAndValue_Color
  (params (In :type Map%20of%20Names%20to%20Linear%20Color%20Structures))
  (bind _keys (Utilities|Map|Keys In))
  (bind _values (Utilities|Map|Values In))
  (return _keys _values))

(fn GetMapKeyAndValue_Float
  (params (In :type Map%20of%20Names%20to%20Float%20%28double-precision%29s))
  (bind _keys (Utilities|Map|Keys In))
  (bind _values (Utilities|Map|Values In))
  (return _keys _values))

(fn GetSweptPointArray
  (params (Offset :type Float%20%28double-precision%29) (EdgeSeg :type Integer) (CenterSeg :type Integer))
  (bind _width (Variables|TmpVarButBake|GetWidth))
  (bind _returnvalue (+ _width (Variables|GetOffset)))
  (bind _returnvalue_2 (/ _returnvalue 2.0))
  (bind _offset (Variables|GetOffset))
  (Utilities|FlowControl|Sequence
    (:then_0
      (Utilities|FlowControl|Sequence
        (:then_1
          ;[1 1 1] Prepare
          (Utilities|FlowControl|Sequence
            (:then_0
              (LocalVariables|SetLeftO (Math|Vector2D|MakeVector2D 0.0 (* _returnvalue_2 -1.0)))
              (LocalVariables|SetRightO (Math|Vector2D|MakeVector2D 0.0 _returnvalue_2)))
            (:then_1
              (LocalVariables|SetR (/ (+ _offset (Variables|TmpVarButBake|GetThickness)) 2.0)))
            (:then_2
              (LocalVariables|SetAlpha (/ 9e+01 (Math|Conversions|ToFloat(Integer) (+ (Variables|GetEdgeSeg) 1)))))
            (:then_3
              (LocalVariables|SetCentSegLength (/ (+ (Variables|TmpVarButBake|GetWidth) _offset) (Math|Conversions|ToFloat(Integer) (+ (Variables|GetCenterSeg) 1)))))))
        (:then_2
          ;[1 1 1] -    -    -
          ; /               \
          ; T                 *
          ; \               /
          ; -    -    -
          (Utilities|Array|Add (LocalVariables|GetPointArray) (- (LocalVariables|GetLeftO) (Math|Vector2D|MakeVector2D 0.0 (LocalVariables|GetR)))))
        (:then_3
          ;[1 1 1] -    -    -
          ; /               \
          ; *                 *
          ; T               /
          ; -    -    -
          (bind _index (Utilities|FlowControl|ForLoopwithBreak 1 (Variables|GetEdgeSeg))
            (:LoopBody
              (Utilities|Array|Add (LocalVariables|GetPointArray) (+ (LocalVariables|GetLeftO) (* (Math|Vector2D|MakeVector2D (* (Math|Trig|Cos(Degrees) (- 9e+01 (* (LocalVariables|GetAlpha) (Math|Conversions|ToFloat(Integer) _index)))) 1.0) (* (Math|Trig|Sin(Degrees) (- 9e+01 (* (LocalVariables|GetAlpha) (Math|Conversions|ToFloat(Integer) _index)))) -1.0)) (Math|Vector2D|MakeVector2D (LocalVariables|GetR) (LocalVariables|GetR))))))))
        (:then_4
          ;[1 1 1] -    -    -
          ; /               \
          ; *                 *
          ; \               /
          ; T    -    -
          (Utilities|Array|Add (LocalVariables|GetPointArray) (+ (LocalVariables|GetLeftO) (Math|Vector2D|MakeVector2D (LocalVariables|GetR)))))
        (:then_5
          ;[1 1 1] -    -    -
          ; /               \
          ; *                 *
          ; \               /
          ; -    T    -
          (bind _index_2 (Utilities|FlowControl|ForLoopwithBreak 1 (Variables|GetCenterSeg))
            (:LoopBody
              (Utilities|Array|Add (LocalVariables|GetPointArray) (Utilities|Operators|Add (LocalVariables|GetLeftO) (Math|Vector2D|MakeVector2D (LocalVariables|GetR)) (Math|Vector2D|MakeVector2D 0.0 (* (LocalVariables|GetCentSegLength) (Math|Conversions|ToFloat(Integer) _index_2))))))))
        (:then_6
          ;[1 1 1] -    -    -
          ; /               \
          ; *                 *
          ; \               /
          ; -    -    T
          (Utilities|Array|Add (LocalVariables|GetPointArray) (+ (LocalVariables|GetRightO) (Math|Vector2D|MakeVector2D (LocalVariables|GetR)))))
        (:then_7
          ;[1 1 1] -    -    -
          ; /               \
          ; *                 *
          ; \               T
          ; -    -    -
          (bind _index_3 (Utilities|FlowControl|ForLoopwithBreak 1 (Variables|GetEdgeSeg))
            (:LoopBody
              (Utilities|Array|Add (LocalVariables|GetPointArray) (+ (LocalVariables|GetRightO) (* (Math|Vector2D|MakeVector2D (* (Math|Trig|Cos(Degrees) (* (LocalVariables|GetAlpha) (Math|Conversions|ToFloat(Integer) _index_3))) 1.0) (* (Math|Trig|Sin(Degrees) (* (LocalVariables|GetAlpha) (Math|Conversions|ToFloat(Integer) _index_3))) 1.0)) (Math|Vector2D|MakeVector2D (LocalVariables|GetR) (LocalVariables|GetR))))))))
        (:then_8
          ;[1 1 1] -    -    -
          ; /               \
          ; *                 T
          ; \               /
          ; -    -    -
          (Utilities|Array|Add (LocalVariables|GetPointArray) (+ (LocalVariables|GetRightO) (Math|Vector2D|MakeVector2D 0.0 (LocalVariables|GetR)))))
        (:then_9
          ;[1 1 1] -    -    -
          ; /               T
          ; *                 *
          ; \               /
          ; -    -    -
          (bind _index_4 (Utilities|FlowControl|ForLoopwithBreak 1 (Variables|GetEdgeSeg))
            (:LoopBody
              (Utilities|Array|Add (LocalVariables|GetPointArray) (+ (LocalVariables|GetRightO) (* (Math|Vector2D|MakeVector2D (* (Math|Trig|Cos(Degrees) (- 9e+01 (* (LocalVariables|GetAlpha) (Math|Conversions|ToFloat(Integer) _index_4)))) -1.0) (* (Math|Trig|Sin(Degrees) (- 9e+01 (* (LocalVariables|GetAlpha) (Math|Conversions|ToFloat(Integer) _index_4)))) 1.0)) (Math|Vector2D|MakeVector2D (LocalVariables|GetR) (LocalVariables|GetR))))))))
        (:then_10
          ;[1 1 1] -    -    T
          ; /               \
          ; *                 *
          ; \               /
          ; -    -    -
          (Utilities|Array|Add (LocalVariables|GetPointArray) (- (LocalVariables|GetRightO) (Math|Vector2D|MakeVector2D (LocalVariables|GetR)))))
        (:then_11
          ;[1 1 1] -    T    -
          ; /               \
          ; *                 *
          ; \               /
          ; -    -    -
          (bind _index_5 (Utilities|FlowControl|ForLoopwithBreak 1 (Variables|GetCenterSeg))
            (:LoopBody
              (Utilities|Array|Add (LocalVariables|GetPointArray) (Utilities|Operators|Add (LocalVariables|GetRightO) (Math|Vector2D|MakeVector2D (* (LocalVariables|GetR) -1.0)) (Math|Vector2D|MakeVector2D 0.0 (* (* (LocalVariables|GetCentSegLength) (Math|Conversions|ToFloat(Integer) _index_5)) -1.0)))))))
        (:then_12
          ;[1 1 1] T    -    -
          ; /               \
          ; *                 *
          ; \               /
          ; -    -    -
          (Utilities|Array|Add (LocalVariables|GetPointArray) (- (LocalVariables|GetLeftO) (Math|Vector2D|MakeVector2D (LocalVariables|GetR)))))
        (:then_13
          ;[1 1 1] -    -    -
          ; /               \
          ; *                 *
          ; \               T
          ; -    -    -
          (bind _index_6 (Utilities|FlowControl|ForLoopwithBreak 1 (Variables|GetEdgeSeg))
            (:LoopBody
              (Utilities|Array|Add (LocalVariables|GetPointArray) (+ (LocalVariables|GetLeftO) (* (Math|Vector2D|MakeVector2D (* (Math|Trig|Cos(Degrees) (* (LocalVariables|GetAlpha) (Math|Conversions|ToFloat(Integer) _index_6))) -1.0) (* (Math|Trig|Sin(Degrees) (* (LocalVariables|GetAlpha) (Math|Conversions|ToFloat(Integer) _index_6))) -1.0)) (Math|Vector2D|MakeVector2D (LocalVariables|GetR) (LocalVariables|GetR))))))))))
    (:then_1
      (return (LocalVariables|GetPointArray)))))

(fn InitialClipWaterBody
  (bind _thickness (Variables|TmpVarButBake|GetThickness))
  (bind _returnvalue (* _thickness 0.5))
  (bind _returnvalue_2 (Math|Vector|MakeVector 0.0 0.0 _returnvalue))
  (bind _bottomclipoffset (Variables|ClipWaterBody|GetBottomClipOffset))
  (bind _returnvalue_3 (Math|Vector|MakeVector 0.0 0.0 _bottomclipoffset))
  (bind _topclipoffset (Variables|ClipWaterBody|GetTopClipOffset))
  (bind _returnvalue_4 (Math|Vector|MakeVector 0.0 0.0 _topclipoffset))
  (bind _clipvtransform (Variables|Widget|GetClipVTransform))
  (bind _scale (.scale _clipvtransform))
  (bind _bottom (Variables|Widget|Get◉Bottom))
  (bind _returnvalue_5 (+ _bottom _returnvalue_3))
  (bind _top (Variables|Widget|Get◉Top))
  (bind _returnvalue_6 (+ _top _returnvalue_2))
  (bind _returnvalue_7 (- _returnvalue_6 _returnvalue_4))
  (bind _returnvalue_8 (- _returnvalue_7 _returnvalue_5))
  (bind _z (.z _returnvalue_8))
  (bind _clipwaterbody (Variables|Default|GetClipWaterBody))
  (bind _clipvtransform_2 (Variables|Widget|GetClipVTransform))
  (bind _clipwaterbody_2 (Variables|Default|GetClipWaterBody))
  (Variables|Widget|SetClipVTransform (Math|Transform|MakeTransform (Math|Vector|MakeVector (.x (.location _clipvtransform)) (.y (.location _clipvtransform)) (+ (/ _z 2.0) (.z _returnvalue_5))) (.rotation _clipvtransform) _scale))
  (bind _returnvalue_9 (FortPhysicsSimulationLibrary|GetBoxExtent _clipwaterbody))
  (Components|Box|SetBoxExtent _clipwaterbody (Math|Vector|MakeVector (* (.x _returnvalue_9) (.x _scale)) (* (.y _returnvalue_9) (.y _scale)) (/ _z 2.0)) true)
  (Transformation|SetRelativeLocation _clipwaterbody (.location _clipvtransform_2) false false)
  (Transformation|SetRelativeRotation _clipwaterbody (.rotation _clipvtransform_2) false false)
  (if (Variables|ClipWaterBody|GetWaterBodyClip)
    (Rendering|SetVisibility _clipwaterbody_2 true false)
    (else
      (Rendering|SetVisibility _clipwaterbody_2 false false))))

(fn InitialFX
  (bind _fxcustomparameters_float (Variables|MatAndFXParameters|GetFXCustomParameters_Float))
  (bind _bottomsplinepointposition (Variables|TmpVarButBake|GetBottomSplinePointPosition))
  (bind _waterfallfx (Variables|Default|GetWaterfallFX))
  (bind _fxcustomparameters_color (Variables|MatAndFXParameters|GetFXCustomParameters_Color))
  (bind _waterfallfx_2 (Variables|Default|GetWaterfallFX))
  (Utilities|FlowControl|Sequence
    (:then_0
      (Niagara|SetNiagaraSystemAsset (Variables|Default|GetWaterfallFX) (Variables|OverrideMaterialandFx|GetOverrideNiagaraSystem) true))
    (:then_1
      (bind _keys (Utilities|Map|Keys _fxcustomparameters_float))
      (bind _values (Utilities|Map|Values _fxcustomparameters_float))
      (for (_array_element _array_index) _keys
        (Niagara|SetNiagaraVariable(Float) (Variables|Default|GetWaterfallFX) _array_element (Utilities|Array|Get(acopy) _values _array_index))))
    (:then_2
      (bind _keys_2 (Utilities|Map|Keys _fxcustomparameters_color))
      (bind _values_2 (Utilities|Map|Values _fxcustomparameters_color))
      (for (_array_element_2 _array_index_2) _keys_2
        (Niagara|SetNiagaraVariable(LinearColor) (Variables|Default|GetWaterfallFX) _array_element_2 (Utilities|Array|Get(acopy) _values_2 _array_index_2))))
    (:then_3
      (Niagara|SetNiagaraVariable(Position) _waterfallfx "TopPosition" (Math|Vector|MakeVector (.x _bottomsplinepointposition) (.y _bottomsplinepointposition) (.z (Variables|TmpVarButBake|GetTopSplinePointPosition))))
      (Niagara|SetNiagaraVariable(Position) _waterfallfx "BottomPosition" _bottomsplinepointposition)
      (Niagara|SetNiagaraVariable(Float) _waterfallfx "FallWidth" (Variables|TmpVarButBake|GetWidth))
      (Niagara|SetNiagaraVariable(Float) _waterfallfx "FallThickness" (Variables|TmpVarButBake|GetThickness))
      (Niagara|SetNiagaraVariable(LinearColor) _waterfallfx "BPInput_FXTint" (Variables|MatAndFXParameters|GetFXTint))
      (Niagara|SetNiagaraVariable(Float) _waterfallfx "BPInput_Opacity" (Variables|MatAndFXParameters|GetFXOpacity)))
    (:then_4
      (Niagara|SetNiagaraVariable(Vector4) _waterfallfx_2 "TopSplashMultipliers" (Math|Vector4|MakeVector4 (Variables|MatAndFXParameters|GetFXOpacityTopSplashesMultiplier) (Variables|MatAndFXParameters|GetFXScaleTopSplashesMultiplier) (Variables|MatAndFXParameters|GetFXSpawnRateTopSplashesMultiplier)))
      (Niagara|SetNiagaraVariable(Vector4) _waterfallfx_2 "BottomSplashMultipliers" (Math|Vector4|MakeVector4 (Variables|MatAndFXParameters|GetFXOpacityBottomSplashesMultiplier) (Variables|MatAndFXParameters|GetFXScaleBottomSplashesMultiplier) (Variables|MatAndFXParameters|GetFXSpawnRateBottomSplashesMultiplier)))
      (Niagara|SetNiagaraVariable(Float) _waterfallfx_2 "ScaleTopSmoke" (Variables|MatAndFXParameters|GetFXScaleTopSmoke)))))

(fn InitialMaterialOnMesh
  (params (InMesh :type Mesh%20Component%20Object%20Reference))
  (bind _bottomlinktolake (Variables|Initial|GetBottomLinkToLake))
  (bind _output (Utilities|ResolveSoftReference _bottomlinktolake))
  (bind _dynamicsplashmaterial (Variables|TmpMaterial|GetDynamicSplashMaterial))
  (bind _matcustomparameters_float (Variables|MatAndFXParameters|GetMatCustomParameters_Float))
  (bind _matcustomparameters_color (Variables|MatAndFXParameters|GetMatCustomParameters_Color))
  (Utilities|FlowControl|Sequence
    (:then_0
      (Utilities|IsValid _output
        (:Is%20Valid
          (bind _returnvalue (Rendering|GetWaterMaterialInstance (Variables|/Script/Water.WaterBody|GetWaterBodyComponent _output)))
          (Rendering|Material|CopyMaterialInstanceParameters _dynamicsplashmaterial _returnvalue false)
          (Rendering|Material|SetScalarParameterValue _dynamicsplashmaterial "LinkToWaterBodyIndex" (Math|Conversions|ToFloat(Integer) (Variables|/Script/Water.WaterBody|GetWaterBodyIndex _output)) :!class /Script/Engine.MaterialInstanceDynamic))
        (:Is%20Not%20Valid
          (CallFunction|DelaySetupMaterial))))
    (:then_1
      (Rendering|Material|SetScalarParameterValueonMaterials InMesh "BPInput_Width" (+ (Variables|TmpVarButBake|GetWidth) (* (Variables|TmpVarButBake|GetThickness) 1.57)))
      (Rendering|Material|SetScalarParameterValueonMaterials InMesh "BPInput_Length" (Variables|TmpVarButBake|GetLength))
      (Rendering|Material|SetScalarParameterValueonMaterials InMesh "BPInput_FallLength" (Variables|TmpVarButBake|GetFallLength))
      (Rendering|Material|SetScalarParameterValueonMaterials InMesh "BPInput_BevelLength" (Variables|TmpVarButBake|GetBevelLength))
      (Rendering|Material|SetVectorParameterValueonMaterials InMesh "BPInput_BaseColor" (Math|Conversions|ToVector(LinearColor) (Variables|MatAndFXParameters|GetBaseColor)))
      (Rendering|Material|SetVectorParameterValueonMaterials InMesh "BPInput_EmissiveColor" (Math|Conversions|ToVector(LinearColor) (Variables|MatAndFXParameters|GetEmissiveColor)))
      (Rendering|Material|SetVectorParameterValueonMaterials InMesh "BPInput_FoamColor" (Math|Conversions|ToVector(LinearColor) (Variables|MatAndFXParameters|GetFoamColor)))
      (Rendering|Material|SetVectorParameterValueonMaterials InMesh "BPInput_RippleColor" (Math|Conversions|ToVector(LinearColor) (Variables|MatAndFXParameters|GetRippleColor)))
      (Rendering|Material|SetScalarParameterValueonMaterials InMesh "DisableEdgeMaskSideForDoubleFaceWaterfall" (Math|Conversions|ToFloat(Boolean) (Variables|Setting|GetbDoubleFaceMesh))))
    (:then_2
      (bind _keys (Utilities|Map|Keys _matcustomparameters_float))
      (bind _values (Utilities|Map|Values _matcustomparameters_float))
      (for (_array_element _array_index) _keys
        (Rendering|Material|SetScalarParameterValueonMaterials InMesh _array_element (Utilities|Array|Get(acopy) _values _array_index))))
    (:then_3
      (bind _keys_2 (Utilities|Map|Keys _matcustomparameters_color))
      (bind _values_2 (Utilities|Map|Values _matcustomparameters_color))
      (for (_array_element_2 _array_index_2) _keys_2
        (Rendering|Material|SetVectorParameterValueonMaterials InMesh _array_element_2 (Math|Conversions|ToVector(LinearColor) (Utilities|Array|Get(acopy) _values_2 _array_index_2)))))))

(fn InitialPostProcess)

(fn InitialTopConnectionWaterBody
  (Utilities|IsValid (Variables|Initial|GetTargetStaticMesh_CustomWaterBody)
    (:Is%20Valid
      (Utilities|IsValid (Utilities|ResolveSoftReference (Variables|TopConnectionWaterBody|GetCustomWaterBodyActor_Soft))
        (:Is%20Valid
          _)
        (:Is%20Not%20Valid
          (bind _returnvalue (Game|SpawnActorfromClass (Variables|OverrideMaterialandFx|GetCustomWaterBodyClass) (Math|Transform|ComposeTransforms (Transformation|GetActorTransform) (Math|Transform|MakeTransform (+ (Variables|Widget|Get◉Top) (Math|Vector|MakeVector 0.0 0.0 (+ (* (Variables|TmpVarButBake|GetThickness) 0.5) (Variables|TopConnectionWaterBody|GetCustomWaterBodyActorZOffset)))))) "AlwaysSpawn" "SelectDefaultAtRuntime" 0 0))
          (Utilities|IsValid _returnvalue
            (:Is%20Valid
              (Variables|TopConnectionWaterBody|SetCustomWaterBodyActor_Soft (Utilities|ResolveSoftReference _returnvalue))))))
      (Default|CustomWaterBodySet))))

(fn ManualUpdateStaticMesh
  (Generation|GenerationBakeMesh)
  (Class|BIWaterfallGenerator|DelayExecute true))

(fn OpenDoc
  (Utilities|Platform|LaunchURL "https://docs.google.com/document/d/1dcYAm1iAYME8MyC_wKeNUsIbU3cg58YyROa-WvupF7w/edit?usp=sharing"))

(fn TransitWaterbodyClipInformation
  (params (WaterBody :type Fort%20Water%20Body%20Actor%20Object%20Reference))
  (bind _waterbodyclipchannel (Variables|ClipWaterBody|GetWaterBodyClipChannel))
  (bind (_p _r _e) (Base|GetClipMaterialParameterNames _waterbodyclipchannel))
  (bind _waterbodycomponent (Variables|/Script/Water.WaterBody|GetWaterBodyComponent WaterBody))
  (bind _clipwaterbody (Variables|Default|GetClipWaterBody))
  (bind _returnvalue (Rendering|GetWaterMaterialInstance _waterbodycomponent))
  (bind _returnvalue_2 (Rendering|GetRivertoLakeTransitionMaterialInstance _waterbodycomponent))
  (bind _returnvalue_3 (Rendering|GetRivertoOceanTransitionMaterialInstance _waterbodycomponent))
  (bind _returnvalue_4 (Rendering|GetWaterLODMaterialInstance _waterbodycomponent))
  (bind _returnvalue_5 (Rendering|GetWaterInfoMaterialInstance _waterbodycomponent))
  (Rendering|Material|SetVectorParameterValue _returnvalue _p (Math|Conversions|ToLinearColor(Vector) (Transformation|GetWorldLocation _clipwaterbody)) :!class /Script/Engine.MaterialInstanceDynamic)
  (Rendering|Material|SetVectorParameterValue _returnvalue _r (Utilities|Struct|MakeLinearColor (.roll (Transformation|GetWorldRotation _clipwaterbody)) (.pitch (Transformation|GetWorldRotation _clipwaterbody)) (.yaw (Transformation|GetWorldRotation _clipwaterbody))) :!class /Script/Engine.MaterialInstanceDynamic)
  (Rendering|Material|SetVectorParameterValue _returnvalue _e (Math|Conversions|ToLinearColor(Vector) (Components|Box|GetScaledBoxExtent _clipwaterbody)) :!class /Script/Engine.MaterialInstanceDynamic))

(fn ConstructionScript
  (bind _waterfallmesh (Variables|Default|GetWaterfallMesh))
  (bind _frozentip (Variables|Default|GetFrozenTip))
  (bind _dynamicmesh (Variables|Default|GetDynamicMesh))
  (bind _waterfallmesh_2 (Variables|Default|GetWaterfallMesh))
  (if (Variables|Setting|GetbPreviewGame)
    _
    (elif (Variables|Setting|GetbFrozen)
      (Base|UserInterfaceLimit3DWidget)
      (CallFunction|InitialClipWaterBody)
      (CallFunction|DotheWaterbodyClip)
      (Rendering|SetVisibility _frozentip true false)
      (Transformation|SetRelativeLocation _frozentip (+ (Variables|Widget|Get◉WdithPoint) "0.000000,0.000000,500.000000") false false)
      (Default|FixCustomWaterBodyWaveSource)
      (Utilities|IsValid (Variables|Initial|GetTargetStaticMesh)
        (:Is%20Valid
          (Components|StaticMesh|SetStaticMesh (Variables|Default|GetWaterfallMesh) (Variables|Initial|GetTargetStaticMesh))
          (Material|CreateDynamicMaterial)
          (Rendering|Material|SetMaterial _waterfallmesh_2 0 (Variables|TmpMaterial|GetDynamicFallMaterial))
          (Rendering|Material|SetMaterial (Variables|Default|GetWaterfallMesh) 1 (Variables|TmpMaterial|GetDynamicSplashMaterial))
          (Rendering|Material|SetMaterial (Variables|Default|GetWaterfallMesh) 2 0)
          (Material|InitialMaterialonMesh _waterfallmesh_2)
          (FX|InitialFX)
          (GameplayAndAudio|GameplayThingsRefresh)
          (GameplayandAudio|AudioSetting)
          (Default|InitialPostProcess)))
      (else
        (Rendering|SetVisibility _frozentip false false)
        (if (Editor|InEditor)
          (Development|PrintString "WaterfallWorkInEditor" true true "(R=1.000000,G=0.000000,B=0.016322,A=1.000000)" 8.0)
          (Utilities|FlowControl|Sequence
            (:then_0
              (Variables|/Game/Tools/NewWaterfallGenerator/Blueprint/BP_WaterfallGeneratorEditorComp.BP_WaterfallGeneratorEditorComp_C|SetTargetActor self (Variables|Default|GetBP_WaterfallGeneratorEditorComp))
              (Base|UserInterfaceLimit3DWidget))
            (:then_1
              (Generation|GenerationSplineGen))
            (:then_2
              (Generation|GenerationSampleSpline))
            (:then_3
              ;[1 1 1] BI
              (Generation|GenerationMeshGen))
            (:then_4
              (Generation|GenerationMeshMerge))
            (:then_5
              (if (Variables|Setting|GetbAutoUpdateStaticMesh)
                (Utilities|Time|SetTimerbyFunctionName (Variables|Default|GetBP_WaterfallGeneratorEditorComp) "BakeWaterfallMeshDispatcher" 0.3333)))
            (:then_6
              (Utilities|Time|SetTimerbyFunctionName (Variables|Default|GetBP_WaterfallGeneratorEditorComp) "DelayExecuteDispatcher" 0.1))
            (:then_7
              (Generation|GenerationSaveMaterialandFXData))
            (:then_8
              (CallFunction|EditorInitialCustomParameters))
            (:then_9
              (CallFunction|InitialClipWaterBody))
            (:then_10
              (Utilities|IsValid (Variables|Initial|GetTargetStaticMesh)
                (:Is%20Valid
                  (Rendering|SetVisibility (Variables|Default|GetDynamicMesh) false false)
                  (Components|StaticMesh|SetStaticMesh (Variables|Default|GetWaterfallMesh) (Variables|Initial|GetTargetStaticMesh))
                  (Material|CreateDynamicMaterial)
                  (Rendering|Material|SetMaterial _waterfallmesh 0 (Variables|TmpMaterial|GetDynamicFallMaterial))
                  (Rendering|Material|SetMaterial (Variables|Default|GetWaterfallMesh) 1 (Variables|TmpMaterial|GetDynamicSplashMaterial))
                  (Rendering|Material|SetMaterial (Variables|Default|GetWaterfallMesh) 2 0)
                  (Material|InitialMaterialonMesh _waterfallmesh))))
            (:then_11
              (Material|CreateDynamicMaterial)
              (Rendering|Material|SetMaterial _dynamicmesh 0 (Variables|TmpMaterial|GetDynamicFallMaterial))
              (Rendering|Material|SetMaterial _dynamicmesh 1 (Variables|TmpMaterial|GetDynamicSplashMaterial))
              (Rendering|Material|SetMaterial _dynamicmesh 2 0)
              (Material|InitialMaterialonMesh _dynamicmesh)
              (Default|CustomWaterBodySet))
            (:then_12
              (CallFunction|UserGuider)))
          (else
            (Development|PrintString "WaterfallWorkInGame" true true "(R=0.387858,G=1.000000,B=0.000000,A=1.000000)" 8.0)))))))

(fn UserGuider
  (bind _wdithpoint (Variables|Widget|Get◉WdithPoint))
  (bind _returnvalue (+ _wdithpoint "0.000000,0.000000,100.000000"))
  (bind _needfrozentip (Variables|Default|GetNeedFrozenTip))
  (bind _errortext (Variables|Default|GetErrorText))
  (bind _warningtext (Variables|Default|GetWarningText))
  (if (and (Utilities|IsValid (Variables|Initial|GetTargetStaticMesh) :!class /Script/Engine.KismetSystemLibrary) (Utilities|IsValid (Variables|Initial|GetTargetStaticMesh_CustomWaterBody) :!class /Script/Engine.KismetSystemLibrary))
    (Rendering|SetVisibility (Variables|Default|GetWaterfallFX) true false)
    (Utilities|FlowControl|Sequence
      (:then_0
        (Utilities|FlowControl|Sequence
          (:then_0
            (Rendering|SetVisibility (Variables|Default|GetErrorText) false false))
          (:then_1
            (if (Utilities|IsValid (Utilities|ResolveSoftReference (Variables|TopConnectionWaterBody|GetCustomWaterBodyActor_Soft)) :!class /Script/Engine.KismetSystemLibrary)
              (Variables|Default|SetUserGuide_TopConnectionWarning false)
              (else
                (Variables|Default|SetUserGuide_TopConnectionWarning true))))
          (:then_2
            (if (and (Utilities|IsValid (Utilities|ResolveSoftReference (Variables|Initial|GetTopLinkToWaterBody)) :!class /Script/Engine.KismetSystemLibrary) (Utilities|IsValid (Utilities|ResolveSoftReference (Variables|Initial|GetBottomLinkToLake)) :!class /Script/Engine.KismetSystemLibrary))
              (Variables|Default|SetUserGuide_LinkWaterBodyWarning false)
              (else
                (Variables|Default|SetUserGuide_LinkWaterBodyWarning true))))
          (:then_4
            (if (Variables|Setting|GetbDisableWarning)
              _
              (elif (or (Variables|Default|GetUserGuide_TopConnectionWarning) (Variables|Default|GetUserGuide_LinkWaterBodyWarning))
                (Rendering|Components|TextRender|SetText _warningtext (Utilities|Text|ToText(String) (Utilities|String|Append "To disable this , check DisableWarning in detail panel
Waterfall warning : " (Utilities|String|SelectString "
****This waterfall doesn't have a top connection water body****
Ignore if you don't need it.Won't break waterfall.
To fix this,click \"Initial Top Connection Water Body\" button." "" (Variables|Default|GetUserGuide_TopConnectionWarning)) (Utilities|String|SelectString "
****This waterfall missed Top/Bottom linking.****
Ignore if you don't need it.Won't break waterfall.
A correct linking can help you snap the water surface." "" (Variables|Default|GetUserGuide_LinkWaterBodyWarning)) "")))
                (Rendering|SetVisibility _warningtext true false)
                (Transformation|SetRelativeLocation _warningtext (+ (Variables|Widget|Get◉WdithPoint) "0.000000,0.000000,1000.000000") false false))))))
      (:then_1
        (if (Variables|Setting|GetbFrozen)
          (Rendering|SetVisibility _needfrozentip false false)
          (Transformation|SetRelativeLocation _needfrozentip _returnvalue false false)
          (else
            (Rendering|SetVisibility _needfrozentip true false)
            (Transformation|SetRelativeLocation _needfrozentip _returnvalue false false)))))
    (else
      (Rendering|SetVisibility (Variables|Default|GetDynamicMesh) false false)
      (Rendering|Components|TextRender|SetText _errortext (Utilities|Text|ToText(String) "Waterfall Error - Waterfall doesn't have valid static mesh target
Fix Steps - 
1.Input \"CreateStaticMeshName\" in blueprint detail panel.
2.Click \"Create New Static Mesh \" Button."))
      (Rendering|SetVisibility _errortext true false))))

(fn UserInterface_Limit3DWidget
  (bind _thickness (Variables|Widget|Get➤Thickness))
  (bind _x (.x _thickness))
  (bind _wavescale (Variables|Widget|Get➤WaveScale))
  (bind _x_2 (.x _wavescale))
  (bind _wdithpoint (Variables|Widget|Get◉WdithPoint))
  (bind _y (.y _wdithpoint))
  (bind _returnvalue (Math|Float|Min(Float) _x -1.0))
  (bind _width (Variables|TmpVarButBake|GetWidth))
  (bind _returnvalue_2 (/ _width 2.0))
  (bind _thickness_2 (Variables|TmpVarButBake|GetThickness))
  (bind _returnvalue_3 (/ _thickness_2 2.0))
  (bind _returnvalue_4 (+ _returnvalue_3 _returnvalue_2))
  (bind _returnvalue_5 (* _returnvalue_4 -1.0))
  (bind _returnvalue_6 (Math|Float|Min(Float) _x_2 _returnvalue_5))
  (bind _bottomlinktolake (Variables|Initial|GetBottomLinkToLake))
  (bind _output (Utilities|ResolveSoftReference _bottomlinktolake))
  (bind _toplinktowaterbody (Variables|Initial|GetTopLinkToWaterBody))
  (bind _output_2 (Utilities|ResolveSoftReference _toplinktowaterbody))
  ;[1 1 1 1 128] Constraint
  (Utilities|FlowControl|Sequence
    (:then_0
      (Utilities|IsValid _output_2
        (:Is%20Valid
          (Variables|Widget|Set◉Top (Math|Vector|MakeVector 0.0 0.0 (- (.z (Math|Transform|InverseTransformLocation (Transformation|GetActorTransform) (Spline|GetLocationatSplinePoint (Variables|/Script/Water.WaterBody|GetSplineComp _output_2) (- (Spline|GetNumberOfSplinePoints (Variables|/Script/Water.WaterBody|GetSplineComp _output_2)) 1) "World"))) (+ (/ (Math|Float|Absolute(Float) (Math|Float|Min(Float) (.x (Variables|Widget|Get➤Thickness)) -1.0)) 4.0) (* (Variables|Initial|GetTopPositionOffset) -1.0))))))
        (:Is%20Not%20Valid
          ;;[1 1 1] Top
          (Variables|Widget|Set◉Top (Math|Vector|MakeVector 0.0 0.0 (.z (Variables|Widget|Get◉Top)))))))
    (:then_1
      (Utilities|IsValid _output
        (:Is%20Valid
          ;;[1 1 1] Bottom
          (Variables|Widget|Set◉Bottom (Math|Vector|MakeVector (.x (Variables|Widget|Get◉Bottom)) 0.0 (.z (Math|Transform|InverseTransformLocation (Transformation|GetActorTransform) (Transformation|GetActorLocation _output))))))
        (:Is%20Not%20Valid
          (Variables|Widget|Set◉Bottom (Math|Vector|MakeVector (.x (Variables|Widget|Get◉Bottom)) 0.0 (.z (Variables|Widget|Get◉Bottom)))))))
    (:then_2
      ;;[1 1 1] Extend
      (Variables|Widget|Set◉Extend (Math|Vector|MakeVector (.x (Variables|Widget|Get◉Extend)) (.y (Variables|Widget|Get◉Top)) (.z (Variables|Widget|Get◉Top)))))
    (:then_3
      ;;[1 1 1] BevelH
      (Variables|Widget|Set◉Bevel_Horizontal (Math|Vector|MakeVector (Math|Float|Max(Float) (.x (Variables|Widget|Get◉Bevel_Horizontal))) 0.0 (.z (Variables|Widget|Get◉Top)))))
    (:then_4
      ;;[1 1 1] BevelV
      (Variables|Widget|Set➤Bevel_Vertical (Math|Vector|MakeVector 0.0 (.y (Variables|Widget|Get➤Bevel_Vertical)) (.z (Variables|Widget|Get◉Top)))))
    (:then_5
      ;;[1 1 1] Width
      (Variables|Widget|Set◉WdithPoint (Math|Vector|MakeVector (.x (Variables|Widget|Get◉Bottom)) (Math|Float|Max(Float) _y 1.0) (.z (Variables|Widget|Get◉Bottom))))
      (Variables|TmpVarButBake|SetWidth (* _y 2.0)))
    (:then_6
      ;;[1 1 1] Thickness
      (Variables|Widget|Set➤Thickness (Math|Vector|MakeVector _returnvalue 0.0 (* (+ (.z (Variables|Widget|Get◉Top)) (.z (Variables|Widget|Get◉Bottom))) 0.5)))
      (Variables|TmpVarButBake|SetThickness (/ (Math|Float|Absolute(Float) _returnvalue) 2.0)))
    (:then_9
      (Variables|Widget|Set➤WaveScale (Math|Vector|MakeVector _returnvalue_6 (.y (Variables|Widget|Get◉Bottom)) (.z (Variables|Widget|Get◉Bottom))))
      (Variables|TmpVarButBake|SetWaveScale (Math|Float|Absolute(Float) _returnvalue_6)))))
