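; Editor component for the waterfall generator tool that builds waterfall and custom water body meshes for a target generator actor.
; Events dispatch mesh generation, merge and bake steps, and static mesh assets are created by duplicating a template.
; Geometry Script copies merged dynamic meshes into static mesh LODs 0–2.
; Water bodies are remeshed spline polygons, with per-vertex flow and white-water values blended from velocity-spline actors by distance falloff.
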(event Custom|BakeWaterfallMeshDispatcher
  (Utilities|IsValid (Variables|Default|GetTargetActor)
    (:Is%20Valid
      (Class|BIWaterfallGenerator|BakeWaterfallMesh(Message) (Variables|Default|GetTargetActor)))))

(event Custom|DelayExecuteDispatcher
  (bind _targetactor (Variables|Default|GetTargetActor))
  (Utilities|IsValid _targetactor
    (:Is%20Valid
      (Class|BIWaterfallGenerator|DelayExecute(Message) _targetactor false))))

(event Custom|DelayExecuteDispatcherWithForceUpdate
  (bind _targetactor (Variables|Default|GetTargetActor))
  (Utilities|IsValid _targetactor
    (:Is%20Valid
      (Class|BIWaterfallGenerator|DelayExecute(Message) _targetactor true))))

(event EventGenerationBI_MeshGen
  (Default|GenerationMeshGen))

(event EventGenerationBI_MeshMerge
  (Default|GenerationMeshMerge))

(event EventGenerationBI_BakeMesh
  (Default|GenerationBakeMesh))

(event EventGenerationBI_GenerateCustomWaterBody
  (params (ForceUpdate :type Boolean))
  (Default|GenerationGenerateCustomWaterBody ForceUpdate))

(event EventMessageBox
  (params (Title :type Text) (Message :type Text))
  (CallFunction|MessageBox-OK Title Message))

(fn CreateNewStaticMesh
  (params (Name :type String) (TemplatePath :type String))
  (bind _newparam (Default|CreateNewStaticMeshIntern Name TemplatePath))
  (return _newparam))

(fn CreateNewStaticMeshIntern
  (params (StaticMeshName :type String) (TemplatePath :type String :default "/Game/NeverCook/TechArt/WaterfallGenerator/Template"))
  (bind _returnvalue (Utilities|String|Append "/Game/Mesh/WaterfallMesh/" StaticMeshName))
  (bind _returnvalue_2 (EditorScripting|Asset|DoesAssetExist _returnvalue :!class /Script/EditorScriptingUtilities.EditorAssetLibrary))
  (if _returnvalue_2
    (EditorScripting|MessageDialog|ShowMessageDialog (Utilities|Text|MakeLiteralText "NSLOCTEXT(\"[FEB440A741E129F77B907E6A9577F1F9]\", \"8FE28ED24A774F87E0D65B81C8E3F49E\", \"Warning\")") (Utilities|Text|MakeLiteralText "NSLOCTEXT(\"[FEB440A741E129F77B907E6A9577F1F9]\", \"77348E114179CB92E09F8F96472ADED9\", \"Target static mesh already exist.Please change name.\")") "Ok" "Ok")
    (else
      (bind _returnvalue_3 (EditorScripting|Asset|DuplicateAsset TemplatePath _returnvalue :!class /Script/EditorScriptingUtilities.EditorAssetLibrary))
      (Utilities|IsValid _returnvalue_3
        (:Is%20Valid
          (Development|PrintString 0 "WaterfallGenerationTool:CreateSMAssetSuccess")
          (return (Utilities|Casting|CastToStaticMesh _returnvalue_3)))
        (:Is%20Not%20Valid
          _))))
  (Development|PrintString 0 "WaterfallGenerationTool:CreateSMAssetFail")
  (return 0))

(fn Generation_BakeMesh
  (bind _geometryscriptcopymeshtoassetoptions (Utilities|Struct|MakeGeometryScriptCopyMeshtoAssetOptions :bEnableRecomputeNormals false :bEnableRecomputeTangents false :bEnableRemoveDegenerates false :BoneHierarchyMismatchHandling "DoNothing" :bRemapBoneIndicesToMatchAsset false :bUseOriginalVertexOrder false :bUseBuildScale true :bReplaceMaterials false :NewMaterialUsage "ReplaceAssetMaterials" :bCleanAssignedMaterials false :GenerateLightmapUVs "MatchTargetLODSetting" :bApplyNaniteSettings false :NaniteSettings "(bEnabled=True,FallbackPercentTriangles=100.000000,FallbackRelativeError=0.000000)" :NewNaniteSettings "(bEnabled=False,bExplicitTangents=False,bLerpUVs=True,bSeparable=False,bVoxelNDF=True,bVoxelOpacity=False,ShapePreservation=None,PositionPrecision=-2147483648,NormalPrecision=-1,TangentPrecision=-1,BoneWeightPrecision=-1,TargetMinimumResidencyInKB=0,KeepPercentTriangles=1.000000,TrimRelativeError=0.000000,GenerateFallback=PlatformDefault,FallbackTarget=Auto,FallbackPercentTriangles=1.000000,FallbackRelativeError=1.000000,MaxEdgeLengthFactor=0.000000,NumRays=1,VoxelLevel=0,RayBackUp=0.000000,DisplacementUVChannel=0,DisplacementMaps=,NaniteAssemblyData=(Parts=,Nodes=))" :bEmitTransaction false :bDeferMeshPostEditChange false))
  (bind _targetactor (Variables|Default|GetTargetActor))
  (bind _targetstaticmesh (Variables|/Game/Tools/NewWaterfallGenerator/Blueprint/BP_WaterfallGenerator.BP_WaterfallGenerator_C|GetTargetStaticMesh _targetactor))
  (bind _targetactor_2 (Variables|Default|GetTargetActor))
  (bind _targetstaticmesh_2 (Variables|/Game/Tools/NewWaterfallGenerator/Blueprint/BP_WaterfallGenerator.BP_WaterfallGenerator_C|GetTargetStaticMesh _targetactor_2))
  (bind _targetactor_3 (Variables|Default|GetTargetActor))
  (bind _targetstaticmesh_3 (Variables|/Game/Tools/NewWaterfallGenerator/Blueprint/BP_WaterfallGenerator.BP_WaterfallGenerator_C|GetTargetStaticMesh _targetactor_3))
  ;[1 1 1] BakeMesh
  (Utilities|FlowControl|Sequence
    (:then_0
      ;;[1 1 1] Bake to static mesh
      (Utilities|IsValid _targetstaticmesh
        (:Is%20Valid
          (GeometryScript|StaticMesh|CopyMeshtoStaticMesh (Variables|Default|GetDynamicMesh_Merged) _targetstaticmesh (Utilities|Struct|MakeGeometryScriptCopyMeshtoAssetOptions :bEnableRecomputeNormals false :bEnableRecomputeTangents false :bEnableRemoveDegenerates false :BoneHierarchyMismatchHandling "DoNothing" :bRemapBoneIndicesToMatchAsset false :bUseOriginalVertexOrder false :bUseBuildScale true :bReplaceMaterials true :NewMaterialUsage "ReplaceAssetMaterials" :NewMaterials (Utilities|Array|MakeArray "/Engine/EngineMaterials/WorldGridMaterial.WorldGridMaterial" "/Engine/EngineMaterials/WorldGridMaterial.WorldGridMaterial") :bCleanAssignedMaterials false :GenerateLightmapUVs "MatchTargetLODSetting" :bApplyNaniteSettings false :NaniteSettings "(bEnabled=True,FallbackPercentTriangles=100.000000,FallbackRelativeError=0.000000)" :NewNaniteSettings "(bEnabled=False,bExplicitTangents=False,bLerpUVs=True,bSeparable=False,bVoxelNDF=True,bVoxelOpacity=False,ShapePreservation=None,PositionPrecision=-2147483648,NormalPrecision=-1,TangentPrecision=-1,BoneWeightPrecision=-1,TargetMinimumResidencyInKB=0,KeepPercentTriangles=1.000000,TrimRelativeError=0.000000,GenerateFallback=PlatformDefault,FallbackTarget=Auto,FallbackPercentTriangles=1.000000,FallbackRelativeError=1.000000,MaxEdgeLengthFactor=0.000000,NumRays=1,VoxelLevel=0,RayBackUp=0.000000,DisplacementUVChannel=0,DisplacementMaps=,NaniteAssemblyData=(Parts=,Nodes=))" :bEmitTransaction false :bDeferMeshPostEditChange false) (Utilities|Struct|MakeGeometryScriptMeshWriteLOD false)))))
    (:then_1
      ;;[1 1 1] Bake to static mesh
      (Utilities|IsValid _targetstaticmesh_3
        (:Is%20Valid
          (GeometryScript|StaticMesh|CopyMeshtoStaticMesh (Variables|Default|GetDynamicMesh_Merged_LOD1) _targetstaticmesh_3 _geometryscriptcopymeshtoassetoptions (Utilities|Struct|MakeGeometryScriptMeshWriteLOD false 1)))))
    (:then_2
      ;;[1 1 1] Bake to static mesh
      (Utilities|IsValid _targetstaticmesh_2
        (:Is%20Valid
          ;;;[1 1 1] Comment
          (GeometryScript|StaticMesh|CopyMeshtoStaticMesh (Variables|Default|GetDynamicMesh_Merged_LOD2) _targetstaticmesh_2 _geometryscriptcopymeshtoassetoptions (Utilities|Struct|MakeGeometryScriptMeshWriteLOD false 2)))))))

(fn Generation_GenerateCustomWaterBody
  (params (ForceUpdate :type Boolean))
  (bind (_x _y _z _w) (Math|Vector4|BreakVector4 (LocalVariables|Gettmpsinglevtxvl)))
  (bind _targetactor (Variables|Default|GetTargetActor))
  (bind _thickness (Variables|/Game/Tools/NewWaterfallGenerator/Blueprint/BP_WaterfallGenerator.BP_WaterfallGenerator_C|GetThickness _targetactor))
  (bind _returnvalue (* _thickness 0.5))
  (bind _returnvalue_2 (* _thickness -0.5))
  (bind _width (Variables|/Game/Tools/NewWaterfallGenerator/Blueprint/BP_WaterfallGenerator.BP_WaterfallGenerator_C|GetWidth _targetactor))
  (bind _returnvalue_3 (/ _width -2.0))
  (bind _returnvalue_4 (+ _returnvalue_3 _returnvalue_2))
  (bind _targetactor_2 (Variables|Default|GetTargetActor))
  (bind _startpointedgeoffset (Variables|/Game/Tools/NewWaterfallGenerator/Blueprint/BP_WaterfallGenerator.BP_WaterfallGenerator_C|GetStartPointEdgeOffset _targetactor_2))
  (bind _returnvalue_5 (Math|Vector|MakeVector _startpointedgeoffset _returnvalue_4 _returnvalue))
  (bind _returnvalue_6 (/ _width 2.0))
  (bind _returnvalue_7 (+ _returnvalue_6 _returnvalue))
  (bind _returnvalue_8 (Math|Vector|MakeVector _startpointedgeoffset _returnvalue_7 _returnvalue))
  (bind _targetactor_3 (Variables|Default|GetTargetActor))
  (bind _returnvalue_9 (Transformation|GetActorTransform _targetactor_3))
  (bind _bevel_horizontal (Variables|/Game/Tools/NewWaterfallGenerator/Blueprint/BP_WaterfallGenerator.BP_WaterfallGenerator_C|Get◉Bevel_Horizontal _targetactor_3))
  (bind _returnvalue_10 (+ _bevel_horizontal _returnvalue_5))
  (bind _returnvalue_11 (Math|Transform|TransformLocation _returnvalue_9 _returnvalue_10))
  (bind _targetactor_4 (Variables|Default|GetTargetActor))
  (bind _thickness_2 (Variables|/Game/Tools/NewWaterfallGenerator/Blueprint/BP_WaterfallGenerator.BP_WaterfallGenerator_C|GetThickness _targetactor_4))
  (bind _returnvalue_12 (* _thickness_2 0.5))
  (bind _targetactor_5 (Variables|Default|GetTargetActor))
  (bind _bevel_horizontal_2 (Variables|/Game/Tools/NewWaterfallGenerator/Blueprint/BP_WaterfallGenerator.BP_WaterfallGenerator_C|Get◉Bevel_Horizontal _targetactor_5))
  (bind _z_2 (.z _bevel_horizontal_2))
  (bind _returnvalue_13 (+ _z_2 _returnvalue_12))
  (bind _x_2 (.x _bevel_horizontal_2))
  (bind _returnvalue_14 (Math|Vector|MakeVector _x_2 0.0 _returnvalue_13))
  (bind _returnvalue_15 (Math|Transform|MakeTransform _returnvalue_14))
  (bind _returnvalue_16 (Math|Transform|ComposeTransforms _returnvalue_15 _returnvalue_9))
  (bind _returnvalue_17 (+ _bevel_horizontal _returnvalue_8))
  (bind _returnvalue_18 (Math|Transform|TransformLocation _returnvalue_9 _returnvalue_17))
  (bind _returnvalue_19 (- _returnvalue_11 _returnvalue_18))
  (bind _returnvalue_20 (Math|Vector|DotProduct _returnvalue_19 _returnvalue_19))
  (bind (_x_3 _y_2 _z_3 _w_2) (Math|Vector4|BreakVector4 (LocalVariables|Gettmpsinglevtxvl)))
  (bind (_x_4 _y_3 _z_4 _w_3) (Math|Vector4|BreakVector4 (LocalVariables|Gettmpsinglevtxvl)))
  (bind (_x_5 _y_4 _z_5) (Math|Vector4|BreakVector4 (LocalVariables|Gettmpsinglevtxvl)))
  (bind _geometryscriptcopymeshtoassetoptions (Utilities|Struct|MakeGeometryScriptCopyMeshtoAssetOptions :bEnableRecomputeNormals false :bEnableRecomputeTangents false :bEnableRemoveDegenerates false :BoneHierarchyMismatchHandling "DoNothing" :bRemapBoneIndicesToMatchAsset false :bUseOriginalVertexOrder false :bUseBuildScale true :bReplaceMaterials false :NewMaterialUsage "ReplaceAssetMaterials" :bCleanAssignedMaterials false :GenerateLightmapUVs "MatchTargetLODSetting" :bApplyNaniteSettings false :NaniteSettings "(bEnabled=True,FallbackPercentTriangles=100.000000,FallbackRelativeError=0.000000)" :NewNaniteSettings "(bEnabled=False,bExplicitTangents=False,bLerpUVs=True,bSeparable=False,bVoxelNDF=True,bVoxelOpacity=False,ShapePreservation=None,PositionPrecision=-2147483648,NormalPrecision=-1,TangentPrecision=-1,BoneWeightPrecision=-1,TargetMinimumResidencyInKB=0,KeepPercentTriangles=1.000000,TrimRelativeError=0.000000,GenerateFallback=PlatformDefault,FallbackTarget=Auto,FallbackPercentTriangles=1.000000,FallbackRelativeError=1.000000,MaxEdgeLengthFactor=0.000000,NumRays=1,VoxelLevel=0,RayBackUp=0.000000,DisplacementUVChannel=0,DisplacementMaps=,NaniteAssemblyData=(Parts=,Nodes=))" :bEmitTransaction false :bDeferMeshPostEditChange false))
  (bind _lod2vtxcl (LocalVariables|GetLOD2Vtxcl))
  (bind _vtxcllst (LocalVariables|Getvtxcllst))
  (bind _dynamicmesh_customwaterbody_collision (Variables|Default|GetDynamicMesh_CustomWaterBody_Collision))
  (bind _targetmesh (LocalVariables|GetTargetMesh))
  (bind _targetactor_6 (Variables|Default|GetTargetActor))
  (bind _watermeshsamplecount (Variables|/Game/Tools/NewWaterfallGenerator/Blueprint/BP_WaterfallGenerator.BP_WaterfallGenerator_C|GetWaterMeshSampleCount _targetactor_6))
  (bind _targetactor_7 (Variables|Default|GetTargetActor))
  (bind _customwaterbodydepth (Variables|/Game/Tools/NewWaterfallGenerator/Blueprint/BP_WaterfallGenerator.BP_WaterfallGenerator_C|GetCustomWaterBodyDepth _targetactor_7))
  (bind _targetactor_8 (Variables|Default|GetTargetActor))
  (bind _targetactor_9 (Variables|Default|GetTargetActor))
  (bind _whitewaterradius (Variables|/Game/Tools/NewWaterfallGenerator/Blueprint/BP_WaterfallGenerator.BP_WaterfallGenerator_C|GetWhiteWaterRadius _targetactor_9))
  (bind _targetstaticmesh_customwaterbody (Variables|/Game/Tools/NewWaterfallGenerator/Blueprint/BP_WaterfallGenerator.BP_WaterfallGenerator_C|GetTargetStaticMesh_CustomWaterBody _targetactor_8))
  (if ForceUpdate
    _
    (elif (Variables|/Game/Tools/NewWaterfallGenerator/Blueprint/BP_WaterfallGenerator.BP_WaterfallGenerator_C|GetbAutoUpdateStaticMesh (Variables|Default|GetTargetActor))
      _))
  (Utilities|FlowControl|Sequence
    (:then_0
      (Utilities|FlowControl|Sequence
        (:then_0
          (bind _returnvalue_21 (Game|ConstructObjectfromClass "/Script/GeometryFramework.DynamicMesh"))
          (Variables|Default|SetDynamicMesh_CustomWaterBody_Collision _returnvalue_21))
        (:then_1
          (bind _returnvalue_22 (Game|ConstructObjectfromClass "/Script/GeometryFramework.DynamicMesh"))
          (Variables|Default|SetDynamicMesh_CustomWaterBody _returnvalue_22))
        (:then_2
          (bind _returnvalue_23 (Game|ConstructObjectfromClass "/Script/GeometryFramework.DynamicMesh"))
          (Variables|Default|SetDynamicMesh_CustomWaterBody_LOD1 _returnvalue_23))
        (:then_3
          (bind _returnvalue_24 (Game|ConstructObjectfromClass "/Script/GeometryFramework.DynamicMesh"))
          (Variables|Default|SetDynamicMesh_CustomWaterBody_LOD2 _returnvalue_24))))
    (:then_1
      (for _index (range (+ _watermeshsamplecount 1))
        (Utilities|Array|Add (LocalVariables|Getpolst) (Math|Vector2D|MakeVector2D (.x (Spline|GetLocationatTime (Variables|/Game/Tools/NewWaterfallGenerator/Blueprint/BP_WaterfallGenerator.BP_WaterfallGenerator_C|GetSpline_CustomWaterBody (Variables|Default|GetTargetActor)) (/ (Math|Conversions|ToFloat(Integer) _index) (Math|Conversions|ToFloat(Integer) _watermeshsamplecount)) "Local" true)) (.y (Spline|GetLocationatTime (Variables|/Game/Tools/NewWaterfallGenerator/Blueprint/BP_WaterfallGenerator.BP_WaterfallGenerator_C|GetSpline_CustomWaterBody (Variables|Default|GetTargetActor)) (/ (Math|Conversions|ToFloat(Integer) _index) (Math|Conversions|ToFloat(Integer) _watermeshsamplecount)) "Local" true))))))
    (:then_2
      (Utilities|FlowControl|Sequence
        (:then_0
          (bind _returnvalue_25 (GeometryScript|Primitives|AppendTriangulatedPolygon :TargetMesh (Variables|Default|GetDynamicMesh_CustomWaterBody) :Transform (Math|Transform|MakeTransform "0, 0, 0") :PolygonVertices (LocalVariables|Getpolst) :bAllowSelfIntersections true))
          (bind _returnvalue_26 (GeometryScript|Simplification|ApplyUniformRemesh :TargetMesh _returnvalue_25 :UniformOptions (Utilities|Struct|MakeGeometryScriptUniformRemeshOptions "TargetEdgeLength" 5000 (Variables|/Game/Tools/NewWaterfallGenerator/Blueprint/BP_WaterfallGenerator.BP_WaterfallGenerator_C|GetWaterMeshPolyEdgeLength (Variables|Default|GetTargetActor)))))
          (bind _returnvalue_27 (GeometryScript|VertexColor|SetMeshConstantVertexColor :TargetMesh _returnvalue_26 :Flags (Utilities|Struct|MakeGeometryScriptColorFlags true) :bClearExisting true))
          (bind _positionlist (GeometryScript|MeshQueries|GetAllVertexPositions _returnvalue_27))
          (bind _outactors (Actor|GetAllActorsOfClass 0 "/Game/Tools/NewWaterfallGenerator/Blueprint/BP_WaterfallCustomWaterBodyVelocity.BP_WaterfallCustomWaterBodyVelocity_C"))
          (bind _vectorarray (GeometryScript|ListUtils|ConvertVectorListtoArray _positionlist))
          (for _array_element _vectorarray
            (LocalVariables|Settmpsinglevtxvl)
            (Utilities|FlowControl|Sequence
              (:then_0
                (for _array_element_2 _outactors
                  (if (>= (Math|Vector|VectorLength (Math|Vector|MakeVector _x_5 _y_4 _z_5)) 0.01)
                    (LocalVariables|Settmpsinglevtxvl (Math|Vector4|MakeVector4 (.x (+ (* (Math|Vector|MakeVector _x_3 _y_2 _z_3) (/ (Math|Vector|VectorLength (Math|Vector|MakeVector _x_3 _y_2 _z_3)) (+ (Math|Vector|VectorLength (Math|Vector|MakeVector _x_3 _y_2 _z_3)) (Math|Vector|VectorLength (Utilities|Operators|Multiply (Math|Float|Clamp(Float) (Math|Float|Power (/ (- (* 1.0 (Variables|/Game/Tools/NewWaterfallGenerator/Blueprint/BP_WaterfallCustomWaterBodyVelocity.BP_WaterfallCustomWaterBodyVelocity_C|GetRadius _array_element_2)) (Math|Vector|Distance(Vector) (Spline|FindLocationClosesttoWorldLocation (Variables|/Game/Tools/NewWaterfallGenerator/Blueprint/BP_WaterfallCustomWaterBodyVelocity.BP_WaterfallCustomWaterBodyVelocity_C|GetSpline _array_element_2) (Math|Transform|TransformLocation _returnvalue_16 _array_element) "World") (Math|Transform|TransformLocation _returnvalue_16 _array_element))) (Variables|/Game/Tools/NewWaterfallGenerator/Blueprint/BP_WaterfallCustomWaterBodyVelocity.BP_WaterfallCustomWaterBodyVelocity_C|GetEdgeSoft _array_element_2)) (Variables|/Game/Tools/NewWaterfallGenerator/Blueprint/BP_WaterfallCustomWaterBodyVelocity.BP_WaterfallCustomWaterBodyVelocity_C|GetEdgeGradientPower _array_element_2))) (Math|Vector|Normalize (Math|Vector|ProjectVectorontoPlane (Spline|FindTangentClosesttoWorldLocation (Variables|/Game/Tools/NewWaterfallGenerator/Blueprint/BP_WaterfallCustomWaterBodyVelocity.BP_WaterfallCustomWaterBodyVelocity_C|GetSpline _array_element_2) (Math|Transform|TransformLocation _returnvalue_16 _array_element) "World") "0.000000,0.000000,1.000000")) (Variables|/Game/Tools/NewWaterfallGenerator/Blueprint/BP_WaterfallCustomWaterBodyVelocity.BP_WaterfallCustomWaterBodyVelocity_C|GetVelocityStrength _array_element_2)))))) (* (Utilities|Operators|Multiply (Math|Float|Clamp(Float) (Math|Float|Power (/ (- (* 1.0 (Variables|/Game/Tools/NewWaterfallGenerator/Blueprint/BP_WaterfallCustomWaterBodyVelocity.BP_WaterfallCustomWaterBodyVelocity_C|GetRadius _array_element_2)) (Math|Vector|Distance(Vector) (Spline|FindLocationClosesttoWorldLocation (Variables|/Game/Tools/NewWaterfallGenerator/Blueprint/BP_WaterfallCustomWaterBodyVelocity.BP_WaterfallCustomWaterBodyVelocity_C|GetSpline _array_element_2) (Math|Transform|TransformLocation _returnvalue_16 _array_element) "World") (Math|Transform|TransformLocation _returnvalue_16 _array_element))) (Variables|/Game/Tools/NewWaterfallGenerator/Blueprint/BP_WaterfallCustomWaterBodyVelocity.BP_WaterfallCustomWaterBodyVelocity_C|GetEdgeSoft _array_element_2)) (Variables|/Game/Tools/NewWaterfallGenerator/Blueprint/BP_WaterfallCustomWaterBodyVelocity.BP_WaterfallCustomWaterBodyVelocity_C|GetEdgeGradientPower _array_element_2))) (Math|Vector|Normalize (Math|Vector|ProjectVectorontoPlane (Spline|FindTangentClosesttoWorldLocation (Variables|/Game/Tools/NewWaterfallGenerator/Blueprint/BP_WaterfallCustomWaterBodyVelocity.BP_WaterfallCustomWaterBodyVelocity_C|GetSpline _array_element_2) (Math|Transform|TransformLocation _returnvalue_16 _array_element) "World") "0.000000,0.000000,1.000000")) (Variables|/Game/Tools/NewWaterfallGenerator/Blueprint/BP_WaterfallCustomWaterBodyVelocity.BP_WaterfallCustomWaterBodyVelocity_C|GetVelocityStrength _array_element_2)) (/ (Math|Vector|VectorLength (Utilities|Operators|Multiply (Math|Float|Clamp(Float) (Math|Float|Power (/ (- (* 1.0 (Variables|/Game/Tools/NewWaterfallGenerator/Blueprint/BP_WaterfallCustomWaterBodyVelocity.BP_WaterfallCustomWaterBodyVelocity_C|GetRadius _array_element_2)) (Math|Vector|Distance(Vector) (Spline|FindLocationClosesttoWorldLocation (Variables|/Game/Tools/NewWaterfallGenerator/Blueprint/BP_WaterfallCustomWaterBodyVelocity.BP_WaterfallCustomWaterBodyVelocity_C|GetSpline _array_element_2) (Math|Transform|TransformLocation _returnvalue_16 _array_element) "World") (Math|Transform|TransformLocation _returnvalue_16 _array_element))) (Variables|/Game/Tools/NewWaterfallGenerator/Blueprint/BP_WaterfallCustomWaterBodyVelocity.BP_WaterfallCustomWaterBodyVelocity_C|GetEdgeSoft _array_element_2)) (Variables|/Game/Tools/NewWaterfallGenerator/Blueprint/BP_WaterfallCustomWaterBodyVelocity.BP_WaterfallCustomWaterBodyVelocity_C|GetEdgeGradientPower _array_element_2))) (Math|Vector|Normalize (Math|Vector|ProjectVectorontoPlane (Spline|FindTangentClosesttoWorldLocation (Variables|/Game/Tools/NewWaterfallGenerator/Blueprint/BP_WaterfallCustomWaterBodyVelocity.BP_WaterfallCustomWaterBodyVelocity_C|GetSpline _array_element_2) (Math|Transform|TransformLocation _returnvalue_16 _array_element) "World") "0.000000,0.000000,1.000000")) (Variables|/Game/Tools/NewWaterfallGenerator/Blueprint/BP_WaterfallCustomWaterBodyVelocity.BP_WaterfallCustomWaterBodyVelocity_C|GetVelocityStrength _array_element_2))) (+ (Math|Vector|VectorLength (Math|Vector|MakeVector _x_3 _y_2 _z_3)) (Math|Vector|VectorLength (Utilities|Operators|Multiply (Math|Float|Clamp(Float) (Math|Float|Power (/ (- (* 1.0 (Variables|/Game/Tools/NewWaterfallGenerator/Blueprint/BP_WaterfallCustomWaterBodyVelocity.BP_WaterfallCustomWaterBodyVelocity_C|GetRadius _array_element_2)) (Math|Vector|Distance(Vector) (Spline|FindLocationClosesttoWorldLocation (Variables|/Game/Tools/NewWaterfallGenerator/Blueprint/BP_WaterfallCustomWaterBodyVelocity.BP_WaterfallCustomWaterBodyVelocity_C|GetSpline _array_element_2) (Math|Transform|TransformLocation _returnvalue_16 _array_element) "World") (Math|Transform|TransformLocation _returnvalue_16 _array_element))) (Variables|/Game/Tools/NewWaterfallGenerator/Blueprint/BP_WaterfallCustomWaterBodyVelocity.BP_WaterfallCustomWaterBodyVelocity_C|GetEdgeSoft _array_element_2)) (Variables|/Game/Tools/NewWaterfallGenerator/Blueprint/BP_WaterfallCustomWaterBodyVelocity.BP_WaterfallCustomWaterBodyVelocity_C|GetEdgeGradientPower _array_element_2))) (Math|Vector|Normalize (Math|Vector|ProjectVectorontoPlane (Spline|FindTangentClosesttoWorldLocation (Variables|/Game/Tools/NewWaterfallGenerator/Blueprint/BP_WaterfallCustomWaterBodyVelocity.BP_WaterfallCustomWaterBodyVelocity_C|GetSpline _array_element_2) (Math|Transform|TransformLocation _returnvalue_16 _array_element) "World") "0.000000,0.000000,1.000000")) (Variables|/Game/Tools/NewWaterfallGenerator/Blueprint/BP_WaterfallCustomWaterBodyVelocity.BP_WaterfallCustomWaterBodyVelocity_C|GetVelocityStrength _array_element_2)))))))) (.y (+ (* (Math|Vector|MakeVector _x_3 _y_2 _z_3) (/ (Math|Vector|VectorLength (Math|Vector|MakeVector _x_3 _y_2 _z_3)) (+ (Math|Vector|VectorLength (Math|Vector|MakeVector _x_3 _y_2 _z_3)) (Math|Vector|VectorLength (Utilities|Operators|Multiply (Math|Float|Clamp(Float) (Math|Float|Power (/ (- (* 1.0 (Variables|/Game/Tools/NewWaterfallGenerator/Blueprint/BP_WaterfallCustomWaterBodyVelocity.BP_WaterfallCustomWaterBodyVelocity_C|GetRadius _array_element_2)) (Math|Vector|Distance(Vector) (Spline|FindLocationClosesttoWorldLocation (Variables|/Game/Tools/NewWaterfallGenerator/Blueprint/BP_WaterfallCustomWaterBodyVelocity.BP_WaterfallCustomWaterBodyVelocity_C|GetSpline _array_element_2) (Math|Transform|TransformLocation _returnvalue_16 _array_element) "World") (Math|Transform|TransformLocation _returnvalue_16 _array_element))) (Variables|/Game/Tools/NewWaterfallGenerator/Blueprint/BP_WaterfallCustomWaterBodyVelocity.BP_WaterfallCustomWaterBodyVelocity_C|GetEdgeSoft _array_element_2)) (Variables|/Game/Tools/NewWaterfallGenerator/Blueprint/BP_WaterfallCustomWaterBodyVelocity.BP_WaterfallCustomWaterBodyVelocity_C|GetEdgeGradientPower _array_element_2))) (Math|Vector|Normalize (Math|Vector|ProjectVectorontoPlane (Spline|FindTangentClosesttoWorldLocation (Variables|/Game/Tools/NewWaterfallGenerator/Blueprint/BP_WaterfallCustomWaterBodyVelocity.BP_WaterfallCustomWaterBodyVelocity_C|GetSpline _array_element_2) (Math|Transform|TransformLocation _returnvalue_16 _array_element) "World") "0.000000,0.000000,1.000000")) (Variables|/Game/Tools/NewWaterfallGenerator/Blueprint/BP_WaterfallCustomWaterBodyVelocity.BP_WaterfallCustomWaterBodyVelocity_C|GetVelocityStrength _array_element_2)))))) (* (Utilities|Operators|Multiply (Math|Float|Clamp(Float) (Math|Float|Power (/ (- (* 1.0 (Variables|/Game/Tools/NewWaterfallGenerator/Blueprint/BP_WaterfallCustomWaterBodyVelocity.BP_WaterfallCustomWaterBodyVelocity_C|GetRadius _array_element_2)) (Math|Vector|Distance(Vector) (Spline|FindLocationClosesttoWorldLocation (Variables|/Game/Tools/NewWaterfallGenerator/Blueprint/BP_WaterfallCustomWaterBodyVelocity.BP_WaterfallCustomWaterBodyVelocity_C|GetSpline _array_element_2) (Math|Transform|TransformLocation _returnvalue_16 _array_element) "World") (Math|Transform|TransformLocation _returnvalue_16 _array_element))) (Variables|/Game/Tools/NewWaterfallGenerator/Blueprint/BP_WaterfallCustomWaterBodyVelocity.BP_WaterfallCustomWaterBodyVelocity_C|GetEdgeSoft _array_element_2)) (Variables|/Game/Tools/NewWaterfallGenerator/Blueprint/BP_WaterfallCustomWaterBodyVelocity.BP_WaterfallCustomWaterBodyVelocity_C|GetEdgeGradientPower _array_element_2))) (Math|Vector|Normalize (Math|Vector|ProjectVectorontoPlane (Spline|FindTangentClosesttoWorldLocation (Variables|/Game/Tools/NewWaterfallGenerator/Blueprint/BP_WaterfallCustomWaterBodyVelocity.BP_WaterfallCustomWaterBodyVelocity_C|GetSpline _array_element_2) (Math|Transform|TransformLocation _returnvalue_16 _array_element) "World") "0.000000,0.000000,1.000000")) (Variables|/Game/Tools/NewWaterfallGenerator/Blueprint/BP_WaterfallCustomWaterBodyVelocity.BP_WaterfallCustomWaterBodyVelocity_C|GetVelocityStrength _array_element_2)) (/ (Math|Vector|VectorLength (Utilities|Operators|Multiply (Math|Float|Clamp(Float) (Math|Float|Power (/ (- (* 1.0 (Variables|/Game/Tools/NewWaterfallGenerator/Blueprint/BP_WaterfallCustomWaterBodyVelocity.BP_WaterfallCustomWaterBodyVelocity_C|GetRadius _array_element_2)) (Math|Vector|Distance(Vector) (Spline|FindLocationClosesttoWorldLocation (Variables|/Game/Tools/NewWaterfallGenerator/Blueprint/BP_WaterfallCustomWaterBodyVelocity.BP_WaterfallCustomWaterBodyVelocity_C|GetSpline _array_element_2) (Math|Transform|TransformLocation _returnvalue_16 _array_element) "World") (Math|Transform|TransformLocation _returnvalue_16 _array_element))) (Variables|/Game/Tools/NewWaterfallGenerator/Blueprint/BP_WaterfallCustomWaterBodyVelocity.BP_WaterfallCustomWaterBodyVelocity_C|GetEdgeSoft _array_element_2)) (Variables|/Game/Tools/NewWaterfallGenerator/Blueprint/BP_WaterfallCustomWaterBodyVelocity.BP_WaterfallCustomWaterBodyVelocity_C|GetEdgeGradientPower _array_element_2))) (Math|Vector|Normalize (Math|Vector|ProjectVectorontoPlane (Spline|FindTangentClosesttoWorldLocation (Variables|/Game/Tools/NewWaterfallGenerator/Blueprint/BP_WaterfallCustomWaterBodyVelocity.BP_WaterfallCustomWaterBodyVelocity_C|GetSpline _array_element_2) (Math|Transform|TransformLocation _returnvalue_16 _array_element) "World") "0.000000,0.000000,1.000000")) (Variables|/Game/Tools/NewWaterfallGenerator/Blueprint/BP_WaterfallCustomWaterBodyVelocity.BP_WaterfallCustomWaterBodyVelocity_C|GetVelocityStrength _array_element_2))) (+ (Math|Vector|VectorLength (Math|Vector|MakeVector _x_3 _y_2 _z_3)) (Math|Vector|VectorLength (Utilities|Operators|Multiply (Math|Float|Clamp(Float) (Math|Float|Power (/ (- (* 1.0 (Variables|/Game/Tools/NewWaterfallGenerator/Blueprint/BP_WaterfallCustomWaterBodyVelocity.BP_WaterfallCustomWaterBodyVelocity_C|GetRadius _array_element_2)) (Math|Vector|Distance(Vector) (Spline|FindLocationClosesttoWorldLocation (Variables|/Game/Tools/NewWaterfallGenerator/Blueprint/BP_WaterfallCustomWaterBodyVelocity.BP_WaterfallCustomWaterBodyVelocity_C|GetSpline _array_element_2) (Math|Transform|TransformLocation _returnvalue_16 _array_element) "World") (Math|Transform|TransformLocation _returnvalue_16 _array_element))) (Variables|/Game/Tools/NewWaterfallGenerator/Blueprint/BP_WaterfallCustomWaterBodyVelocity.BP_WaterfallCustomWaterBodyVelocity_C|GetEdgeSoft _array_element_2)) (Variables|/Game/Tools/NewWaterfallGenerator/Blueprint/BP_WaterfallCustomWaterBodyVelocity.BP_WaterfallCustomWaterBodyVelocity_C|GetEdgeGradientPower _array_element_2))) (Math|Vector|Normalize (Math|Vector|ProjectVectorontoPlane (Spline|FindTangentClosesttoWorldLocation (Variables|/Game/Tools/NewWaterfallGenerator/Blueprint/BP_WaterfallCustomWaterBodyVelocity.BP_WaterfallCustomWaterBodyVelocity_C|GetSpline _array_element_2) (Math|Transform|TransformLocation _returnvalue_16 _array_element) "World") "0.000000,0.000000,1.000000")) (Variables|/Game/Tools/NewWaterfallGenerator/Blueprint/BP_WaterfallCustomWaterBodyVelocity.BP_WaterfallCustomWaterBodyVelocity_C|GetVelocityStrength _array_element_2)))))))) (.z (+ (* (Math|Vector|MakeVector _x_3 _y_2 _z_3) (/ (Math|Vector|VectorLength (Math|Vector|MakeVector _x_3 _y_2 _z_3)) (+ (Math|Vector|VectorLength (Math|Vector|MakeVector _x_3 _y_2 _z_3)) (Math|Vector|VectorLength (Utilities|Operators|Multiply (Math|Float|Clamp(Float) (Math|Float|Power (/ (- (* 1.0 (Variables|/Game/Tools/NewWaterfallGenerator/Blueprint/BP_WaterfallCustomWaterBodyVelocity.BP_WaterfallCustomWaterBodyVelocity_C|GetRadius _array_element_2)) (Math|Vector|Distance(Vector) (Spline|FindLocationClosesttoWorldLocation (Variables|/Game/Tools/NewWaterfallGenerator/Blueprint/BP_WaterfallCustomWaterBodyVelocity.BP_WaterfallCustomWaterBodyVelocity_C|GetSpline _array_element_2) (Math|Transform|TransformLocation _returnvalue_16 _array_element) "World") (Math|Transform|TransformLocation _returnvalue_16 _array_element))) (Variables|/Game/Tools/NewWaterfallGenerator/Blueprint/BP_WaterfallCustomWaterBodyVelocity.BP_WaterfallCustomWaterBodyVelocity_C|GetEdgeSoft _array_element_2)) (Variables|/Game/Tools/NewWaterfallGenerator/Blueprint/BP_WaterfallCustomWaterBodyVelocity.BP_WaterfallCustomWaterBodyVelocity_C|GetEdgeGradientPower _array_element_2))) (Math|Vector|Normalize (Math|Vector|ProjectVectorontoPlane (Spline|FindTangentClosesttoWorldLocation (Variables|/Game/Tools/NewWaterfallGenerator/Blueprint/BP_WaterfallCustomWaterBodyVelocity.BP_WaterfallCustomWaterBodyVelocity_C|GetSpline _array_element_2) (Math|Transform|TransformLocation _returnvalue_16 _array_element) "World") "0.000000,0.000000,1.000000")) (Variables|/Game/Tools/NewWaterfallGenerator/Blueprint/BP_WaterfallCustomWaterBodyVelocity.BP_WaterfallCustomWaterBodyVelocity_C|GetVelocityStrength _array_element_2)))))) (* (Utilities|Operators|Multiply (Math|Float|Clamp(Float) (Math|Float|Power (/ (- (* 1.0 (Variables|/Game/Tools/NewWaterfallGenerator/Blueprint/BP_WaterfallCustomWaterBodyVelocity.BP_WaterfallCustomWaterBodyVelocity_C|GetRadius _array_element_2)) (Math|Vector|Distance(Vector) (Spline|FindLocationClosesttoWorldLocation (Variables|/Game/Tools/NewWaterfallGenerator/Blueprint/BP_WaterfallCustomWaterBodyVelocity.BP_WaterfallCustomWaterBodyVelocity_C|GetSpline _array_element_2) (Math|Transform|TransformLocation _returnvalue_16 _array_element) "World") (Math|Transform|TransformLocation _returnvalue_16 _array_element))) (Variables|/Game/Tools/NewWaterfallGenerator/Blueprint/BP_WaterfallCustomWaterBodyVelocity.BP_WaterfallCustomWaterBodyVelocity_C|GetEdgeSoft _array_element_2)) (Variables|/Game/Tools/NewWaterfallGenerator/Blueprint/BP_WaterfallCustomWaterBodyVelocity.BP_WaterfallCustomWaterBodyVelocity_C|GetEdgeGradientPower _array_element_2))) (Math|Vector|Normalize (Math|Vector|ProjectVectorontoPlane (Spline|FindTangentClosesttoWorldLocation (Variables|/Game/Tools/NewWaterfallGenerator/Blueprint/BP_WaterfallCustomWaterBodyVelocity.BP_WaterfallCustomWaterBodyVelocity_C|GetSpline _array_element_2) (Math|Transform|TransformLocation _returnvalue_16 _array_element) "World") "0.000000,0.000000,1.000000")) (Variables|/Game/Tools/NewWaterfallGenerator/Blueprint/BP_WaterfallCustomWaterBodyVelocity.BP_WaterfallCustomWaterBodyVelocity_C|GetVelocityStrength _array_element_2)) (/ (Math|Vector|VectorLength (Utilities|Operators|Multiply (Math|Float|Clamp(Float) (Math|Float|Power (/ (- (* 1.0 (Variables|/Game/Tools/NewWaterfallGenerator/Blueprint/BP_WaterfallCustomWaterBodyVelocity.BP_WaterfallCustomWaterBodyVelocity_C|GetRadius _array_element_2)) (Math|Vector|Distance(Vector) (Spline|FindLocationClosesttoWorldLocation (Variables|/Game/Tools/NewWaterfallGenerator/Blueprint/BP_WaterfallCustomWaterBodyVelocity.BP_WaterfallCustomWaterBodyVelocity_C|GetSpline _array_element_2) (Math|Transform|TransformLocation _returnvalue_16 _array_element) "World") (Math|Transform|TransformLocation _returnvalue_16 _array_element))) (Variables|/Game/Tools/NewWaterfallGenerator/Blueprint/BP_WaterfallCustomWaterBodyVelocity.BP_WaterfallCustomWaterBodyVelocity_C|GetEdgeSoft _array_element_2)) (Variables|/Game/Tools/NewWaterfallGenerator/Blueprint/BP_WaterfallCustomWaterBodyVelocity.BP_WaterfallCustomWaterBodyVelocity_C|GetEdgeGradientPower _array_element_2))) (Math|Vector|Normalize (Math|Vector|ProjectVectorontoPlane (Spline|FindTangentClosesttoWorldLocation (Variables|/Game/Tools/NewWaterfallGenerator/Blueprint/BP_WaterfallCustomWaterBodyVelocity.BP_WaterfallCustomWaterBodyVelocity_C|GetSpline _array_element_2) (Math|Transform|TransformLocation _returnvalue_16 _array_element) "World") "0.000000,0.000000,1.000000")) (Variables|/Game/Tools/NewWaterfallGenerator/Blueprint/BP_WaterfallCustomWaterBodyVelocity.BP_WaterfallCustomWaterBodyVelocity_C|GetVelocityStrength _array_element_2))) (+ (Math|Vector|VectorLength (Math|Vector|MakeVector _x_3 _y_2 _z_3)) (Math|Vector|VectorLength (Utilities|Operators|Multiply (Math|Float|Clamp(Float) (Math|Float|Power (/ (- (* 1.0 (Variables|/Game/Tools/NewWaterfallGenerator/Blueprint/BP_WaterfallCustomWaterBodyVelocity.BP_WaterfallCustomWaterBodyVelocity_C|GetRadius _array_element_2)) (Math|Vector|Distance(Vector) (Spline|FindLocationClosesttoWorldLocation (Variables|/Game/Tools/NewWaterfallGenerator/Blueprint/BP_WaterfallCustomWaterBodyVelocity.BP_WaterfallCustomWaterBodyVelocity_C|GetSpline _array_element_2) (Math|Transform|TransformLocation _returnvalue_16 _array_element) "World") (Math|Transform|TransformLocation _returnvalue_16 _array_element))) (Variables|/Game/Tools/NewWaterfallGenerator/Blueprint/BP_WaterfallCustomWaterBodyVelocity.BP_WaterfallCustomWaterBodyVelocity_C|GetEdgeSoft _array_element_2)) (Variables|/Game/Tools/NewWaterfallGenerator/Blueprint/BP_WaterfallCustomWaterBodyVelocity.BP_WaterfallCustomWaterBodyVelocity_C|GetEdgeGradientPower _array_element_2))) (Math|Vector|Normalize (Math|Vector|ProjectVectorontoPlane (Spline|FindTangentClosesttoWorldLocation (Variables|/Game/Tools/NewWaterfallGenerator/Blueprint/BP_WaterfallCustomWaterBodyVelocity.BP_WaterfallCustomWaterBodyVelocity_C|GetSpline _array_element_2) (Math|Transform|TransformLocation _returnvalue_16 _array_element) "World") "0.000000,0.000000,1.000000")) (Variables|/Game/Tools/NewWaterfallGenerator/Blueprint/BP_WaterfallCustomWaterBodyVelocity.BP_WaterfallCustomWaterBodyVelocity_C|GetVelocityStrength _array_element_2)))))))) (Math|Float|Max(Float) _w_2 (Math|Float|Clamp(Float) (Utilities|Operators|Multiply (Math|Float|Clamp(Float) (Math|Float|Power (/ (- (* 1.0 (Variables|/Game/Tools/NewWaterfallGenerator/Blueprint/BP_WaterfallCustomWaterBodyVelocity.BP_WaterfallCustomWaterBodyVelocity_C|GetRadius _array_element_2)) (Math|Vector|Distance(Vector) (Spline|FindLocationClosesttoWorldLocation (Variables|/Game/Tools/NewWaterfallGenerator/Blueprint/BP_WaterfallCustomWaterBodyVelocity.BP_WaterfallCustomWaterBodyVelocity_C|GetSpline _array_element_2) (Math|Transform|TransformLocation _returnvalue_16 _array_element) "World") (Math|Transform|TransformLocation _returnvalue_16 _array_element))) (Variables|/Game/Tools/NewWaterfallGenerator/Blueprint/BP_WaterfallCustomWaterBodyVelocity.BP_WaterfallCustomWaterBodyVelocity_C|GetEdgeSoft _array_element_2)) (Variables|/Game/Tools/NewWaterfallGenerator/Blueprint/BP_WaterfallCustomWaterBodyVelocity.BP_WaterfallCustomWaterBodyVelocity_C|GetEdgeGradientPower _array_element_2))) (Variables|/Game/Tools/NewWaterfallGenerator/Blueprint/BP_WaterfallCustomWaterBodyVelocity.BP_WaterfallCustomWaterBodyVelocity_C|GetVelocityStrength _array_element_2) (Variables|/Game/Tools/NewWaterfallGenerator/Blueprint/BP_WaterfallCustomWaterBodyVelocity.BP_WaterfallCustomWaterBodyVelocity_C|GetWhiteWaterScale _array_element_2) 1.0)))))
                    (else
                      (LocalVariables|Settmpsinglevtxvl (Math|Vector4|MakeVector4 (.x (Utilities|Operators|Multiply (Math|Float|Clamp(Float) (Math|Float|Power (/ (- (* 1.0 (Variables|/Game/Tools/NewWaterfallGenerator/Blueprint/BP_WaterfallCustomWaterBodyVelocity.BP_WaterfallCustomWaterBodyVelocity_C|GetRadius _array_element_2)) (Math|Vector|Distance(Vector) (Spline|FindLocationClosesttoWorldLocation (Variables|/Game/Tools/NewWaterfallGenerator/Blueprint/BP_WaterfallCustomWaterBodyVelocity.BP_WaterfallCustomWaterBodyVelocity_C|GetSpline _array_element_2) (Math|Transform|TransformLocation _returnvalue_16 _array_element) "World") (Math|Transform|TransformLocation _returnvalue_16 _array_element))) (Variables|/Game/Tools/NewWaterfallGenerator/Blueprint/BP_WaterfallCustomWaterBodyVelocity.BP_WaterfallCustomWaterBodyVelocity_C|GetEdgeSoft _array_element_2)) (Variables|/Game/Tools/NewWaterfallGenerator/Blueprint/BP_WaterfallCustomWaterBodyVelocity.BP_WaterfallCustomWaterBodyVelocity_C|GetEdgeGradientPower _array_element_2))) (Math|Vector|Normalize (Math|Vector|ProjectVectorontoPlane (Spline|FindTangentClosesttoWorldLocation (Variables|/Game/Tools/NewWaterfallGenerator/Blueprint/BP_WaterfallCustomWaterBodyVelocity.BP_WaterfallCustomWaterBodyVelocity_C|GetSpline _array_element_2) (Math|Transform|TransformLocation _returnvalue_16 _array_element) "World") "0.000000,0.000000,1.000000")) (Variables|/Game/Tools/NewWaterfallGenerator/Blueprint/BP_WaterfallCustomWaterBodyVelocity.BP_WaterfallCustomWaterBodyVelocity_C|GetVelocityStrength _array_element_2))) (.y (Utilities|Operators|Multiply (Math|Float|Clamp(Float) (Math|Float|Power (/ (- (* 1.0 (Variables|/Game/Tools/NewWaterfallGenerator/Blueprint/BP_WaterfallCustomWaterBodyVelocity.BP_WaterfallCustomWaterBodyVelocity_C|GetRadius _array_element_2)) (Math|Vector|Distance(Vector) (Spline|FindLocationClosesttoWorldLocation (Variables|/Game/Tools/NewWaterfallGenerator/Blueprint/BP_WaterfallCustomWaterBodyVelocity.BP_WaterfallCustomWaterBodyVelocity_C|GetSpline _array_element_2) (Math|Transform|TransformLocation _returnvalue_16 _array_element) "World") (Math|Transform|TransformLocation _returnvalue_16 _array_element))) (Variables|/Game/Tools/NewWaterfallGenerator/Blueprint/BP_WaterfallCustomWaterBodyVelocity.BP_WaterfallCustomWaterBodyVelocity_C|GetEdgeSoft _array_element_2)) (Variables|/Game/Tools/NewWaterfallGenerator/Blueprint/BP_WaterfallCustomWaterBodyVelocity.BP_WaterfallCustomWaterBodyVelocity_C|GetEdgeGradientPower _array_element_2))) (Math|Vector|Normalize (Math|Vector|ProjectVectorontoPlane (Spline|FindTangentClosesttoWorldLocation (Variables|/Game/Tools/NewWaterfallGenerator/Blueprint/BP_WaterfallCustomWaterBodyVelocity.BP_WaterfallCustomWaterBodyVelocity_C|GetSpline _array_element_2) (Math|Transform|TransformLocation _returnvalue_16 _array_element) "World") "0.000000,0.000000,1.000000")) (Variables|/Game/Tools/NewWaterfallGenerator/Blueprint/BP_WaterfallCustomWaterBodyVelocity.BP_WaterfallCustomWaterBodyVelocity_C|GetVelocityStrength _array_element_2))) (.z (Utilities|Operators|Multiply (Math|Float|Clamp(Float) (Math|Float|Power (/ (- (* 1.0 (Variables|/Game/Tools/NewWaterfallGenerator/Blueprint/BP_WaterfallCustomWaterBodyVelocity.BP_WaterfallCustomWaterBodyVelocity_C|GetRadius _array_element_2)) (Math|Vector|Distance(Vector) (Spline|FindLocationClosesttoWorldLocation (Variables|/Game/Tools/NewWaterfallGenerator/Blueprint/BP_WaterfallCustomWaterBodyVelocity.BP_WaterfallCustomWaterBodyVelocity_C|GetSpline _array_element_2) (Math|Transform|TransformLocation _returnvalue_16 _array_element) "World") (Math|Transform|TransformLocation _returnvalue_16 _array_element))) (Variables|/Game/Tools/NewWaterfallGenerator/Blueprint/BP_WaterfallCustomWaterBodyVelocity.BP_WaterfallCustomWaterBodyVelocity_C|GetEdgeSoft _array_element_2)) (Variables|/Game/Tools/NewWaterfallGenerator/Blueprint/BP_WaterfallCustomWaterBodyVelocity.BP_WaterfallCustomWaterBodyVelocity_C|GetEdgeGradientPower _array_element_2))) (Math|Vector|Normalize (Math|Vector|ProjectVectorontoPlane (Spline|FindTangentClosesttoWorldLocation (Variables|/Game/Tools/NewWaterfallGenerator/Blueprint/BP_WaterfallCustomWaterBodyVelocity.BP_WaterfallCustomWaterBodyVelocity_C|GetSpline _array_element_2) (Math|Transform|TransformLocation _returnvalue_16 _array_element) "World") "0.000000,0.000000,1.000000")) (Variables|/Game/Tools/NewWaterfallGenerator/Blueprint/BP_WaterfallCustomWaterBodyVelocity.BP_WaterfallCustomWaterBodyVelocity_C|GetVelocityStrength _array_element_2))) (Math|Float|Clamp(Float) (Utilities|Operators|Multiply (Math|Float|Clamp(Float) (Math|Float|Power (/ (- (* 1.0 (Variables|/Game/Tools/NewWaterfallGenerator/Blueprint/BP_WaterfallCustomWaterBodyVelocity.BP_WaterfallCustomWaterBodyVelocity_C|GetRadius _array_element_2)) (Math|Vector|Distance(Vector) (Spline|FindLocationClosesttoWorldLocation (Variables|/Game/Tools/NewWaterfallGenerator/Blueprint/BP_WaterfallCustomWaterBodyVelocity.BP_WaterfallCustomWaterBodyVelocity_C|GetSpline _array_element_2) (Math|Transform|TransformLocation _returnvalue_16 _array_element) "World") (Math|Transform|TransformLocation _returnvalue_16 _array_element))) (Variables|/Game/Tools/NewWaterfallGenerator/Blueprint/BP_WaterfallCustomWaterBodyVelocity.BP_WaterfallCustomWaterBodyVelocity_C|GetEdgeSoft _array_element_2)) (Variables|/Game/Tools/NewWaterfallGenerator/Blueprint/BP_WaterfallCustomWaterBodyVelocity.BP_WaterfallCustomWaterBodyVelocity_C|GetEdgeGradientPower _array_element_2))) (Variables|/Game/Tools/NewWaterfallGenerator/Blueprint/BP_WaterfallCustomWaterBodyVelocity.BP_WaterfallCustomWaterBodyVelocity_C|GetVelocityStrength _array_element_2) (Variables|/Game/Tools/NewWaterfallGenerator/Blueprint/BP_WaterfallCustomWaterBodyVelocity.BP_WaterfallCustomWaterBodyVelocity_C|GetWhiteWaterScale _array_element_2) 1.0))))))))
              (:then_2
                (if (<= (Math|Vector|VectorLength (Math|Vector|ProjectVectorontoPlane (- (+ _returnvalue_18 (* _returnvalue_19 (Math|Float|Clamp(Float) (/ (Math|Vector|DotProduct (- (Math|Transform|TransformLocation _returnvalue_16 _array_element) _returnvalue_18) _returnvalue_19) _returnvalue_20)))) (Math|Transform|TransformLocation _returnvalue_16 _array_element)) "0.000000,0.000000,1.000000")) _whitewaterradius)
                  (LocalVariables|Settmpsinglevtxvl (Math|Vector4|MakeVector4 (.x (* (Math|Vector|Normalize (Math|Vector|SelectVector (Math|Vector|Lerp(Vector) (Math|Vector|Normalize (Math|Vector|MakeVector _x_4 _y_3 _z_4)) (Math|Vector|Normalize (* (Transformation|GetActorForwardVector (Variables|Default|GetTargetActor)) "-1.000000,-1.000000,-1.000000")) (Math|Float|Clamp(Float) (Math|Float|Power (/ (- _whitewaterradius (Math|Vector|VectorLength (Math|Vector|ProjectVectorontoPlane (- (+ _returnvalue_18 (* _returnvalue_19 (Math|Float|Clamp(Float) (/ (Math|Vector|DotProduct (- (Math|Transform|TransformLocation _returnvalue_16 _array_element) _returnvalue_18) _returnvalue_19) _returnvalue_20)))) (Math|Transform|TransformLocation _returnvalue_16 _array_element)) "0.000000,0.000000,1.000000"))) (Math|Float|Max(Float) (- (Variables|/Game/Tools/NewWaterfallGenerator/Blueprint/BP_WaterfallGenerator.BP_WaterfallGenerator_C|GetWhiteWaterRadius (Variables|Default|GetTargetActor)) (Variables|/Game/Tools/NewWaterfallGenerator/Blueprint/BP_WaterfallGenerator.BP_WaterfallGenerator_C|GetWhiteWaterSoftWidth (Variables|Default|GetTargetActor))))) (Variables|/Game/Tools/NewWaterfallGenerator/Blueprint/BP_WaterfallGenerator.BP_WaterfallGenerator_C|GetWhiteWaterSoftGradient (Variables|Default|GetTargetActor))))) (Math|Vector|Normalize (* (Transformation|GetActorForwardVector (Variables|Default|GetTargetActor)) "-1.000000,-1.000000,-1.000000")) (>= (Math|Vector|VectorLength (Math|Vector|Normalize (Math|Vector|MakeVector _x_4 _y_3 _z_4))) 0.1))) (Math|Float|Lerp (Math|Vector|VectorLength (Math|Vector|MakeVector _x_4 _y_3 _z_4)) (Variables|/Game/Tools/NewWaterfallGenerator/Blueprint/BP_WaterfallGenerator.BP_WaterfallGenerator_C|GetWhiteWaterVelocity (Variables|Default|GetTargetActor)) (Math|Float|Clamp(Float) (Math|Float|Power (/ (- _whitewaterradius (Math|Vector|VectorLength (Math|Vector|ProjectVectorontoPlane (- (+ _returnvalue_18 (* _returnvalue_19 (Math|Float|Clamp(Float) (/ (Math|Vector|DotProduct (- (Math|Transform|TransformLocation _returnvalue_16 _array_element) _returnvalue_18) _returnvalue_19) _returnvalue_20)))) (Math|Transform|TransformLocation _returnvalue_16 _array_element)) "0.000000,0.000000,1.000000"))) (Math|Float|Max(Float) (- (Variables|/Game/Tools/NewWaterfallGenerator/Blueprint/BP_WaterfallGenerator.BP_WaterfallGenerator_C|GetWhiteWaterRadius (Variables|Default|GetTargetActor)) (Variables|/Game/Tools/NewWaterfallGenerator/Blueprint/BP_WaterfallGenerator.BP_WaterfallGenerator_C|GetWhiteWaterSoftWidth (Variables|Default|GetTargetActor))))) (Variables|/Game/Tools/NewWaterfallGenerator/Blueprint/BP_WaterfallGenerator.BP_WaterfallGenerator_C|GetWhiteWaterSoftGradient (Variables|Default|GetTargetActor))))))) (.y (* (Math|Vector|Normalize (Math|Vector|SelectVector (Math|Vector|Lerp(Vector) (Math|Vector|Normalize (Math|Vector|MakeVector _x_4 _y_3 _z_4)) (Math|Vector|Normalize (* (Transformation|GetActorForwardVector (Variables|Default|GetTargetActor)) "-1.000000,-1.000000,-1.000000")) (Math|Float|Clamp(Float) (Math|Float|Power (/ (- _whitewaterradius (Math|Vector|VectorLength (Math|Vector|ProjectVectorontoPlane (- (+ _returnvalue_18 (* _returnvalue_19 (Math|Float|Clamp(Float) (/ (Math|Vector|DotProduct (- (Math|Transform|TransformLocation _returnvalue_16 _array_element) _returnvalue_18) _returnvalue_19) _returnvalue_20)))) (Math|Transform|TransformLocation _returnvalue_16 _array_element)) "0.000000,0.000000,1.000000"))) (Math|Float|Max(Float) (- (Variables|/Game/Tools/NewWaterfallGenerator/Blueprint/BP_WaterfallGenerator.BP_WaterfallGenerator_C|GetWhiteWaterRadius (Variables|Default|GetTargetActor)) (Variables|/Game/Tools/NewWaterfallGenerator/Blueprint/BP_WaterfallGenerator.BP_WaterfallGenerator_C|GetWhiteWaterSoftWidth (Variables|Default|GetTargetActor))))) (Variables|/Game/Tools/NewWaterfallGenerator/Blueprint/BP_WaterfallGenerator.BP_WaterfallGenerator_C|GetWhiteWaterSoftGradient (Variables|Default|GetTargetActor))))) (Math|Vector|Normalize (* (Transformation|GetActorForwardVector (Variables|Default|GetTargetActor)) "-1.000000,-1.000000,-1.000000")) (>= (Math|Vector|VectorLength (Math|Vector|Normalize (Math|Vector|MakeVector _x_4 _y_3 _z_4))) 0.1))) (Math|Float|Lerp (Math|Vector|VectorLength (Math|Vector|MakeVector _x_4 _y_3 _z_4)) (Variables|/Game/Tools/NewWaterfallGenerator/Blueprint/BP_WaterfallGenerator.BP_WaterfallGenerator_C|GetWhiteWaterVelocity (Variables|Default|GetTargetActor)) (Math|Float|Clamp(Float) (Math|Float|Power (/ (- _whitewaterradius (Math|Vector|VectorLength (Math|Vector|ProjectVectorontoPlane (- (+ _returnvalue_18 (* _returnvalue_19 (Math|Float|Clamp(Float) (/ (Math|Vector|DotProduct (- (Math|Transform|TransformLocation _returnvalue_16 _array_element) _returnvalue_18) _returnvalue_19) _returnvalue_20)))) (Math|Transform|TransformLocation _returnvalue_16 _array_element)) "0.000000,0.000000,1.000000"))) (Math|Float|Max(Float) (- (Variables|/Game/Tools/NewWaterfallGenerator/Blueprint/BP_WaterfallGenerator.BP_WaterfallGenerator_C|GetWhiteWaterRadius (Variables|Default|GetTargetActor)) (Variables|/Game/Tools/NewWaterfallGenerator/Blueprint/BP_WaterfallGenerator.BP_WaterfallGenerator_C|GetWhiteWaterSoftWidth (Variables|Default|GetTargetActor))))) (Variables|/Game/Tools/NewWaterfallGenerator/Blueprint/BP_WaterfallGenerator.BP_WaterfallGenerator_C|GetWhiteWaterSoftGradient (Variables|Default|GetTargetActor))))))) (.z (* (Math|Vector|Normalize (Math|Vector|SelectVector (Math|Vector|Lerp(Vector) (Math|Vector|Normalize (Math|Vector|MakeVector _x_4 _y_3 _z_4)) (Math|Vector|Normalize (* (Transformation|GetActorForwardVector (Variables|Default|GetTargetActor)) "-1.000000,-1.000000,-1.000000")) (Math|Float|Clamp(Float) (Math|Float|Power (/ (- _whitewaterradius (Math|Vector|VectorLength (Math|Vector|ProjectVectorontoPlane (- (+ _returnvalue_18 (* _returnvalue_19 (Math|Float|Clamp(Float) (/ (Math|Vector|DotProduct (- (Math|Transform|TransformLocation _returnvalue_16 _array_element) _returnvalue_18) _returnvalue_19) _returnvalue_20)))) (Math|Transform|TransformLocation _returnvalue_16 _array_element)) "0.000000,0.000000,1.000000"))) (Math|Float|Max(Float) (- (Variables|/Game/Tools/NewWaterfallGenerator/Blueprint/BP_WaterfallGenerator.BP_WaterfallGenerator_C|GetWhiteWaterRadius (Variables|Default|GetTargetActor)) (Variables|/Game/Tools/NewWaterfallGenerator/Blueprint/BP_WaterfallGenerator.BP_WaterfallGenerator_C|GetWhiteWaterSoftWidth (Variables|Default|GetTargetActor))))) (Variables|/Game/Tools/NewWaterfallGenerator/Blueprint/BP_WaterfallGenerator.BP_WaterfallGenerator_C|GetWhiteWaterSoftGradient (Variables|Default|GetTargetActor))))) (Math|Vector|Normalize (* (Transformation|GetActorForwardVector (Variables|Default|GetTargetActor)) "-1.000000,-1.000000,-1.000000")) (>= (Math|Vector|VectorLength (Math|Vector|Normalize (Math|Vector|MakeVector _x_4 _y_3 _z_4))) 0.1))) (Math|Float|Lerp (Math|Vector|VectorLength (Math|Vector|MakeVector _x_4 _y_3 _z_4)) (Variables|/Game/Tools/NewWaterfallGenerator/Blueprint/BP_WaterfallGenerator.BP_WaterfallGenerator_C|GetWhiteWaterVelocity (Variables|Default|GetTargetActor)) (Math|Float|Clamp(Float) (Math|Float|Power (/ (- _whitewaterradius (Math|Vector|VectorLength (Math|Vector|ProjectVectorontoPlane (- (+ _returnvalue_18 (* _returnvalue_19 (Math|Float|Clamp(Float) (/ (Math|Vector|DotProduct (- (Math|Transform|TransformLocation _returnvalue_16 _array_element) _returnvalue_18) _returnvalue_19) _returnvalue_20)))) (Math|Transform|TransformLocation _returnvalue_16 _array_element)) "0.000000,0.000000,1.000000"))) (Math|Float|Max(Float) (- (Variables|/Game/Tools/NewWaterfallGenerator/Blueprint/BP_WaterfallGenerator.BP_WaterfallGenerator_C|GetWhiteWaterRadius (Variables|Default|GetTargetActor)) (Variables|/Game/Tools/NewWaterfallGenerator/Blueprint/BP_WaterfallGenerator.BP_WaterfallGenerator_C|GetWhiteWaterSoftWidth (Variables|Default|GetTargetActor))))) (Variables|/Game/Tools/NewWaterfallGenerator/Blueprint/BP_WaterfallGenerator.BP_WaterfallGenerator_C|GetWhiteWaterSoftGradient (Variables|Default|GetTargetActor))))))) (Math|Float|Lerp _w_3 (Variables|/Game/Tools/NewWaterfallGenerator/Blueprint/BP_WaterfallGenerator.BP_WaterfallGenerator_C|GetWhiteWaterStrength (Variables|Default|GetTargetActor)) (Math|Float|Clamp(Float) (Math|Float|Power (/ (- _whitewaterradius (Math|Vector|VectorLength (Math|Vector|ProjectVectorontoPlane (- (+ _returnvalue_18 (* _returnvalue_19 (Math|Float|Clamp(Float) (/ (Math|Vector|DotProduct (- (Math|Transform|TransformLocation _returnvalue_16 _array_element) _returnvalue_18) _returnvalue_19) _returnvalue_20)))) (Math|Transform|TransformLocation _returnvalue_16 _array_element)) "0.000000,0.000000,1.000000"))) (Math|Float|Max(Float) (- (Variables|/Game/Tools/NewWaterfallGenerator/Blueprint/BP_WaterfallGenerator.BP_WaterfallGenerator_C|GetWhiteWaterRadius (Variables|Default|GetTargetActor)) (Variables|/Game/Tools/NewWaterfallGenerator/Blueprint/BP_WaterfallGenerator.BP_WaterfallGenerator_C|GetWhiteWaterSoftWidth (Variables|Default|GetTargetActor))))) (Variables|/Game/Tools/NewWaterfallGenerator/Blueprint/BP_WaterfallGenerator.BP_WaterfallGenerator_C|GetWhiteWaterSoftGradient (Variables|Default|GetTargetActor))))))))
                (Utilities|Array|Add (LocalVariables|Getvtxcllst) (Math|Color|MakeColor (Math|Float|SelectFloat (- 1.0 (/ (Math|Trig|Acos(Degrees) (Math|Vector|DotProduct (Math|Vector|Normalize (Math|Vector|MakeVector _x _y _z)) (Math|Vector|MakeVector 1.0))) 3.6e+02)) (/ (Math|Trig|Acos(Degrees) (Math|Vector|DotProduct (Math|Vector|Normalize (Math|Vector|MakeVector _x _y _z)) (Math|Vector|MakeVector 1.0))) 3.6e+02) (>= (Math|Vector|DotProduct (Math|Vector|CrossProduct (Math|Vector|Normalize (Math|Vector|MakeVector _x _y _z)) (Math|Vector|MakeVector 1.0)) "0.000000,0.000000,1.000000") 0)) (Math|Vector|VectorLength (Math|Vector|MakeVector _x _y _z)) _w (Math|Float|SelectFloat 0.0 1.0 (< (Math|Vector|VectorLength (- (Spline|FindLocationClosesttoWorldLocation (Variables|/Game/Tools/NewWaterfallGenerator/Blueprint/BP_WaterfallGenerator.BP_WaterfallGenerator_C|GetSpline_CustomWaterBody (Variables|Default|GetTargetActor)) (Math|Transform|TransformLocation (Math|Transform|ComposeTransforms (Math|Transform|MakeTransform (Math|Vector|MakeVector (.x (Variables|/Game/Tools/NewWaterfallGenerator/Blueprint/BP_WaterfallGenerator.BP_WaterfallGenerator_C|Get◉Top (Variables|Default|GetTargetActor))) 0.0 (+ (.z (Variables|/Game/Tools/NewWaterfallGenerator/Blueprint/BP_WaterfallGenerator.BP_WaterfallGenerator_C|Get◉Top (Variables|Default|GetTargetActor))) (* (Variables|/Game/Tools/NewWaterfallGenerator/Blueprint/BP_WaterfallGenerator.BP_WaterfallGenerator_C|GetThickness (Variables|Default|GetTargetActor)) 0.5)))) _returnvalue_9) _array_element) "World") (Math|Transform|TransformLocation (Math|Transform|ComposeTransforms (Math|Transform|MakeTransform (Math|Vector|MakeVector (.x (Variables|/Game/Tools/NewWaterfallGenerator/Blueprint/BP_WaterfallGenerator.BP_WaterfallGenerator_C|Get◉Top (Variables|Default|GetTargetActor))) 0.0 (+ (.z (Variables|/Game/Tools/NewWaterfallGenerator/Blueprint/BP_WaterfallGenerator.BP_WaterfallGenerator_C|Get◉Top (Variables|Default|GetTargetActor))) (* (Variables|/Game/Tools/NewWaterfallGenerator/Blueprint/BP_WaterfallGenerator.BP_WaterfallGenerator_C|GetThickness (Variables|Default|GetTargetActor)) 0.5)))) _returnvalue_9) _array_element))) 1e+01)))))))
          ;[1 0.0338 0 0.5] TODO:Fix emptry vertexcolor mesh
          (if (Utilities|Array|IsEmpty _vtxcllst)
            (LocalVariables|SetTargetMesh _returnvalue_26)
            (else
              (bind _colorlist (GeometryScript|ListUtils|ConvertArraytoColorList _vtxcllst))
              (bind _returnvalue_28 (GeometryScript|VertexColor|SetMeshPerVertexColors _returnvalue_26 _colorlist))
              (LocalVariables|SetTargetMesh _returnvalue_28)))
          (Utilities|FlowControl|Sequence
            (:then_0
              (bind _copytomeshout (GeometryScript|Decomposition|CopyMeshtoMesh _targetmesh (Variables|Default|GetDynamicMesh_CustomWaterBody_LOD1)))
              (GeometryScript|Simplification|ApplyUniformRemesh :TargetMesh _copytomeshout :UniformOptions (Utilities|Struct|MakeGeometryScriptUniformRemeshOptions "TargetEdgeLength" 5000 (Variables|/Game/Tools/NewWaterfallGenerator/Blueprint/BP_WaterfallGenerator.BP_WaterfallGenerator_C|GetWaterMeshPolyEdgeLength_LOD1 (Variables|Default|GetTargetActor)))))
            (:then_1
              (bind _copytomeshout_2 (GeometryScript|Decomposition|CopyMeshtoMesh _targetmesh (Variables|Default|GetDynamicMesh_CustomWaterBody_LOD2)))
              (bind _returnvalue_29 (GeometryScript|Transforms|TranslateMesh _copytomeshout_2))
              (bind _returnvalue_30 (GeometryScript|Simplification|ApplyUniformRemesh :TargetMesh _returnvalue_29 :UniformOptions (Utilities|Struct|MakeGeometryScriptUniformRemeshOptions "TargetEdgeLength" 5000 (Variables|/Game/Tools/NewWaterfallGenerator/Blueprint/BP_WaterfallGenerator.BP_WaterfallGenerator_C|GetWaterMeshPolyEdgeLength_LOD2 (Variables|Default|GetTargetActor)))))
              ;[1 0.0338 0 0.5] TODO:Fix emptry vertexcolor mesh
              (if (not (Utilities|Array|IsEmpty _lod2vtxcl))
                (bind _colorlist_2 (GeometryScript|ListUtils|ConvertArraytoColorList _lod2vtxcl))
                (GeometryScript|VertexColor|SetMeshPerVertexColors _returnvalue_29 _colorlist_2)))))
        (:then_1
          ; Comment
          (bind _returnvalue_31 (GeometryScript|Primitives|AppendSimpleExtrudePolygon :TargetMesh _dynamicmesh_customwaterbody_collision :Transform (Math|Transform|MakeTransform "0, 0, 0") :PolygonVertices (LocalVariables|Getpolst) :Height _customwaterbodydepth :HeightSteps 0 :bCapped true :Origin "Base"))
          (GeometryScript|Transforms|TranslateMesh _returnvalue_31 (Math|Vector|MakeVector 0.0 0.0 (* _customwaterbodydepth -1.0)))
          (GeometryScript|Primitives|AppendBox :TargetMesh _dynamicmesh_customwaterbody_collision :Transform (Math|Transform|MakeTransform (Math|Vector|MakeVector (Utilities|Operators|Add (/ (.x (Variables|/Game/Tools/NewWaterfallGenerator/Blueprint/BP_WaterfallGenerator.BP_WaterfallGenerator_C|Get◉Bevel_Horizontal (Variables|Default|GetTargetActor))) 2.0) (/ (/ (Variables|/Game/Tools/NewWaterfallGenerator/Blueprint/BP_WaterfallGenerator.BP_WaterfallGenerator_C|GetThickness (Variables|Default|GetTargetActor)) -2.0) 2.0) 2e+01))) :DimensionX (- (.x (Variables|/Game/Tools/NewWaterfallGenerator/Blueprint/BP_WaterfallGenerator.BP_WaterfallGenerator_C|Get◉Bevel_Horizontal (Variables|Default|GetTargetActor))) (/ (Variables|/Game/Tools/NewWaterfallGenerator/Blueprint/BP_WaterfallGenerator.BP_WaterfallGenerator_C|GetThickness (Variables|Default|GetTargetActor)) -2.0)) :DimensionY (+ (Variables|/Game/Tools/NewWaterfallGenerator/Blueprint/BP_WaterfallGenerator.BP_WaterfallGenerator_C|GetWidth (Variables|Default|GetTargetActor)) (Variables|/Game/Tools/NewWaterfallGenerator/Blueprint/BP_WaterfallGenerator.BP_WaterfallGenerator_C|GetThickness (Variables|Default|GetTargetActor))) :DimensionZ (* (- (+ (.z (Variables|/Game/Tools/NewWaterfallGenerator/Blueprint/BP_WaterfallGenerator.BP_WaterfallGenerator_C|Get◉Bevel_Horizontal (Variables|Default|GetTargetActor))) (/ (Variables|/Game/Tools/NewWaterfallGenerator/Blueprint/BP_WaterfallGenerator.BP_WaterfallGenerator_C|GetThickness (Variables|Default|GetTargetActor)) 2.0)) (.z (Variables|/Game/Tools/NewWaterfallGenerator/Blueprint/BP_WaterfallGenerator.BP_WaterfallGenerator_C|Get◉Bottom (Variables|Default|GetTargetActor)))) -1.0) :StepsX 0 :StepsY 0 :StepsZ 0 :Origin "Base"))
        (:then_2
          ; Comment
          (Utilities|IsValid _targetstaticmesh_customwaterbody
            (:Is%20Valid
              (Utilities|FlowControl|Sequence
                (:then_0
                  (GeometryScript|StaticMesh|CopyMeshtoStaticMesh (Variables|Default|GetDynamicMesh_CustomWaterBody) _targetstaticmesh_customwaterbody (Utilities|Struct|MakeGeometryScriptCopyMeshtoAssetOptions :bEnableRecomputeNormals false :bEnableRecomputeTangents false :bEnableRemoveDegenerates false :BoneHierarchyMismatchHandling "DoNothing" :bRemapBoneIndicesToMatchAsset false :bUseOriginalVertexOrder false :bUseBuildScale true :bReplaceMaterials true :NewMaterialUsage "ReplaceAssetMaterials" :NewMaterials (Utilities|Array|MakeArray 0) :bCleanAssignedMaterials false :GenerateLightmapUVs "MatchTargetLODSetting" :bApplyNaniteSettings false :NaniteSettings "(bEnabled=True,FallbackPercentTriangles=100.000000,FallbackRelativeError=0.000000)" :NewNaniteSettings "(bEnabled=False,bExplicitTangents=False,bLerpUVs=True,bSeparable=False,bVoxelNDF=True,bVoxelOpacity=False,ShapePreservation=None,PositionPrecision=-2147483648,NormalPrecision=-1,TangentPrecision=-1,BoneWeightPrecision=-1,TargetMinimumResidencyInKB=0,KeepPercentTriangles=1.000000,TrimRelativeError=0.000000,GenerateFallback=PlatformDefault,FallbackTarget=Auto,FallbackPercentTriangles=1.000000,FallbackRelativeError=1.000000,MaxEdgeLengthFactor=0.000000,NumRays=1,VoxelLevel=0,RayBackUp=0.000000,DisplacementUVChannel=0,DisplacementMaps=,NaniteAssemblyData=(Parts=,Nodes=))" :bEmitTransaction false :bDeferMeshPostEditChange false) (Utilities|Struct|MakeGeometryScriptMeshWriteLOD false)))
                (:then_1
                  (GeometryScript|StaticMesh|CopyMeshtoStaticMesh (Variables|Default|GetDynamicMesh_CustomWaterBody_LOD1) _targetstaticmesh_customwaterbody _geometryscriptcopymeshtoassetoptions (Utilities|Struct|MakeGeometryScriptMeshWriteLOD false 1)))
                (:then_2
                  (GeometryScript|StaticMesh|CopyMeshtoStaticMesh (Variables|Default|GetDynamicMesh_CustomWaterBody_LOD2) _targetstaticmesh_customwaterbody _geometryscriptcopymeshtoassetoptions (Utilities|Struct|MakeGeometryScriptMeshWriteLOD false 2)))
                (:then_3
                  (GeometryScript|Collision|SetStaticMeshCollisionfromMesh (Variables|Default|GetDynamicMesh_CustomWaterBody_Collision) _targetstaticmesh_customwaterbody (Utilities|Struct|MakeGeometryScriptCollisionfromMeshOptions false "SweptHulls") "(bMarkAsCustomized=True)"))))))))))

(fn Generation_MeshGen
  ;[1 1 1 1 128] Generate mesh
  (Utilities|FlowControl|Sequence
    (:then_0
      (Utilities|FlowControl|Sequence
        (:then_0
          (bind _returnvalue (Game|ConstructObjectfromClass "/Script/GeometryFramework.DynamicMesh"))
          (Variables|Default|SetDynamicMesh_BAF _returnvalue))
        (:then_1
          (bind _returnvalue_2 (Game|ConstructObjectfromClass "/Script/GeometryFramework.DynamicMesh"))
          (Variables|Default|SetDynamicMesh_BAF_LOD1 _returnvalue_2))
        (:then_2
          (bind _returnvalue_3 (Game|ConstructObjectfromClass "/Script/GeometryFramework.DynamicMesh"))
          (Variables|Default|SetDynamicMesh_BAF_LOD2 _returnvalue_3))
        (:then_6
          (bind _returnvalue_4 (Game|ConstructObjectfromClass "/Script/GeometryFramework.DynamicMesh"))
          (Variables|Default|SetDynamicMesh_Wave _returnvalue_4))
        (:then_7
          (bind _returnvalue_5 (Game|ConstructObjectfromClass "/Script/GeometryFramework.DynamicMesh"))
          (Variables|Default|SetDynamicMesh_Wave_LOD1 _returnvalue_5))
        (:then_8
          (bind _returnvalue_6 (Game|ConstructObjectfromClass "/Script/GeometryFramework.DynamicMesh"))
          (Variables|Default|SetDynamicMesh_Wave_LOD2 _returnvalue_6))))
    (:then_1
      (Utilities|FlowControl|Sequence
        (:then_0
          (Utilities|FlowControl|Sequence
            (:then_0
              (bind _returnvalue_7 (GeometryScript|Primitives|AppendSimpleSweptPolygon (Variables|Default|GetDynamicMesh_BAF) (Utilities|Struct|MakeGeometryScriptPrimitiveOptions "PerFace") (Math|Transform|MakeTransform "0, 0, 0") (Base|GetSweptPointArray (Variables|Default|GetTargetActor) 0.0 (Variables|/Game/Tools/NewWaterfallGenerator/Blueprint/BP_WaterfallGenerator.BP_WaterfallGenerator_C|GetEdgeSegment (Variables|Default|GetTargetActor)) (Variables|/Game/Tools/NewWaterfallGenerator/Blueprint/BP_WaterfallGenerator.BP_WaterfallGenerator_C|GetCenterSegment (Variables|Default|GetTargetActor))) (Variables|/Game/Tools/NewWaterfallGenerator/Blueprint/BP_WaterfallGenerator.BP_WaterfallGenerator_C|GetTmpSampleSP_BAF (Variables|Default|GetTargetActor)) false false))
              (bind _returnvalue_8 (GeometryScript|UVs|ScaleMeshUVs _returnvalue_7 0 "(X=2.000000,Y=1.000000)"))
              (if (not (Variables|/Game/Tools/NewWaterfallGenerator/Blueprint/BP_WaterfallGenerator.BP_WaterfallGenerator_C|GetbDoubleFaceMesh (Variables|Default|GetTargetActor)))
                (bind _outvertexlist (CallFunction|MeshSelectionfromUV _returnvalue_8))
                (GeometryScript|MeshEdits|DeleteVerticesfromMesh _returnvalue_8 _outvertexlist)))
            (:then_1
              (bind _returnvalue_9 (GeometryScript|Primitives|AppendSimpleSweptPolygon (Variables|Default|GetDynamicMesh_BAF_LOD1) (Utilities|Struct|MakeGeometryScriptPrimitiveOptions "PerFace") (Math|Transform|MakeTransform "0, 0, 0") (Base|GetSweptPointArray (Variables|Default|GetTargetActor) 0.0 (Variables|/Game/Tools/NewWaterfallGenerator/Blueprint/BP_WaterfallGenerator.BP_WaterfallGenerator_C|GetEdgeSegment_LOD1 (Variables|Default|GetTargetActor)) (Variables|/Game/Tools/NewWaterfallGenerator/Blueprint/BP_WaterfallGenerator.BP_WaterfallGenerator_C|GetCenterSegment_LOD1 (Variables|Default|GetTargetActor))) (Variables|/Game/Tools/NewWaterfallGenerator/Blueprint/BP_WaterfallGenerator.BP_WaterfallGenerator_C|GetTmpSampleSP_BAF_LOD1 (Variables|Default|GetTargetActor)) false false))
              (bind _returnvalue_10 (GeometryScript|UVs|ScaleMeshUVs _returnvalue_9 0 "(X=2.000000,Y=1.000000)"))
              (if (not (Variables|/Game/Tools/NewWaterfallGenerator/Blueprint/BP_WaterfallGenerator.BP_WaterfallGenerator_C|GetbDoubleFaceMesh (Variables|Default|GetTargetActor)))
                (bind _outvertexlist_2 (CallFunction|MeshSelectionfromUV _returnvalue_10))
                (GeometryScript|MeshEdits|DeleteVerticesfromMesh _returnvalue_10 _outvertexlist_2)))
            (:then_2
              (bind _returnvalue_11 (GeometryScript|Primitives|AppendSimpleSweptPolygon (Variables|Default|GetDynamicMesh_BAF_LOD2) (Utilities|Struct|MakeGeometryScriptPrimitiveOptions "PerFace") (Math|Transform|MakeTransform "0, 0, 0") (Base|GetSweptPointArray (Variables|Default|GetTargetActor) 0.0 (Variables|/Game/Tools/NewWaterfallGenerator/Blueprint/BP_WaterfallGenerator.BP_WaterfallGenerator_C|GetEdgeSegment_LOD2 (Variables|Default|GetTargetActor)) (Variables|/Game/Tools/NewWaterfallGenerator/Blueprint/BP_WaterfallGenerator.BP_WaterfallGenerator_C|GetCenterSegment_LOD2 (Variables|Default|GetTargetActor))) (Variables|/Game/Tools/NewWaterfallGenerator/Blueprint/BP_WaterfallGenerator.BP_WaterfallGenerator_C|GetTmpSampleSP_BAF_LOD2 (Variables|Default|GetTargetActor)) false false))
              (bind _returnvalue_12 (GeometryScript|UVs|ScaleMeshUVs _returnvalue_11 0 "(X=2.000000,Y=1.000000)"))
              (if (not (Variables|/Game/Tools/NewWaterfallGenerator/Blueprint/BP_WaterfallGenerator.BP_WaterfallGenerator_C|GetbDoubleFaceMesh (Variables|Default|GetTargetActor)))
                (bind _outvertexlist_3 (CallFunction|MeshSelectionfromUV _returnvalue_12))
                (GeometryScript|MeshEdits|DeleteVerticesfromMesh _returnvalue_12 _outvertexlist_3)))))
        (:then_2
          (Utilities|FlowControl|Sequence
            (:then_0
              (bind _returnvalue_13 (GeometryScript|Primitives|AppendDisc (Variables|Default|GetDynamicMesh_Wave) (Utilities|Struct|MakeGeometryScriptPrimitiveOptions "PerFace") (Math|Transform|MakeTransform (Spline|GetLocationatDistanceAlongSpline (Variables|/Game/Tools/NewWaterfallGenerator/Blueprint/BP_WaterfallGenerator.BP_WaterfallGenerator_C|GetSpline_BevelAndFall (Variables|Default|GetTargetActor)) 0.0 "Local")) (Variables|/Game/Tools/NewWaterfallGenerator/Blueprint/BP_WaterfallGenerator.BP_WaterfallGenerator_C|GetWaveScale (Variables|Default|GetTargetActor)) (Variables|/Game/Tools/NewWaterfallGenerator/Blueprint/BP_WaterfallGenerator.BP_WaterfallGenerator_C|GetSplashSegment_AngleStep (Variables|Default|GetTargetActor)) (Variables|/Game/Tools/NewWaterfallGenerator/Blueprint/BP_WaterfallGenerator.BP_WaterfallGenerator_C|GetSplashSegment_SpokeStep (Variables|Default|GetTargetActor))))
              (GeometryScript|Normals|RecomputeNormals _returnvalue_13 (Utilities|Struct|MakeGeometryScriptCalculateNormalsOptions true)))
            (:then_1
              (bind _returnvalue_14 (GeometryScript|Primitives|AppendDisc (Variables|Default|GetDynamicMesh_Wave_LOD1) (Utilities|Struct|MakeGeometryScriptPrimitiveOptions "PerFace") (Math|Transform|MakeTransform (Spline|GetLocationatDistanceAlongSpline (Variables|/Game/Tools/NewWaterfallGenerator/Blueprint/BP_WaterfallGenerator.BP_WaterfallGenerator_C|GetSpline_BevelAndFall (Variables|Default|GetTargetActor)) 0.0 "Local")) (Variables|/Game/Tools/NewWaterfallGenerator/Blueprint/BP_WaterfallGenerator.BP_WaterfallGenerator_C|GetWaveScale (Variables|Default|GetTargetActor)) (Variables|/Game/Tools/NewWaterfallGenerator/Blueprint/BP_WaterfallGenerator.BP_WaterfallGenerator_C|GetSplashSegment_AngleStep_LOD1 (Variables|Default|GetTargetActor)) (Variables|/Game/Tools/NewWaterfallGenerator/Blueprint/BP_WaterfallGenerator.BP_WaterfallGenerator_C|GetSplashSegment_SpokeStep_LOD1 (Variables|Default|GetTargetActor))))
              (GeometryScript|Normals|RecomputeNormals _returnvalue_14 (Utilities|Struct|MakeGeometryScriptCalculateNormalsOptions true)))
            (:then_2
              (bind _returnvalue_15 (GeometryScript|Primitives|AppendDisc (Variables|Default|GetDynamicMesh_Wave_LOD2) (Utilities|Struct|MakeGeometryScriptPrimitiveOptions "PerFace") (Math|Transform|MakeTransform (Spline|GetLocationatDistanceAlongSpline (Variables|/Game/Tools/NewWaterfallGenerator/Blueprint/BP_WaterfallGenerator.BP_WaterfallGenerator_C|GetSpline_BevelAndFall (Variables|Default|GetTargetActor)) 0.0 "Local")) (Variables|/Game/Tools/NewWaterfallGenerator/Blueprint/BP_WaterfallGenerator.BP_WaterfallGenerator_C|GetWaveScale (Variables|Default|GetTargetActor)) (Variables|/Game/Tools/NewWaterfallGenerator/Blueprint/BP_WaterfallGenerator.BP_WaterfallGenerator_C|GetSplashSegment_AngleStep_LOD2 (Variables|Default|GetTargetActor)) (Variables|/Game/Tools/NewWaterfallGenerator/Blueprint/BP_WaterfallGenerator.BP_WaterfallGenerator_C|GetSplashSegment_SpokeStep_LOD2 (Variables|Default|GetTargetActor))))
              (GeometryScript|Normals|RecomputeNormals _returnvalue_15 (Utilities|Struct|MakeGeometryScriptCalculateNormalsOptions true)))))))))

(fn Generation_MeshMerge
  ;[1 1 1 1 128] Merge mesh
  (Utilities|FlowControl|Sequence
    (:then_0
      (Utilities|FlowControl|Sequence
        (:then_0
          (bind _returnvalue (Game|ConstructObjectfromClass "/Script/GeometryFramework.DynamicMesh"))
          (Variables|Default|SetDynamicMesh_Merged _returnvalue))
        (:then_1
          (bind _returnvalue_2 (Game|ConstructObjectfromClass "/Script/GeometryFramework.DynamicMesh"))
          (Variables|Default|SetDynamicMesh_Merged_LOD1 _returnvalue_2))
        (:then_2
          (bind _returnvalue_3 (Game|ConstructObjectfromClass "/Script/GeometryFramework.DynamicMesh"))
          (Variables|Default|SetDynamicMesh_Merged_LOD2 _returnvalue_3))))
    (:then_2
      (Utilities|FlowControl|Sequence
        (:then_0
          (Utilities|FlowControl|Sequence
            (:then_0
              (bind _returnvalue_4 (GeometryScript|Materials|EnableMaterialIDs (Variables|Default|GetDynamicMesh_BAF)))
              (bind (_triangleidlist _bhastriangleidgaps _returnvalue_5) (GeometryScript|MeshQueries|GetAllTriangleIDs _returnvalue_4))
              (bind _returnvalue_6 (GeometryScript|Materials|SetMaterialIDOnTriangles _returnvalue_5 _triangleidlist)))
            (:then_2
              (bind _returnvalue_7 (GeometryScript|Materials|EnableMaterialIDs (Variables|Default|GetDynamicMesh_Wave)))
              (bind (_triangleidlist_2 _bhastriangleidgaps_2 _returnvalue_8) (GeometryScript|MeshQueries|GetAllTriangleIDs _returnvalue_7))
              (bind _returnvalue_9 (GeometryScript|Materials|SetMaterialIDOnTriangles _returnvalue_8 _triangleidlist_2 1)))
            (:then_4
              (bind _returnvalue_10 (GeometryScript|Materials|EnableMaterialIDs (Variables|Default|GetDynamicMesh_Merged)))
              (bind _returnvalue_11 (GeometryScript|MeshEdits|AppendMesh :TargetMesh _returnvalue_10 :AppendMesh _returnvalue_6 :bDeferChangeNotifications false :AppendOptions "(CombineMode=EnableAllMatching)"))
              (GeometryScript|MeshEdits|AppendMesh :TargetMesh _returnvalue_11 :AppendMesh _returnvalue_9 :bDeferChangeNotifications false :AppendOptions "(CombineMode=EnableAllMatching)"))))
        (:then_1
          (Utilities|FlowControl|Sequence
            (:then_0
              (bind _returnvalue_12 (GeometryScript|Materials|EnableMaterialIDs (Variables|Default|GetDynamicMesh_BAF_LOD1)))
              (bind (_triangleidlist_3 _bhastriangleidgaps_3 _returnvalue_13) (GeometryScript|MeshQueries|GetAllTriangleIDs _returnvalue_12))
              (bind _returnvalue_14 (GeometryScript|Materials|SetMaterialIDOnTriangles _returnvalue_13 _triangleidlist_3)))
            (:then_2
              (bind _returnvalue_15 (GeometryScript|Materials|EnableMaterialIDs (Variables|Default|GetDynamicMesh_Wave_LOD1)))
              (bind (_triangleidlist_4 _bhastriangleidgaps_4 _returnvalue_16) (GeometryScript|MeshQueries|GetAllTriangleIDs _returnvalue_15))
              (bind _returnvalue_17 (GeometryScript|Materials|SetMaterialIDOnTriangles _returnvalue_16 _triangleidlist_4 1)))
            (:then_4
              (bind _returnvalue_18 (GeometryScript|Materials|EnableMaterialIDs (Variables|Default|GetDynamicMesh_Merged_LOD1)))
              (bind _returnvalue_19 (GeometryScript|MeshEdits|AppendMesh :TargetMesh _returnvalue_18 :AppendMesh _returnvalue_14 :bDeferChangeNotifications false :AppendOptions "(CombineMode=EnableAllMatching)"))
              (GeometryScript|MeshEdits|AppendMesh :TargetMesh _returnvalue_19 :AppendMesh _returnvalue_17 :bDeferChangeNotifications false :AppendOptions "(CombineMode=EnableAllMatching)"))))
        (:then_2
          (Utilities|FlowControl|Sequence
            (:then_0
              (bind _returnvalue_20 (GeometryScript|Materials|EnableMaterialIDs (Variables|Default|GetDynamicMesh_BAF_LOD2)))
              (bind (_triangleidlist_5 _bhastriangleidgaps_5 _returnvalue_21) (GeometryScript|MeshQueries|GetAllTriangleIDs _returnvalue_20))
              (bind _returnvalue_22 (GeometryScript|Materials|SetMaterialIDOnTriangles _returnvalue_21 _triangleidlist_5)))
            (:then_2
              (bind _returnvalue_23 (GeometryScript|Materials|EnableMaterialIDs (Variables|Default|GetDynamicMesh_Wave_LOD2)))
              (bind (_triangleidlist_6 _bhastriangleidgaps_6 _returnvalue_24) (GeometryScript|MeshQueries|GetAllTriangleIDs _returnvalue_23))
              (bind _returnvalue_25 (GeometryScript|Materials|SetMaterialIDOnTriangles _returnvalue_24 _triangleidlist_6 1)))
            (:then_4
              (bind _returnvalue_26 (GeometryScript|Materials|EnableMaterialIDs (Variables|Default|GetDynamicMesh_Merged_LOD2)))
              (bind _returnvalue_27 (GeometryScript|MeshEdits|AppendMesh :TargetMesh _returnvalue_26 :AppendMesh _returnvalue_22 :bDeferChangeNotifications false :AppendOptions "(CombineMode=EnableAllMatching)"))
              (GeometryScript|MeshEdits|AppendMesh :TargetMesh _returnvalue_27 :AppendMesh _returnvalue_25 :bDeferChangeNotifications false :AppendOptions "(CombineMode=EnableAllMatching)"))))))
    (:then_3
      (DynamicMeshComponent|SetDynamicMesh (Variables|/Game/Tools/NewWaterfallGenerator/Blueprint/BP_WaterfallGenerator.BP_WaterfallGenerator_C|GetDynamicMesh (Variables|Default|GetTargetActor)) (Variables|Default|GetDynamicMesh_Merged)))))

(fn MeshSelectionFromUV
  (params (InDM :type Dynamic%20Mesh%20Object%20Reference))
  (bind _uvlist (GeometryScript|UVs|GetMeshPerVertexUVs InDM))
  (bind _uvarray (GeometryScript|ListUtils|ConvertUVListtoArray _uvlist))
  (for (_array_element _array_index) _uvarray
    (if (>= (Math|Vector2D|BreakVector2D _array_element) 1.0)
      (Utilities|Array|Add (LocalVariables|GetIndexArray) _array_index)))
  (bind _indexlist (GeometryScript|ListUtils|ConvertArraytoIndexList (LocalVariables|GetIndexArray) "Vertex"))
  (return _indexlist))

(fn MessageBox-OK
  (params (Title :type Text) (Message :type Text))
  (EditorScripting|MessageDialog|ShowMessageDialog Title Message "Ok" "Ok" "Info"))
