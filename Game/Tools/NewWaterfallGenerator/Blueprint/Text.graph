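# Translucent, unlit, two-sided text material in the waterfall generator tool, with depth testing disabled so labels draw over geometry.
# It samples the Roboto distance-field font and remaps the red channel through symmetric lerps and saturates, with parameters A and B setting edge sharpness.
# The A mask blends in vertex color, while the B mask passes through CheapContrast.
# AlignMeshToTheCamera with a custom axis basis supplies camera-facing billboarding.
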Material /Game/Tools/NewWaterfallGenerator/Blueprint/Text
graph {
 properties {
  bDisableDepthTest=true
  blendMode=BLEND_Translucent
  pixelDepthOffsetMode=PDOM_Legacy
  shadingModel=MSM_Unlit
  twoSided=true
  usedShadingModels=Unlit
 }
 nodes [
  n1 = ScalarParameter() {defaultValue=90, parameterName=A}
  n2 = FontSampleParameter() {font=/Engine/EngineFonts/RobotoDistanceField, parameterName=Font}
  n3 = Saturate(None=LinearInterpolate(A=Multiply(A=@n1) {constB=-1}, B=@n1, Alpha=@n2.R) {constA=-90, constB=90})
  Add(A=@n3)
  LinearInterpolate(B=VertexColor(), Alpha=@n3)
  LinearInterpolate(Alpha=@n3)
  MaterialFunctionCall!(
   WorldPosition
   PivotLocation
   WS Normals
   Custom Object Basis 1=Constant3Vector() {constant=<rgba -1 0 0 0>}
   Custom Object Basis 2=Constant3Vector() {constant=<rgba 0 1 0 0>}
   Custom Object Basis 3=Constant3Vector() {constant=<rgba 0 0 1 0>}
  ) {materialFunction=/Engine/Functions/Engine_MaterialFunctions02/WorldPositionOffset/AlignMeshToTheCamera}
  n4 = ScalarParameter() {defaultValue=20, parameterName=B}
  n5 = Multiply(A=@n4) {constB=-1}
  MaterialFunctionCall(In=Saturate(None=LinearInterpolate(A=@n5, B=@n4, Alpha=@n2.R) {constA=-90, constB=90}), Contrast=Constant() {r=5}) {materialFunction=&CheapContrast}
 ]
}
