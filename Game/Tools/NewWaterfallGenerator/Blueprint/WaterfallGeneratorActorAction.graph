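# Editor utility that adds a DuplicateWaterfall action, cloning the first selected BP_WaterfallGenerator actor at a fixed Y offset in the editor world.
# It resets the copy by clearing target static meshes and water body links, unfreezing it, and restoring warning and auto-update flags.
# It then calls UserGuider, hides the frozen and warning text components, and shifts the viewport camera by the same offset.
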EditorUtilityBlueprint /Game/Tools/NewWaterfallGenerator/Blueprint/WaterfallGeneratorActorAction
graphs [
 DuplicateWaterfall {nodes [
   DuplicateWaterfall() {is=title K2Node_FunctionEntry}
   n1 = Get EditorActorSubsystem()
   > n2 = GetSelectedLevelActors(self=@n1)
   > n3 = Cast To BP_WaterfallGenerator!(execute, Object="Get (a copy)"(Array:Array of Actor Object References=@n2) -> (Output:Actor Object Reference)) -> (then, CastFailed, AsBP Waterfall Generator:BP Waterfall Generator Object Reference) {is=title K2Node_DynamicCast}
   > n4 = GetEditorWorld(self=Get UnrealEditorSubsystem())
   > n5 = DuplicateActor!(
    execute
    self:Editor Actor Subsystem Object Reference=@n1
    ActorToDuplicate:Actor Object Reference=@n3
    ToWorld:World Object Reference=@n4
    Offset:Vector="0,5000,0"
   ) -> (then, ReturnValue:Actor Object Reference)
   > n6 = Cast To BP_WaterfallGenerator(Object=@n5)
   > Set TargetStaticMesh!(execute, TargetStaticMesh:Static Mesh Object Reference, self:BP Waterfall Generator Object Reference=@n6) -> (then, Output_Get:Static Mesh Object Reference)
   > Set TargetStaticMesh_CustomWaterBody!(execute, TargetStaticMesh_CustomWaterBody:Static Mesh Object Reference, self:BP Waterfall Generator Object Reference=@n6) -> (then, Output_Get:Static Mesh Object Reference)
   > Set bFrozen!(execute, bFrozen:bool=false, self:BP Waterfall Generator Object Reference=@n6) -> (then, Output_Get:bool)
   > Set BottomLinkToLake!(execute, BottomLinkToLake:Fort Water Body Actor Soft Object Reference, self:BP Waterfall Generator Object Reference=@n6) -> (then, Output_Get:Fort Water Body Actor Soft Object Reference)
   > Set TopLinkToWaterBody!(execute, TopLinkToWaterBody:Fort Water Body Actor Soft Object Reference, self:BP Waterfall Generator Object Reference=@n6) -> (then, Output_Get:Fort Water Body Actor Soft Object Reference)
   > Set bDisableWarning!(execute, bDisableWarning:bool=false, self:BP Waterfall Generator Object Reference=@n6) -> (then, Output_Get:bool)
   > Set bAutoUpdateStaticMesh!(execute, bAutoUpdateStaticMesh:bool=true, self:BP Waterfall Generator Object Reference=@n6) -> (then, Output_Get:bool)
   > UserGuider(self:BP Waterfall Generator Object Reference=@n6)
   > SetVisibility!(
    execute
    self:Scene Component Object Reference=[
     Get FrozenTip!(self:BP Waterfall Generator Object Reference=@n6) -> (FrozenTip:Text Render Component Object Reference)
     Get NeedFrozenTip!(self:BP Waterfall Generator Object Reference=@n6) -> (NeedFrozenTip:Text Render Component Object Reference)
     Get WarningText!(self:BP Waterfall Generator Object Reference=@n6) -> (WarningText:Text Render Component Object Reference)
    ]
    bNewVisibility:bool=false
    bPropagateToChildren:bool=false
   ) -> (then)
   n7 = Get UnrealEditorSubsystem()
   n8 = GetLevelViewportCameraInfo!(self:Unreal Editor Subsystem Object Reference=@n7) -> (CameraLocation:Vector, CameraRotation:Rotator, ReturnValue:bool)
   > SetLevelViewportCameraInfo!(
    execute
    self:Unreal Editor Subsystem Object Reference=@n7
    CameraLocation:Vector=vector + vector(A=@n8, B="0,5000,0")
    CameraRotation:Rotator=@n8.CameraRotation
   ) -> (then)
  ]}
 EventGraph {links [], nodes []}
]
variables []
