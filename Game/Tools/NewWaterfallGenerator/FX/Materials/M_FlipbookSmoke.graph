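# Translucent Niagara sprite material for flipbook smoke, likely waterfall mist from the waterfall generator tool.
# A dynamic Frame parameter picks two adjacent cells in a square grid (only SubImages.R is used), blended by its fraction unless DisableBlend is set.
# The sampled alpha or RGB is refined separately to drive a two-color tint, scaled by particle color and emissive strength, and opacity with particle alpha, depth fade and optional view-angle fade.
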Material /Game/Tools/NewWaterfallGenerator/FX/Materials/M_FlipbookSmoke
graph {
 properties {bUsedWithNiagaraSprites=true, blendMode=BLEND_Translucent, pixelDepthOffsetMode=PDOM_Legacy, usedShadingModels=Default Lit}
 nodes [
  n1 = DynamicParameter() {paramNames [Frame,Param2,Param3,Param4]}
  n2 = Frac(None=@n1.Frame) {desc=Frame Lerp}
  n3 = Subtract(A=@n1.Frame, B=@n2) {desc=IndexA}
  n4 = VectorParameter() {parameterName=SubImages, defaultValue=<rgba 10 10 10 1>}
  n5 = MaterialFunctionCall(R=@n4.R, G=@n4.R) {materialFunction=&AppendMany}
  n6 = ComponentMask(None=@n5.RG) {desc="## Horz Images", r=true}
  n7 = Divide(A=Constant2Vector() {g=1, r=1}, B=@n5.RG) {desc=Scale}
  n8 = ComponentMask(None=@n7) {desc="1 / ## Horz Images", r=true}
  n9 = Multiply(A=@n3, B=@n8)
  n10 = VectorParameter() {parameterName=ScaleUVByCenter, defaultValue=<rgba 2 1 0 1>}
  n11 = MaterialFunctionCall(
   UVs=TextureCoordinate()
   Texture Scale=MaterialFunctionCall(R=@n10.R, G=@n10.R) {materialFunction=&AppendMany}.RG
  ) {materialFunction=&ScaleUVsByCenter}
  n12 = Add(A=AppendVector(A=Fmod(A=@n3, B=@n6) {desc=IndexA_Horz}, B=Floor(None=@n9) {desc=IndexA_Vert}) {desc=IndexA_Combined}, B=@n11.UVs)
  n13 = TextureObjectParameter() {parameterName=Param, texture=/Game/Tools/NewWaterfallGenerator/FX/Texture/T_DropSmoothSmoke_11x6}
  n14 = TextureSample(UVs=Frac(None=Multiply(A=@n12, B=@n7) {desc=UV_A}), Tex=@n13) {samplerType=SAMPLERTYPE_LinearColor, texture=/Engine/EngineMaterials/DefaultDiffuse}
  n15 = Add(B=@n3) {constA=1, desc=IndexB}
  n16 = Multiply(A=@n15, B=@n8)
  n17 = Add(A=@n11.UVs, B=AppendVector(A=Fmod(A=@n15, B=@n6) {desc=IndexB_Horz}, B=Floor(None=@n16) {desc=IndexB_Vert}) {desc=IndexB_Combined})
  n18 = TextureSample(UVs=Frac(None=Multiply(A=@n17, B=@n7) {desc=UV_B}), Tex=@n13) {samplerType=SAMPLERTYPE_LinearColor, texture=&DefaultDiffuse}
  n19 = StaticSwitchParameter(
   True=StaticSwitchParameter(True=@n14.A, False=LinearInterpolate(A=@n14.A, B=@n18.A, Alpha=@n2)) {parameterName=DisableBlend}
   False=StaticSwitchParameter(True=@n14.RGB, False=LinearInterpolate(A=@n14.RGB, B=@n18.RGB, Alpha=@n2)) {parameterName=DisableBlend}
  ) {parameterName=UseA}
  n20 = VectorParameter() {parameterName=ColorRefine, defaultValue=<rgba 1 1 1 0.5>}
  n21 = LinearInterpolate(
   A=VectorParameter() {parameterName=ColorA, defaultValue=<rgba 1 1 1 1>}.RGB
   B=VectorParameter() {parameterName=ColorB, defaultValue=<rgba 1 1 1 1>}.RGB
   Alpha=MaterialFunctionCall(Texture=@n19, RGB=@n20.RGB, A=@n20.RGB) {materialFunction=&MF_FX_Refine}.Result
  )
  n22 = ParticleColor()
  MaterialFunctionCall(Emissive=Multiply(A=Multiply(A=@n21, B=@n22.RGB), B=ScalarParameter() {parameterName=WaterfallMistEmissive})) {materialFunction=&MF_FX_EmissiveAdjustments}
  n23 = VectorParameter() {parameterName=OpacityRefine, defaultValue=<rgba 1 1 1 0.5>}
  n24 = MaterialFunctionCall(Texture=@n19, RGB=@n23.RGB, A=@n23.RGB) {materialFunction=&MF_FX_Refine}
  n25 = Power(
   Base=DotProduct(A=CameraVectorWS(), B=VertexNormalWS())
   Exp=ScalarParameter() {defaultValue=1, group=Opacity, parameterName=Angle Fade Power}
  )
  Multiply(
   A=DepthFade(Opacity=Multiply(A=@n22.A, B=@n24.Result), FadeDistance=ScalarParameter() {parameterName=DF})
   B=If(A=ScalarParameter() {group=Opacity, parameterName=Enable Angle Fade}, "A > B"=@n25, "A == B"=@n25, "A < B"=Constant() {r=1}) {constB=0.99}
  )
 ]
}
