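# Additive unlit Niagara sprite material for waterfall mist, named as a SubUV test of smoke flipbook.
# An 8x8 flipbook's green channel, optionally sphere-masked, passes through add, multiply and power adjustments, then depth fade set by a dynamic parameter and particle alpha.
# A color parameter scaled by mist emissive is added to particle color.
# SubUV and emissive-adjustment functions are present but feed no other node shown; the texture sample has no UV input.
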Material /Game/Tools/NewWaterfallGenerator/FX/Materials/M_FlipbookSmoke2_TestSubuv_Unlit
graph {
 properties {
  bCastRayTracedShadows=false
  bEnableResponsiveAA=true
  bUsedWithNiagaraSprites=true
  blendMode=BLEND_Additive
  pixelDepthOffsetMode=PDOM_Legacy
  shadingModel=MSM_Unlit
  translucencyLightingMode=TLM_VolumetricPerVertexNonDirectional
  usedShadingModels=Unlit
 }
 nodes [
  n1 = ParticleColor()
  Add(
   A=Multiply(
    A=VectorParameter() {parameterName=Color, defaultValue=<rgba 0.75 0.9167 1 1.2>}.RGB
    B=ScalarParameter() {parameterName=WaterfallMistEmissive}
   )
   B=@n1.RGB
  )
  n2 = Constant() {r=8}
  MaterialFunctionCall!(
   U_Images=@n2
   V_Images=@n2
   SubUV Texture RGB=TextureObject() {samplerType=SAMPLERTYPE_LinearColor, texture=/Game/Tools/NewWaterfallGenerator/FX/Texture/T_SmokeFlipbook_8x8}
   Time Speed
   Manual Time Control=StaticBool() {value=true}
   Time=Multiply(A=ParticleRelativeTime(), B=ScalarParameter() {defaultValue=8, parameterName=SubUV Relative Time Multiplier}) {constB=8}
   UVs
  ) {materialFunction=/Game/Effects/Fort_Effects/Materials/_Functions/MF_SubUV}
  MaterialFunctionCall(Emissive=@n1.RGB) {materialFunction=&MF_FX_EmissiveAdjustments}
  n3 = Constant2Vector() {g=0.5, r=0.5}
  n4 = SphereMask(A=TextureCoordinate(), B=@n3, Radius=@n3, Hardness=ScalarParameter() {parameterName=Sphere Mask Hardness}) {hardnessPercent=50}
  n5 = If(A=ScalarParameter() {parameterName=Use Sphere Mask}, "A > B"=@n4, "A == B"=@n4, "A < B"=Constant() {r=1}) {constB=0.99}
  n6 = TextureSample() {samplerType=SAMPLERTYPE_LinearColor, texture=&T_SmokeFlipbook_8x8}
  n7 = Add(
   A=MaterialFunctionCall(Float3=Multiply(A=@n5, B=@n6.G)) {materialFunction=&BreakOutFloat3Components}.R
   B=ScalarParameter() {group=MistA, parameterName=MistA_GrayAdjust_Add}
  )
  n8 = ScalarParameter() {defaultValue=1, group=MistA, parameterName=MistA_GrayAdjust_Mul}
  n9 = Power(Base=Saturate(None=Multiply(A=@n7, B=@n8)), Exp=ScalarParameter() {defaultValue=1.5, group=MistA, parameterName=MistA_GrayAdjust_Power})
  Multiply(
   A=@n1.A
   B=DepthFade(
    Opacity=Saturate(None=@n9)
    FadeDistance=DynamicParameter() {defaultValue=<rgba 200 1 1 1>, paramNames [DepthFade,Param2,Param3,Param4]}.DepthFade
   ) {fadeDistanceDefault=200}
  )
 ]
}
