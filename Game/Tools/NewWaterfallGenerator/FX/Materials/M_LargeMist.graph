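# Translucent lit Niagara sprite material for large waterfall mist from the waterfall generator tool.
# A time-driven flipbook, offset by a per-particle random value, blends adjacent frames unless DisableBlend is on, then remaps through a curve atlas into particle-tinted emissive.
# Opacity combines the flipbook with edge masks, a vertical gradient, particle alpha and depth fade.
# A screen-aligned, camera-distance-faded dust texture boosts opacity only at Default and Epic quality.
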Material /Game/Tools/NewWaterfallGenerator/FX/Materials/M_LargeMist
graph {
 properties {bUsedWithNiagaraSprites=true, blendMode=BLEND_Translucent, pixelDepthOffsetMode=PDOM_Legacy, usedShadingModels=Default Lit}
 nodes [
  n1 = VectorParameter() {parameterName=SubImages, defaultValue=<rgba 9 12 10 1>}
  n2 = Multiply(
   A=@n1.B
   B=Add(A=Time() {bOverride_Period=true, period=1888}, B=DynamicParameter() {paramNames [RandomOffset,Offset,Param3,Param4]}.RandomOffset)
  )
  n3 = Frac(None=@n2) {desc=Frame Lerp}
  n4 = Subtract(A=@n2, B=@n3) {desc=IndexA}
  n5 = MaterialFunctionCall(R=@n1.R, G=@n1.R) {materialFunction=&AppendMany}
  n6 = ComponentMask(None=@n5.RG) {desc="## Horz Images", r=true}
  n7 = Divide(A=Constant2Vector() {g=1, r=1}, B=@n5.RG) {desc=Scale}
  n8 = ComponentMask(None=@n7) {desc="1 / ## Horz Images", r=true}
  n9 = Multiply(A=@n4, B=@n8)
  n10 = TextureCoordinate()
  n11 = Add(A=AppendVector(A=Fmod(A=@n4, B=@n6) {desc=IndexA_Horz}, B=Floor(None=@n9) {desc=IndexA_Vert}) {desc=IndexA_Combined}, B=@n10)
  n12 = TextureObjectParameter() {parameterName=Param, texture=/Game/Tools/NewWaterfallGenerator/FX/Texture/T_LargeMist_5x8}
  n13 = TextureSample(UVs=Frac(None=Multiply(A=@n11, B=@n7) {desc=UV_A}), Tex=@n12) {samplerType=SAMPLERTYPE_LinearColor, texture=/Engine/EngineMaterials/DefaultDiffuse}
  n14 = Add(B=@n4) {constA=1, desc=IndexB}
  n15 = Multiply(A=@n14, B=@n8)
  n16 = Add(A=@n10, B=AppendVector(A=Fmod(A=@n14, B=@n6) {desc=IndexB_Horz}, B=Floor(None=@n15) {desc=IndexB_Vert}) {desc=IndexB_Combined})
  n17 = TextureSample(UVs=Frac(None=Multiply(A=@n16, B=@n7) {desc=UV_B}), Tex=@n12) {samplerType=SAMPLERTYPE_LinearColor, texture=&DefaultDiffuse}
  n18 = MaterialFunctionCall(Float3=StaticSwitchParameter(True=@n13.RGB, False=LinearInterpolate(A=@n13.RGB, B=@n17.RGB, Alpha=@n3)) {parameterName=DisableBlend}) {materialFunction=&BreakOutFloat3Components}
  n19 = LinearInterpolate(A=Multiply(A=Add(A=@n18.R)) {constB=0.5}, B=@n18.R, Alpha=ScalarParameter() {parameterName=awaw})
  n20 = CurveAtlasRowParameter(CurveTime=@n19) {atlas=/Game/Tools/NewWaterfallGenerator/FX/Texture/CLA_Waterfall, curve=/Game/Tools/NewWaterfallGenerator/FX/Texture/CL_MistColorRemap}
  n21 = Multiply(A=ScalarParameter() {defaultValue=2, parameterName=ScaleColor}, B=@n20)
  n22 = ParticleColor()
  MaterialFunctionCall(Emissive=Multiply(A=Multiply(A=@n21, B=@n22.RGB), B=ScalarParameter() {parameterName=WaterfallMistEmissive})) {materialFunction=&MF_FX_EmissiveAdjustments}
  n23 = TextureCoordinate()
  n24 = VectorParameter() {group=Edge, parameterName=Edge_VGradient_ControlV, channelNames {a="", b=Power, g=Multiply, r=Offset}, defaultValue=<rgba 0 1 2 1>}
  n25 = Multiply(
   A=Add(
    A=MaterialFunctionCall(Float2=@n23) {materialFunction=&BreakOutFloat2Components}.G
    B=@n24.Offset
   )
   B=@n24.Multiply
  )
  n26 = TextureObjectParameter() {group=RainTexture, parameterName=RainTexture_Texture, sortPriority=0, texture=/Game/Global/Textures/Default/PointSplatter/T_Tiling_Dust}
  n27 = VectorParameter() {group=RainTexture, parameterName=RainTexture_UVControl, channelNames {a=VPan, b=UPan, g=VScale, r=UScale}, defaultValue=<rgba 1 1 0 0>}
  n28 = ParticlePositionWS()
  n29 = Add(
   A=MaterialFunctionCall(
    Texture=@n26
    UVControl_RGB=@n27
    UVControl_A=@n27
    "[Override]UVInput"=MaterialFunctionCall!(Center Point World Position=@n28) {materialFunction=/Engine/Functions/Engine_MaterialFunctions02/UVs/WorldSpaceAlignedScreenCoordinates}.Offset and Scaled UVS
   ) {materialFunction=&FXCommon_TextureSample}.R
   B=ScalarParameter() {group=RainTexture, parameterName=RainTexture_GrayAdjust_Add}
  )
  n30 = ScalarParameter() {defaultValue=1, group=RainTexture, parameterName=RainTexture_GrayAdjust_Mul}
  n31 = Power(
   Base=Saturate(None=Multiply(A=@n29, B=@n30))
   Exp=ScalarParameter() {defaultValue=1, group=RainTexture, parameterName=RainTexture_GrayAdjust_Power}
  )
  n32 = WorldPosition()
  n33 = CameraPositionWS()
  n34 = VectorParameter() {parameterName=camerafade, defaultValue=<rgba 512 4096 1.5 0>}
  n35 = Divide(A=Subtract(A=Distance(A=@n32.XYZ, B=@n33), B=@n34.R), B=@n34.G) {constB=2048}
  n36 = Multiply(
   A=Multiply(A=Saturate(None=@n31), B=VectorParameter() {parameterName=RainBoost, defaultValue=<rgba 1 1 1 1>}.RGB)
   B=Power(Base=Saturate(None=@n35), Exp=@n34.B)
  )
  n37 = Constant()
  n38 = MaterialFunctionCall(Float2=TextureCoordinate()) {materialFunction=&BreakOutFloat2Components}
  n39 = VectorParameter() {group=Edge, parameterName=Edge_EdgeMask_Control, channelNames {a="", b=Power, g=Multiply, r=Offset}, defaultValue=<rgba 0 1 1 1>}
  n40 = VectorParameter() {group=Edge, parameterName=Edge_EdgeMask_ControlV, channelNames {a="", b=Power, g=Multiply, r=Offset}, defaultValue=<rgba 0 1 1 1>}
  n41 = Multiply(
   A=MaterialFunctionCall!(UorV=@n38.R, EdgeControl=@n39) {materialFunction=/Game/Effects/Niagara/MasterMaterials/Functions/Common/FXCommon_EdgeMask_Simple}.Mask
   B=MaterialFunctionCall(UorV=@n38.G, EdgeControl=@n40) {materialFunction=&FXCommon_EdgeMask_Simple}.Mask
  )
  n42 = Add(A=Multiply(A=QualitySwitch(Default=@n36, Low=@n37, High=@n37, Medium=@n37, Epic=@n36), B=@n41), B=@n41)
  n43 = Saturate(None=Multiply(A=@n19, B=ScalarParameter() {parameterName=EdgeM}))
  n44 = ScalarParameter() {parameterName=EdgeP}
  n45 = ScalarParameter() {parameterName=arw}
  n46 = DepthFade(
   Opacity=Multiply(A=Power(Base=@n43, Exp=@n44), B=Power(Base=@n19, Exp=@n45))
   FadeDistance=ScalarParameter() {defaultValue=300, parameterName=DF}
  )
  Multiply(A=Power(Base=Saturate(None=@n25), Exp=@n24.Power), B=Multiply(A=@n22.A, B=Multiply(A=@n42, B=@n46)))
  VectorParameter() {parameterName=SubImage, defaultValue=<rgba 0 0 0 1>}
 ]
}
