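# Lit translucent Niagara sprite material for splash and mist effects in the waterfall generator.
# A powered texture red channel, tinted by particle color, feeds emissive adjustments and opacity, which gets depth fade and optional camera-distance fade.
# A world-position offset pulls vertices toward particle center by camera distance to scale sprites.
# A dithered threshold mask clips below BottomPos height, and a UseCloth switch substitutes flat opacity and scaled color.
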Material /Game/Tools/NewWaterfallGenerator/FX/Materials/M_SplashFlipbook
graph {
 properties {
  bEnableResponsiveAA=true
  bGenerateSphericalParticleNormals=true
  bUsedWithNiagaraSprites=true
  blendMode=BLEND_Translucent
  pixelDepthOffsetMode=PDOM_Legacy
  translucencyLightingMode=TLM_VolumetricDirectional
  usedShadingModels=Default Lit
 }
 nodes [
  n1 = ParticleColor()
  n2 = TextureSampleParameter2D() {parameterName=Param}
  n3 = ScalarParameter() {parameterName=PO}
  n4 = Multiply(A=@n1.A, B=Multiply(A=Power(Base=@n2.R, Exp=@n3), B=@n2.A))
  n5 = Subtract(A=WorldPosition().Z, B=VectorParameter() {parameterName=BottomPos, defaultValue=<rgba 0 0 -100000000 1>}.B)
  n6 = ScalarParameter() {defaultValue=0.5, parameterName=MaskScale}
  n7 = Constant()
  If(
   A=Multiply(A=@n4, B=Saturate(None=@n5))
   B=ScalarParameter() {defaultValue=0.333, parameterName=MaskTV}
   "A > B"=MaterialFunctionCall(Alpha Threshold=Multiply(A=@n4, B=@n6)) {materialFunction=&DitherTemporalAA}.Result
   "A == B"=@n7
   "A < B"=@n7
  )
  LinearInterpolate(B=NamedRerouteUsage(), Alpha=ScalarParameter() {parameterName=ScaleParticleAlphaAndSizeByCameraDistance})
  n8 = Multiply(A=Power(Base=@n2.R, Exp=ScalarParameter() {defaultValue=0.2, parameterName=P}), B=@n1.RGB)
  n9 = Multiply(A=ScalarParameter() {parameterName=ScaleBaseColor}, B=@n8)
  MaterialFunctionCall(Emissive=Multiply(A=StaticSwitchParameter(True=@n9, False=@n8) {parameterName=UseCloth}, B=ScalarParameter() {parameterName=WaterfallMistEmissive})) {materialFunction=&MF_FX_EmissiveAdjustments}
  NamedRerouteDeclaration(None=ScalarParameter() {defaultValue=650, parameterName=CameraDistanceFade}) {name=CameraDistanceFade, nodeColor=<rgba 1 0.1406 0 1>}
  n10 = Subtract(A=WorldPosition().XYZ, B=ParticlePositionWS())
  n11 = Subtract(A=NamedRerouteUsage(), B=ScalarParameter() {defaultValue=500, parameterName=ScaleDownAdd})
  NamedRerouteDeclaration(None=LinearInterpolate(
    A=Multiply(A=@n10) {constB=-1}
    Alpha=MaterialFunctionCall!(Distance=@n11) {materialFunction=/Game/Effects/Fort_Effects/Materials/_Functions/MF_CameraDistanceFade}.Result
   )) {name=WPO, nodeColor=<rgba 1 0 0.8203 1>}
  n12 = Multiply(
   A=@n4
   B=MaterialFunctionCall(Distance=NamedRerouteUsage()) {materialFunction=&MF_CameraDistanceFade}.Result
  )
  n13 = ScalarParameter() {parameterName=ScaleParticleAlphaAndSizeByCameraDistance}
  StaticSwitchParameter(
   True=ScalarParameter() {parameterName=OpacityCloth}
   False=DepthFade(Opacity=LinearInterpolate(A=@n4, B=@n12, Alpha=@n13), FadeDistance=ScalarParameter() {defaultValue=50, parameterName=DF})
  ) {parameterName=UseCloth}
 ]
}
