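# Masked, lit Niagara sprite material for water splash flipbooks in the waterfall generator tool, with responsive AA enabled.
# Particle color tints a power-adjusted texture green channel and an optional scalar emissive term.
# The mask multiplies particle alpha, texture alpha and green raised to dynamic MaskErode, scaled by MaskScale, without dithering.
# A second dynamic parameter flattens the splash normal map per particle.
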Material /Game/Tools/NewWaterfallGenerator/FX/Materials/M_SplashFlipbook_Lit_NoDither
graph {
 properties {
  bEnableResponsiveAA=true
  bUsedWithNiagaraSprites=true
  blendMode=BLEND_Masked
  pixelDepthOffsetMode=PDOM_Legacy
  usedShadingModels=Default Lit
 }
 nodes [
  MaterialFunctionCall() {materialFunction=&FlattenNormal}
  MaterialFunctionCall(
   Normal=TextureSample() {samplerType=SAMPLERTYPE_Normal, texture=/Game/Effects/Fort_Effects/Textures/Water/T_waterSplash_R_N}.RGB
   Flatness=DynamicParameter() {paramNames [MaskErode, FlattenNormal, Param3, Param4]}.FlattenNormal
  ) {materialFunction=&FlattenNormal}
  n1 = ParticleColor()
  n2 = TextureSampleParameter2D() {parameterName=Param}
  Multiply(A=@n1.RGB, B=Power(Base=@n2.G, Exp=ScalarParameter() {defaultValue=0.2, parameterName=FlipBook_Power}))
  n3 = Power(Base=@n2.G, Exp=DynamicParameter() {paramNames [MaskErode, FlattenNormal, Param3, Param4]}.MaskErode)
  Multiply(A=Multiply(A=@n1.A, B=Multiply(A=@n3, B=@n2.A)), B=ScalarParameter() {defaultValue=0.5, parameterName=MaskScale})
  Multiply(
   A=@n1.RGB
   B=LinearInterpolate(B=ScalarParameter() {defaultValue=0.1, parameterName=EmissiveColor}, Alpha=ScalarParameter() {parameterName=UseEmissive})
  )
  ScalarParameter() {parameterName=Specular}
  ScalarParameter() {parameterName=Roughness}
  ScalarParameter() {parameterName=Metallic}
 ]
}
