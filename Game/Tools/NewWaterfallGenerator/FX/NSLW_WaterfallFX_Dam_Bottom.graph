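# Niagara system for the bottom of a dam waterfall in the waterfall generator tool, named for splash and smoke effects at the base.
# It holds nine emitters: three Splash, three Smoke and three flat variants, with some named for left, middle or right placement.
# SystemState is the only listed module, and no user variables, renderers or data interfaces appear.
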NiagaraSystem /Game/Tools/NewWaterfallGenerator/FX/NSLW_WaterfallFX_Dam_Bottom
dependencies {usedDataInterfaces [], usedDynamicInputs [], usedModules [&SystemState]}
emitters {
 Flat_right {rendererClasses [], renderers []}
 Smoke {rendererClasses [], renderers []}
 Smoke001 {rendererClasses [], renderers []}
 Smoke_Right {rendererClasses [], renderers []}
 Splash {rendererClasses [], renderers []}
 Splash001 {rendererClasses [], renderers []}
 Splash002 {rendererClasses [], renderers []}
 Splash_Left_Flaty {rendererClasses [], renderers []}
 Splash_Middle_Flat {rendererClasses [], renderers []}
}
summary {
 systemName=NSLW_WaterfallFX_Dam_Bottom
 emitters [
  {emitterName=Splash, rendererClasses []}
  {emitterName=Smoke, rendererClasses []}
  {emitterName=Splash001, rendererClasses []}
  {emitterName=Smoke001, rendererClasses []}
  {emitterName=Splash002, rendererClasses []}
  {emitterName=Smoke_Right, rendererClasses []}
  {emitterName=Splash_Left_Flaty, rendererClasses []}
  {emitterName=Splash_Middle_Flat, rendererClasses []}
  {emitterName=Flat_right, rendererClasses []}
 ]
 userVariables []
}
