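# Niagara system for the middle piece of a dam waterfall in the NewWaterfallGenerator tool, apparently providing splash and smoke effects.
# Ten emitters pair splash with smoke across left, middle and right, plus an extra SplashLeft001 and three named Empty.
# SystemState is the only module listed; no renderers, data interfaces or user variables appear, so emitter behavior is not visible here.
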NiagaraSystem /Game/Tools/NewWaterfallGenerator/FX/NSLW_WaterfallFX_Dam_Middle
dependencies {usedDataInterfaces [], usedDynamicInputs [], usedModules [&SystemState]}
emitters {
 Empty {rendererClasses [], renderers []}
 Empty001 {rendererClasses [], renderers []}
 Empty002 {rendererClasses [], renderers []}
 SmokeLeft {rendererClasses [], renderers []}
 SmokeMiddle {rendererClasses [], renderers []}
 SmokeRight {rendererClasses [], renderers []}
 Splash {rendererClasses [], renderers []}
 SplashLeft {rendererClasses [], renderers []}
 SplashLeft001 {rendererClasses [], renderers []}
 SplashRight {rendererClasses [], renderers []}
}
summary {
 systemName=NSLW_WaterfallFX_Dam_Middle
 emitters [
  {emitterName=Splash, rendererClasses []}
  {emitterName=SmokeMiddle, rendererClasses []}
  {emitterName=SplashLeft, rendererClasses []}
  {emitterName=SmokeLeft, rendererClasses []}
  {emitterName=SplashRight, rendererClasses []}
  {emitterName=SmokeRight, rendererClasses []}
  {emitterName=Empty001, rendererClasses []}
  {emitterName=Empty, rendererClasses []}
  {emitterName=Empty002, rendererClasses []}
  {emitterName=SplashLeft001, rendererClasses []}
 ]
 userVariables []
}
