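# Niagara effect system in the waterfall generator tool, named for the inside top of a dam waterfall.
# It holds two emitters, Splash and Smoke, which by name supply water spray and mist.
# The only listed module is SystemState, with no user parameters, data interfaces or dynamic inputs.
# Neither emitter lists any renderers, so their particle behavior and look are not visible here.
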NiagaraSystem /Game/Tools/NewWaterfallGenerator/FX/NSLW_WaterfallFX_Dam_Top_Inside
dependencies {usedDataInterfaces [], usedDynamicInputs [], usedModules [&SystemState]}
emitters {Smoke {rendererClasses [], renderers []}, Splash {rendererClasses [], renderers []}}
summary {
 systemName=NSLW_WaterfallFX_Dam_Top_Inside
 emitters [{emitterName=Splash, rendererClasses []}, {emitterName=Smoke, rendererClasses []}]
 userVariables []
}
