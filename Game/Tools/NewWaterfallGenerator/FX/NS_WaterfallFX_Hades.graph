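# Niagara system for the Hades variant of generated waterfall effects, layering splashes, drops and smoke along the fall.
# User parameters for top and bottom positions, width and thickness drive placement, with some spawn rates scaled by width.
# GPU splash and drop sprites take red tints from time-of-day curves, alongside smoke sprites and side-smoke ribbons strung top to bottom.
# Camera distance and angle drive fades and culling; several top and large-smoke emitters are disabled.
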NiagaraSystem /Game/Tools/NewWaterfallGenerator/FX/NS_WaterfallFX_Hades
dependencies {
 usedDataInterfaces [
  &NiagaraDataInterfaceCurve
  &NiagaraDataInterfaceCamera
  &NiagaraDataInterfacePlatformSet
  &NiagaraDataInterfaceVectorCurve
  &NiagaraDataInterfaceSpriteRendererInfo
  &NiagaraDataInterfaceColorCurve
  &NiagaraDataInterfaceVector2DCurve
 ]
 usedDynamicInputs [
  &Vector2DFromFloat
  &self:SelectFloatDependOnPlatForm
  &Multiply_Vector2D_ByFloat
 ]
 usedModules [
  &self:SystemSpawnScript.SystemScriptSource.SystemScriptGraph.NiagaraNodeAssignment_3.SetVariables_#1
  &SystemState
  &self:SystemSpawnScript.SystemScriptSource.SystemScriptGraph.NiagaraNodeAssignment_0.SetVariables_#2
  &self:SystemSpawnScript.SystemScriptSource.SystemScriptGraph.NiagaraNodeAssignment_2.SetVariables_#3
  &self:SystemSpawnScript.SystemScriptSource.SystemScriptGraph.NiagaraNodeAssignment_1.SetVariables_#4
 ]
}
emitters {
 BottomSplash {
  simTarget=GPUComputeSim
  emitterUpdateScript [
   &EmitterState(
    Enable Distance Culling=true
    Life Cycle Mode=::NewEnumerator1
    Loop Duration=1
    MaxDistance=10000
    Module.bUseMaxDistance=true
    Module.bUseMinDistance=false
    Scalability Mode=::NewEnumerator1
    UseLoopDelay=false
   )
   &SpawnRate(
    SpawnRate=&Multiply_Float(
     A=&Multiply_Float(
      A=250
      B=&FloatFromCurve(
       FloatCurve=DataInterface() {propertyValues {Curve {keys [<at 0.7 1>,<at 1.1 1>]}}}
       CurveIndex=&Divide_Float(
        A=$Engine.Owner.LODDistance
        B=&MakeCustomFloatFromBool(
         Bool=&SetBoolByFloatComparison(A=$User.BPInput_BottomSmokeDisableDistance, B=0, Comparison Type=::NewEnumerator4)
         True Float=1
         False Float=$User.BPInput_BottomSmokeDisableDistance
        )
       )
      )
     )
     B=&Divide_Float(A=$User.FallWidth, B=1500)
    )
    Use Spawn Probability=false
   )
   &SpawnRate(
    SpawnGroup=1
    SpawnRate=&Multiply_Float(
     A=&Multiply_Float(
      A=25
      B=&FloatFromCurve(
       FloatCurve=DataInterface() {propertyValues {Curve {keys [<at 0.7 1>,<at 1.1 0>]}}}
       CurveIndex=&Divide_Float(
        A=$Engine.Owner.LODDistance
        B=&MakeCustomFloatFromBool(
         Bool=&SetBoolByFloatComparison(A=$User.BPInput_BottomSmokeDisableDistance, B=0, Comparison Type=::NewEnumerator4)
         True Float=1
         False Float=$User.BPInput_BottomSmokeDisableDistance
        )
       )
      )
     )
     B=&Divide_Float(A=$User.FallWidth, B=1500)
    )
    Use Spawn Probability=false
   ) {module=SpawnRate001}
  ]
  particleSpawnScript [
   &InitializeParticle(
    Color=$User.BaseFXColor
    Lifetime Max=1.2
    Lifetime Mode=::NewEnumerator1
    Sprite Rotation Mode=::NewEnumerator2
    Sprite Size Mode=::NewEnumerator1
    Uniform Sprite Size Max=300
    Uniform Sprite Size Min=150
    UsePositionOffset=false
    Write Material Random=true
   )
   &ShapeLocation(
    Box Size=&MakeVector(X=$User.FallWidth, Y=50, Z=0)
    MaskSpawnGroup=false
    Offset=&MakeVector(
     X=0
     Y=&Max_Float(
      A:NiagaraFloat=&Subtract_Float(A=&Divide_Float(A=$User.FallThickness, B=2), B=20)
      B:NiagaraFloat=0
     )
     Z=&MaskFloatBySpawnGroup(Masked Float=10, Passthrough Float=50)
    )
    Offset Coordinate Space=::Local
    Offset Mode=::NewEnumerator0
    Override Seed=false
    Shape Origin=$User.BottomPosition
    Shape Primitive=::NewEnumerator2
    Surface Only=false
    UseOwnerScale=true
   )
   &AddVelocity(
    Cone Angle=60
    Cone Axis=<0 0 1>
    Module.UseConeFalloff=false
    Module.UseVelDistribution=false
    Override Seed=false
    Velocity Mode=::NewEnumerator2
   )
   /SetVariables(Particles.VisibilityTag=&SpawnGroup()) {isSetParameters=true}
  ]
  particleUpdateScript [
   &ParticleState()
   &RegionMask(Region Dimensions:Vector3f=<500 5000 1900>, Region Type:ENiagaraDebugDrawShape=::NewEnumerator1, UseOffset=false, UseRotation=false)
   &ScaleSpriteSize(
    Uniform Curve Scale=$User.ScaleBottomMainSplash
    Uniform Curve Sprite Scale=DataInterface() {propertyValues {Curve {keys [
        {interpMode=RCIM_Cubic, tangentMode=RCTM_User, time=0, value=0}
        {interpMode=RCIM_Cubic, tangentMode=RCTM_User, time=0.5, value=1}
        {tangentMode=RCTM_User, time=1, value=0}
       ]}}}
   )
   &SubUVAnimation(
    Evaluation Type=::NewEnumerator0
    "Sprite Renderer "=DataInterface() {propertyValues {SpriteRenderer=&self:BottomSplash_5.NiagaraSpriteRendererProperties_0}}
    SubUV Animation Mode:ENiagara_SubUVLookupModeV2=::NewEnumerator2
    UseEndFrame=false
    UseStartFrame=false
   )
   &ScaleColor(
    Scale Alpha=&FloatFromCurve(FloatCurve=DataInterface() {propertyValues {Curve {keys [
         {interpMode=RCIM_Cubic, tangentMode=RCTM_Break, time=0, value=0}
         {interpMode=RCIM_Cubic, tangentMode=RCTM_Break, time=0.2, value=1}
         {arriveTangent=-0.9842, interpMode=RCIM_Cubic, leaveTangent=-0.9842, tangentMode=RCTM_User, time=0.6, value=0}
        ]}}})
    Scale RGB=&MakeVectorFromLinearColorRGB(Color=&ColorFromCurve(ColorCurve=DataInterface() {propertyValues {
        AlphaCurve {keys [<at 0 1 RCIM_Cubic>,<at 1 1 RCIM_Cubic>]}
        BlueCurve {keys [<at 0 0.4206 RCIM_Cubic>,<at 0.115 0.4206>,<at 1 0 RCIM_Cubic>]}
        GreenCurve {keys [<at 0 1 RCIM_Cubic>,<at 0.115 1>,<at 1 0.521 RCIM_Cubic>]}
        RedCurve {keys [<at 0 0 RCIM_Cubic>,<at 0.115 0>,<at 1 0.0941 RCIM_Cubic>]}
       }}))
    ScaleA=true
    ScaleRGB=true
   ) {enabled=false}
   &ScaleColor(
    Scale Alpha=&FloatFromCurve(FloatCurve=DataInterface() {propertyValues {Curve {keys [
         {interpMode=RCIM_Cubic, tangentMode=RCTM_Break, time=0, value=0}
         {interpMode=RCIM_Cubic, tangentMode=RCTM_Break, time=0.2, value=1}
         {arriveTangent=-0.9842, interpMode=RCIM_Cubic, leaveTangent=-0.9842, tangentMode=RCTM_User, time=0.6, value=0}
        ]}}})
    Scale RGB=&MakeVectorFromLinearColorRGB(Color=&ColorFromCurve(
      ColorCurve=DataInterface() {propertyValues {
        AlphaCurve {keys [<at 0 1 RCIM_Cubic>]}
        BlueCurve {keys [<at 0 0.0596 RCIM_Cubic>,<at 0.1744 0.122>,<at 0.7931 0.122>,<at 1 0.0596>]}
        GreenCurve {keys [<at 0 0 RCIM_Cubic>,<at 0.1744 0>,<at 0.7931 0>,<at 1 0>]}
        RedCurve {keys [<at 0 1 RCIM_Cubic>,<at 0.1744 1>,<at 0.7931 1>,<at 1 1>]}
       }}
      CurveIndex=Linked() {linkedVariable {name=NPC.FortniteNPC.TimeOfDay_0-1, type=NiagaraFloat}}
     ))
    ScaleA=true
    ScaleRGB=true
   ) {module=ScaleColor001}
   &GravityForce(Gravity=<0 0 -800>)
   &SolveForcesAndVelocity(Clamp Velocity=false, Limit Acceleration=false)
   &CameraOffset(Camera Offset Amount=100)
   &DynamicMaterialParameters(
    Index 0 Param 1=&FloatFromCurve(FloatCurve=DataInterface() {propertyValues {Curve {keys [
         {interpMode=RCIM_Cubic, tangentMode=RCTM_User, time=0, value=0.5}
         {arriveTangent=0.8093, interpMode=RCIM_Cubic, leaveTangent=0.8093, tangentMode=RCTM_User, time=0.6, value=1.025}
         {interpMode=RCIM_Cubic, tangentMode=RCTM_User, time=1, value=1.25}
        ]}}})
    Module.Index0Param0Write=true
    Module.Index0Param1Write=true
    Module.Index0Param2Write=true
    Module.Index0Param3Write=true
   )
  ]
  rendererClasses [&NiagaraSpriteRendererProperties]
  renderers [{rendererClass=&NiagaraSpriteRendererProperties}]
 }
 CenterSmoke {
  emitterUpdateScript [
   &EmitterState(
    Enable Distance Culling=true
    Life Cycle Mode=::NewEnumerator1
    Loop Duration=1
    MaxDistance=10000
    Module.bUseMaxDistance=true
    Module.bUseMinDistance=false
    Scalability Mode=::NewEnumerator1
    UseLoopDelay=false
   )
   &SpawnRate(
    SpawnRate=&Multiply_Float(A=1.5, B=&Divide_Float(A=$User.FallWidth, B=1500))
    Use Spawn Probability=false
   )
  ]
  particleSpawnScript [
   &InitializeParticle(
    Color=$User.BaseFXColor
    Lifetime=3
    Position Mode=::NewEnumerator0
    Sprite Rotation Mode=::NewEnumerator2
    Sprite Size Mode=::NewEnumerator1
    Uniform Sprite Size Max=1357.6738
    Uniform Sprite Size Min=1052.9229
    Write Material Random=true
   )
   &ShapeLocation(
    Box Size=&MakeVector(X=$User.FallWidth, Y=100, Z=0)
    MaskSpawnGroup=false
    Override Seed=false
    Shape Origin=&TransformPosition(Position=$User.BottomPosition, Source Space=::World, Destination Space=::Local)
    Shape Primitive=::NewEnumerator2
    Surface Only=false
    UseOwnerScale=true
   )
   &AddVelocity(Velocity=<0 500 0>)
   /SetVariables(Particles.Velocity=&Multiply_Vector(A=$Particles.Velocity, B=<1 1 0>)) {isSetParameters=true}
  ]
  particleUpdateScript [
   &ParticleState()
   &ScaleSpriteSize(Scale Sprite Size Mode=::NewEnumerator0, Uniform Scale Factor=$User.ScaleBottomMediumSmoke)
   &ScaleColor(
    Scale Alpha=&FloatFromCurve(
     FloatCurve=DataInterface() {propertyValues {Curve {keys [
         {interpMode=RCIM_Cubic, tangentMode=RCTM_User, time=0, value=0}
         {interpMode=RCIM_Cubic, tangentMode=RCTM_User, time=0.5, value=1}
         {tangentMode=RCTM_User, time=1, value=0}
        ]}}}
     Scale Curve=1.2
    )
    ScaleA=true
    ScaleRGB=true
   )
   &GravityForce(Gravity=&VectorFromCurve(
     VectorCurve=DataInterface() {propertyValues {
       XCurve {keys [<at 0 1>,<at 1 0>]}
       YCurve {keys [<at 0 1>,<at 1 0>]}
       ZCurve {keys [{interpMode=RCIM_Cubic, leaveTangent=2.1548, tangentMode=RCTM_Break, time=0, value=0}, <at 1 1>]}
      }}
     Scale Curve=<0 0 120>
    ))
   &Drag(Drag=0.5, Use Linear Drag=true, Use Rotational Drag=true)
   &CurlNoiseForce(Noise Frequency=1, Noise Quality / Cost=::NewEnumerator0, Noise Strength=100, Pan Noise=true, Pan Noise Field=<0 0 0.3>)
   &SolveForcesAndVelocity(Clamp Velocity=false, Limit Acceleration=false)
   &CameraOffset(Camera Offset Amount=&Multiply_Float(
     A=&RandomRangeFloat(Minimum=300, Maximum=700)
     B=&FloatFromCurve(FloatCurve=DataInterface() {propertyValues {Curve {keys [<at 0 0>,<at 4096 1>]}}}, CurveIndex=$Engine.Owner.LODDistance)
    ))
   &SubUVAnimation(
    "Sprite Renderer "=DataInterface() {propertyValues {SpriteRenderer=&self:CenterSmoke_2.NiagaraSpriteRendererProperties_0}}
    UseEndFrame=false
    UseStartFrame=false
   )
   &ScaleColor(
    Scale Alpha=&FloatFromCurve(
     FloatCurve=DataInterface() {propertyValues {Curve {keys [<at 0.5 0>,<at 1 1>]}}}
     CurveIndex=&Divide_Float(
      A=&DistanceBetweenPositions(
       A=$Particles.Position
       B=&ReturnCameraPosition(Coordinate Space:ENiagaraCoordinateSpace=::Simulation, Camera Query:NiagaraDataInterfaceCamera=DataInterface())
      )
      B=$User.BPInput_BottomSmokeDisableDistance
     )
    )
    ScaleA=true
    ScaleRGB=true
   ) {module=ScaleColor002}
   &ScaleSpriteSize(
    Uniform Curve Index=&MakeFloatFromLinearColor(LinearColor=$Particles.Color, Channel:ENiagaraLinearColor_Channels=::NewEnumerator3)
    Uniform Curve Sprite Scale=DataInterface() {propertyValues {Curve {keys [{tangentMode=RCTM_User, time=0, value=0}, {tangentMode=RCTM_User, time=0.01, value=1}]}}}
   ) {module=ScaleSpriteSize003}
   &ScaleColor(Scale Alpha=$User.Opacity_BottomSmoke, ScaleA=true, ScaleRGB=true) {module=ScaleColor003}
  ]
  rendererClasses [&NiagaraSpriteRendererProperties]
  renderers [{rendererClass=&NiagaraSpriteRendererProperties}]
 }
 DropSmokeA {
  emitterUpdateScript [
   &EmitterState(
    Enable Distance Culling=true
    Life Cycle Mode=::NewEnumerator1
    Loop Duration=1
    MaxDistance=7500
    Module.bUseMaxDistance=true
    Module.bUseMinDistance=false
    Scalability Mode=::NewEnumerator1
    UseLoopDelay=false
   )
   &SpawnRate(SpawnRate=&Multiply_Float(A=1.2, B=$User.ScaleDropSmokeSpawnRate), Use Spawn Probability=false)
   &SpawnRate(SpawnGroup=1, SpawnRate=&Multiply_Float(A=1, B=$User.ScaleDropSmokeSpawnRate), Use Spawn Probability=false) {module=SpawnRate001}
  ]
  particleSpawnScript [
   &InitializeParticle(
    Color=$User.BaseFXColor
    Lifetime=&Add_Float(A=&RandomRangeFloat(Minimum=3, Maximum=4), B=$User.DropSmokeAgeOffset)
    Sprite Size Mode=::NewEnumerator1
    Uniform Sprite Size Max=1600
    Uniform Sprite Size Min=1200
    UsePositionOffset=false
    Write Material Random=true
   )
   &ShapeLocation(
    Box Midpoint=<0.5 0.5 0.45>
    Box Size=<0 100 650>
    MaskSpawnGroup=false
    Offset=&AddVectorToPosition(
     Position=<0 50 0>
     Vector=&MakeVector(
      X=0
      Y=&Max_Float(
       A:NiagaraFloat=&Subtract_Float(A=&Divide_Float(A=$User.FallThickness, B=2), B=100)
       B:NiagaraFloat=0
      )
      Z=0
     )
    )
    Offset Mode=::NewEnumerator0
    Override Seed=false
    Shape Origin=&TransformPosition(Position=$User.TopPosition, Source Space=::World, Destination Space=::Local)
    Shape Primitive=::NewEnumerator2
    Surface Only=false
    UseOwnerScale=true
   )
   &AddVelocity(Velocity=&RandomRangeVector(Minimum=<0 90.4105 -1.5>, Maximum=<0 177.0799 -1.5>))
   /SetVariables(Particles.Position=&AddVectorToPosition(
     Position=$Particles.Position
     Vector=&Lerp_Vector(
      A=&Multiply_VectorByFloat(
       Vector=&MakeVector(X=&Subtract_Float(A=&Divide_Float(A=$User.FallWidth, B=2), B=100), Y=0, Z=0)
       Float=&Lerp_Float(
        A=-1
        B=1
        Alpha=&MakeFloatFromInt(Int=&RandomRangeInt(Minimum=0, Maximum=1))
       )
      )
      B=<0 0 0>
      Alpha=&FloatFromCurve(
       FloatCurve=DataInterface() {propertyValues {Curve {keys [
           {interpMode=RCIM_Cubic, leaveTangent=0.2575, tangentMode=RCTM_Break, time=0.5, value=0}
           {arriveTangent=2.0275, interpMode=RCIM_Cubic, tangentMode=RCTM_Break, time=1, value=1}
          ]}}}
       CurveIndex=&RandomRangeFloat(Minimum=0, Maximum=1)
      )
     )
    )) {isSetParameters=true}
   /SetVariables(Particles.VisibilityTag=&SpawnGroup()) {isSetParameters=true}
  ]
  particleUpdateScript [
   &ParticleState()
   &ScaleSpriteSize(
    Scale Factor=&Multiply_Vector2D(
     A=&MaskVector2dBySpawnGroup(Masked Vector2D=<0.4794 1.7>, Passthrough Vector2D=<0.8 1.6>, Spawn Group Mask=0)
     B=&MakeVector2D(
      X=&MakeCustomFloatFromBool(Bool=&RandomBool(), True Float=1, False Float=-1)
      Y=1
     )
    )
    Scale Sprite Size Mode=::NewEnumerator2
   ) {module=ScaleSpriteSize001}
   &ScaleSpriteSize(
    Uniform Curve Scale=$User.ScaleDropSmoke
    Uniform Curve Sprite Scale=DataInterface() {propertyValues {Curve {keys [
        {arriveTangent=19.6806, interpMode=RCIM_Cubic, leaveTangent=19.6806, tangentMode=RCTM_User, time=0, value=0}
        {arriveTangent=0.7853, interpMode=RCIM_Cubic, leaveTangent=0.7853, tangentMode=RCTM_User, time=0.1, value=0.8}
        {arriveTangent=0.0224, interpMode=RCIM_Cubic, leaveTangent=0.0224, tangentMode=RCTM_User, time=1, value=1.35}
       ]}}}
   )
   &CameraOffset(Camera Offset Amount=100)
   &SpriteFacingAndAlignment(Sprite Alignment=<0 0 1>, Sprite Facing=<0 0 1>, WriteAlignment=true, WriteFacing=false)
   &ScaleColor(
    Scale Alpha=&FloatFromCurve(
     FloatCurve=DataInterface() {propertyValues {Curve {keys [
         {arriveTangent=-0.0941, interpMode=RCIM_Cubic, leaveTangent=-0.0941, tangentMode=RCTM_User, time=0, value=0}
         {arriveTangent=1.5029, interpMode=RCIM_Cubic, leaveTangent=1.5029, tangentMode=RCTM_User, time=0.5, value=0.95}
         {interpMode=RCIM_Cubic, tangentMode=RCTM_Break, time=0.6, value=1}
         {arriveTangent=-0.2938, interpMode=RCIM_Cubic, leaveTangent=-6.3411, tangentMode=RCTM_Break, time=0.7, value=0}
        ]}}}
     Scale Curve=0.9
    )
    ScaleA=true
    ScaleRGB=true
   )
   &GravityForce(Gravity=&RandomRangeVector(Minimum=<0 0 -5.1996>, Maximum=<1 1 -11.8155>))
   &SolveForcesAndVelocity(Clamp Velocity=false, Limit Acceleration=false)
   &DynamicMaterialParameters(
    Index 0 Param 1=&FloatFromCurve(FloatCurve=DataInterface() {propertyValues {Curve {keys [<at 0 0>,<at 1 66>]}}})
    Module.Param0WriteEnabled=true
    Module.Param1WriteEnabled=true
    Module.Param2WriteEnabled=true
    Module.Param3WriteEnabled=true
   )
   &KillParticles(Kill Particles=&SetBoolByFloatComparison(A=$Particles.NormalizedAge, B=0.7))
   &ScaleColor(Scale Alpha=$User.Opacity_DropSmoke, ScaleA=true, ScaleRGB=true) {module=ScaleColor003}
  ]
  rendererClasses [&NiagaraSpriteRendererProperties]
  renderers [
   {rendererClass=&NiagaraSpriteRendererProperties}
   {rendererClass=&NiagaraSpriteRendererProperties, rendererIndex=1}
   {rendererClass=&NiagaraSpriteRendererProperties, rendererIndex=2}
   {rendererClass=&NiagaraSpriteRendererProperties, rendererIndex=3}
  ]
 }
 Drops {
  simTarget=GPUComputeSim
  emitterUpdateScript [
   &EmitterState()
   &SpawnRate(SpawnRate=2500, Use Spawn Probability=true)
  ]
  particleSpawnScript [
   &InitializeParticle(
    Color=$User.BaseFXColor
    Lifetime Min=0.3
    Lifetime Mode=::NewEnumerator1
    Sprite Rotation Mode=::NewEnumerator0
    Sprite Size Mode=::NewEnumerator3
    Uniform Sprite Size=&RandomRangeFloat(Minimum=25, Maximum=50)
    UsePositionOffset=false
    Write Material Random=true
   )
   &ShapeLocation(
    Box Size=&MakeVector(X=$User.FallWidth, Y=50, Z=0)
    MaskSpawnGroup=false
    Offset=&MakeVector(
     X=0
     Y=&Max_Float(
      A:NiagaraFloat=&Subtract_Float(A=&Divide_Float(A=$User.FallThickness, B=2), B=20)
      B:NiagaraFloat=0
     )
     Z=&MaskFloatBySpawnGroup(Masked Float=10, Passthrough Float=50)
    )
    Offset Coordinate Space=::Local
    Offset Mode=::NewEnumerator0
    Override Seed=false
    Shape Origin=$User.BottomPosition
    Shape Primitive=::NewEnumerator2
    Surface Only=false
    UseOwnerScale=true
   )
   &AddVelocity(Velocity=&RandomRangeVector(Minimum=<0 50 0>, Maximum=<1 250 1>))
   &SubUVAnimation(End Frame=3, SubUV Animation Mode:ENiagara_SubUVLookupMode=::NewEnumerator2)
   &AddVelocity(
    Velocity=$Particles.ShapeLocation.ShapeNormal
    Velocity Speed Scale=&RandomRangeFloat(Minimum=200, Maximum=400)
   ) {enabled=false, module=AddVelocity001}
  ]
  particleUpdateScript [
   &ParticleState()
   &SolveForcesAndVelocity(Clamp Velocity=false, Limit Acceleration=false)
   &ScaleColor(
    Scale Alpha=&FloatFromCurve(FloatCurve=DataInterface() {propertyValues {Curve {keys [
         {interpMode=RCIM_Cubic, tangentMode=RCTM_Break, time=0, value=0}
         {interpMode=RCIM_Cubic, tangentMode=RCTM_Break, time=0.1, value=1}
         {arriveTangent=-0.9842, interpMode=RCIM_Cubic, leaveTangent=-0.9842, tangentMode=RCTM_User, time=0.6, value=0}
        ]}}})
    Scale RGB=&MakeVectorFromLinearColorRGB(Color=&ColorFromCurve(ColorCurve=DataInterface() {propertyValues {
        AlphaCurve {keys [<at 0 1 RCIM_Cubic>,<at 1 1 RCIM_Cubic>]}
        BlueCurve {keys [<at 0 0.4206 RCIM_Cubic>,<at 0.115 0.4206>,<at 1 0 RCIM_Cubic>]}
        GreenCurve {keys [<at 0 1 RCIM_Cubic>,<at 0.115 1>,<at 1 0.521 RCIM_Cubic>]}
        RedCurve {keys [<at 0 0 RCIM_Cubic>,<at 0.115 0>,<at 1 0.0941 RCIM_Cubic>]}
       }}))
    ScaleA=true
    ScaleRGB=true
   ) {enabled=false}
   &ScaleColor(
    Scale Alpha=&FloatFromCurve(FloatCurve=DataInterface() {propertyValues {Curve {keys [
         {interpMode=RCIM_Cubic, tangentMode=RCTM_Break, time=0, value=0}
         {interpMode=RCIM_Cubic, tangentMode=RCTM_Break, time=0.1, value=1}
         {arriveTangent=-0.9842, interpMode=RCIM_Cubic, leaveTangent=-0.9842, tangentMode=RCTM_User, time=0.6, value=0}
        ]}}})
    Scale RGB=&MakeVectorFromLinearColorRGB(Color=&ColorFromCurve(
      ColorCurve=DataInterface() {propertyValues {
        AlphaCurve {keys [<at 0 1 RCIM_Cubic>]}
        BlueCurve {keys [<at 0 0.0596 RCIM_Cubic>,<at 0.1744 0.122>,<at 0.7931 0.122>,<at 1 0.0596>]}
        GreenCurve {keys [<at 0 0 RCIM_Cubic>,<at 0.1744 0>,<at 0.7931 0>,<at 1 0>]}
        RedCurve {keys [<at 0 1 RCIM_Cubic>,<at 0.1744 1>,<at 0.7931 1>,<at 1 1>]}
       }}
      CurveIndex=Linked() {linkedVariable {name=NPC.FortniteNPC.TimeOfDay_0-1, type=NiagaraFloat}}
     ))
    ScaleA=true
    ScaleRGB=true
   ) {module=ScaleColor001}
   &ScaleSpriteSize(Uniform Curve Sprite Scale=DataInterface() {propertyValues {Curve {keys [
        {arriveTangent=9.0725, interpMode=RCIM_Cubic, leaveTangent=9.0725, tangentMode=RCTM_User, time=0, value=0}
        {arriveTangent=-2.5249, leaveTangent=-2.5249, time=0.1408, value=0.9847}
        <at 0.2 1>
        {arriveTangent=-0.0405, interpMode=RCIM_Cubic, leaveTangent=-0.0405, tangentMode=RCTM_User, time=1, value=-0.5}
       ]}}})
   &SpriteFacingAndAlignment(Sprite Alignment=<0 1 1>, Sprite Facing=<0 0 1>, WriteAlignment=true, WriteFacing=true)
  ]
  rendererClasses [&NiagaraSpriteRendererProperties]
  renderers [{rendererClass=&NiagaraSpriteRendererProperties}]
 }
 LargeSmoke {
  bEnabled=false
  emitterUpdateScript [
   &EmitterState(Life Cycle Mode=::NewEnumerator1, Loop Behavior=::NewEnumerator1, Loop Duration Mode=::NewEnumerator1, UseLoopDelay=false)
   &SpawnBurst_Instantaneous(Use Spawn Probability=false, UseLoopCountLimit=false)
   &SpawnBurst_Instantaneous(Spawn Group=1, Use Spawn Probability=false, UseLoopCountLimit=false) {module=SpawnBurst_Instantaneous001}
   &SpawnRate(SpawnRate=0.5, Use Spawn Probability=false) {enabled=false}
  ]
  particleSpawnScript [
   &InitializeParticle(
    Color=$User.BaseFXColor
    Lifetime=6
    Sprite Size=$System.LargeSmokeSize
    Sprite Size Mode=::NewEnumerator4
    UsePositionOffset=false
    Write Material Random=true
   )
   /SetVariables(Particles.bIsLRRevert:NiagaraBool=&RandomBool()) {isSetParameters=true}
   &ShapeLocation(
    Box Size=<1 1 1>
    MaskSpawnGroup=false
    Offset=&MakeVector(X=0, Y=&MaskFloatBySpawnGroup(Masked Float=$User.LargeSmokeOffset, Passthrough Float=0), Z=0)
    Offset Mode=::NewEnumerator0
    Override Seed=false
    Shape Origin=&TransformPosition(Position=$User.BottomPosition, Source Space=::World, Destination Space=::Local)
    Shape Primitive=::NewEnumerator2
    Surface Only=false
    UseOwnerScale=true
   )
   /SetVariables(Particles.VisibilityTag=&SpawnGroup()) {isSetParameters=true}
  ]
  particleUpdateScript [
   &ParticleState(
    Kill Particles When Lifetime Has Elapsed=false
    Let Infinitely Lived Particles Die When Emitter Deactivates=true
    Loop Particles Lifetime=true
   )
   &ScaleSpriteSize(Uniform Curve Sprite Scale=DataInterface() {propertyValues {Curve {keys [
        {interpMode=RCIM_Cubic, leaveTangent=0.8147, tangentMode=RCTM_Break, time=0, value=0.7}
        {arriveTangent=0.1425, interpMode=RCIM_Cubic, tangentMode=RCTM_Break, time=1, value=1.15}
       ]}}}) {enabled=false, module=ScaleSpriteSize001}
   &ScaleSpriteSize(
    Scale Factor=&Multiply_Vector2D(
     A=&MakeVector2D(X=&MakeCustomFloatFromBool(Bool=$Particles.bIsLRRevert, True Float=-1, False Float=1), Y=1)
     B=&MakeVector2D(
      X=&RandomRangeFloat(Minimum=1, Maximum=1)
      Y=&RandomRangeFloat(Minimum=1, Maximum=1)
     )
    )
    Scale Sprite Size Mode=::NewEnumerator2
   )
   &CameraOffset(Camera Offset Amount=&Multiply_Float(
     A=750
     B=&FloatFromCurve(FloatCurve=DataInterface() {propertyValues {Curve {keys [<at 0 0>,<at 4096 1>]}}}, CurveIndex=$Engine.Owner.LODDistance)
    ))
   &DynamicMaterialParameters(
    Index 0 Param 1=&RandomRangeFloat(Minimum=0, Maximum=10)
    Index 0 Param 2=$Particles.CameraOffset
    Module.Param0WriteEnabled=true
    Module.Param1WriteEnabled=true
    Module.Param2WriteEnabled=true
    Module.Param3WriteEnabled=true
   )
   &ScaleColor(
    Scale Alpha=&FloatFromCurve(FloatCurve=DataInterface() {propertyValues {Curve {keys [
         {arriveTangent=4.1785, interpMode=RCIM_Cubic, leaveTangent=4.1785, tangentMode=RCTM_User, time=0, value=0}
         {interpMode=RCIM_Cubic, tangentMode=RCTM_Break, time=0.5, value=1}
         {arriveTangent=-4.8978, interpMode=RCIM_Cubic, leaveTangent=0.0224, tangentMode=RCTM_Break, time=1, value=0}
        ]}}})
    ScaleA=true
    ScaleRGB=true
   ) {enabled=false}
   &SpriteFacingAndAlignment(Sprite Facing=<0 0 1>, WriteAlignment=true, WriteFacing=true)
   &ScaleColor(
    Scale Alpha=&MaskFloatBySpawnGroup(
     Masked Float=&FloatFromCurve(FloatCurve=DataInterface() {propertyValues {Curve {keys [<at 0.6 1>,<at 0.8 0>]}}}, CurveIndex=$System.CamAngle)
     Passthrough Float=&FloatFromCurve(FloatCurve=DataInterface() {propertyValues {Curve {keys [<at 0.6 0>,<at 0.8 1>]}}}, CurveIndex=$System.CamAngle)
    )
    ScaleA=true
    ScaleRGB=true
   ) {module=ScaleColor001}
   &ScaleSpriteSize(
    Scale Sprite Size Mode=::NewEnumerator0
    Uniform Scale Factor=&MaskFloatBySpawnGroup(
     Masked Float=&FloatFromCurve(FloatCurve=DataInterface() {propertyValues {Curve {keys [<at 0.8 1>,<at 0.81 0>]}}}, CurveIndex=$System.CamAngle)
     Passthrough Float=&FloatFromCurve(FloatCurve=DataInterface() {propertyValues {Curve {keys [<at 0.59 0>,<at 0.6 1>]}}}, CurveIndex=$System.CamAngle)
    )
   ) {module=ScaleSpriteSize002}
   &ScaleColor(
    Scale Alpha=&FloatFromCurve(
     FloatCurve=DataInterface() {propertyValues {Curve {keys [<at 0.5 0>,<at 1 1>]}}}
     CurveIndex=&Divide_Float(
      A=&DistanceBetweenPositions(
       A=$Particles.Position
       B=&ReturnCameraPosition(Coordinate Space:ENiagaraCoordinateSpace=::Simulation, Camera Query:NiagaraDataInterfaceCamera=DataInterface())
      )
      B=$User.BPInput_BottomSmokeDisableDistance
     )
     Scale Curve=2
    )
    ScaleA=true
    ScaleRGB=true
   ) {module=ScaleColor002}
   &ScaleSpriteSize(
    Uniform Curve Index=&MakeFloatFromLinearColor(LinearColor=$Particles.Color, Channel:ENiagaraLinearColor_Channels=::NewEnumerator3)
    Uniform Curve Sprite Scale=DataInterface() {propertyValues {Curve {keys [{tangentMode=RCTM_User, time=0, value=0}, {tangentMode=RCTM_User, time=0.01, value=1}]}}}
   ) {module=ScaleSpriteSize003}
   &ScaleColor(Scale Alpha=$User.Opacity_BottomSmoke, ScaleA=true, ScaleRGB=true) {module=ScaleColor003}
   &ScaleColor(
    Scale Alpha=&FN_PickFloatDependOnPlatform(TargetPlatform=DataInterface() {propertyValues {Platforms {qualityLevelMask=-29}}}, TargetPlatformActive=1.5, TargetPlatformNoActive=1.5)
    ScaleA=true
    ScaleRGB=true
   ) {module=ScaleColor004}
  ]
  rendererClasses [&NiagaraSpriteRendererProperties]
  renderers [
   {rendererClass=&NiagaraSpriteRendererProperties}
   {rendererClass=&NiagaraSpriteRendererProperties, rendererIndex=1}
   {rendererClass=&NiagaraSpriteRendererProperties, rendererIndex=2}
   {rendererClass=&NiagaraSpriteRendererProperties, rendererIndex=3}
  ]
 }
 LeftSideSmoke {
  emitterSpawnScript [/SetVariables(Emitter.Random:NiagaraFloat=&RandomRangeFloat(Minimum=0, Maximum=10)) {isSetParameters=true}]
  emitterUpdateScript [
   &EmitterState(
    Inactive Response=::NewEnumerator1
    Life Cycle Mode=::NewEnumerator1
    Loop Behavior=::NewEnumerator1
    Loop Duration Mode=::NewEnumerator1
    UseLoopDelay=false
   )
   &SpawnBurst_Instantaneous(Spawn Count=8, Spawn Time=0.2, Use Spawn Probability=false, UseLoopCountLimit=false)
  ]
  particleSpawnScript [
   &InitializeParticle(Color=$User.BaseFXColor, Ribbon Width=800, Ribbon Width Mode=::NewEnumerator1, UsePositionOffset=false, Write Material Random=true)
   &self:GetTargetExeIndexN(ParticleCount=8) {module=GetTargetExeIndexN}
   /SetVariables(Particles.Position=&MakeCustomVectorFromBool(
     Bool=$Particles.OtherSide
     True Vector=&Lerp_Vector(
      A=&MakeVector(
       X=&Add_Float(A=&Divide_Float(A=$User.FallWidth, B=-2), B=$User.SideSmokeOffset)
       Y=0
       Z=&Add_Float(
        A=&Subtract_Float(
         A=&MakeFloatFromVector(Vector=Linked() {linkedVariable {name=User.TopPosition, type=NiagaraPosition}}, Channel=::NewEnumerator2)
         B=&MakeFloatFromVector(Vector=Linked() {linkedVariable {name=Engine.Owner.Position, type=NiagaraPosition}}, Channel=::NewEnumerator2)
        )
        B=$System.TopZOffset
       )
      )
      B=&MakeVector(
       X=&Add_Float(A=&Divide_Float(A=$User.FallWidth, B=-2), B=$User.SideSmokeOffset)
       Y=0
       Z=&Subtract_Float(
        A=&MakeFloatFromVector(Vector=Linked() {linkedVariable {name=User.BottomPosition, type=NiagaraPosition}}, Channel=::NewEnumerator2)
        B=&MakeFloatFromVector(Vector=Linked() {linkedVariable {name=Engine.Owner.Position, type=NiagaraPosition}}, Channel=::NewEnumerator2)
       )
      )
      Alpha=$Particles.ExeInd
     )
     False Vector=&Lerp_Vector(
      A=&MakeVector(
       X=&Add_Float(
        A=&Divide_Float(A=$User.FallWidth, B=2)
        B=&Multiply_Float(A=$User.SideSmokeOffset, B=-1)
       )
       Y=0
       Z=&Add_Float(
        A=&Subtract_Float(
         A=&MakeFloatFromVector(Vector=Linked() {linkedVariable {name=User.TopPosition, type=NiagaraPosition}}, Channel=::NewEnumerator2)
         B=&MakeFloatFromVector(Vector=Linked() {linkedVariable {name=Engine.Owner.Position, type=NiagaraPosition}}, Channel=::NewEnumerator2)
        )
        B=$System.TopZOffset
       )
      )
      B=&MakeVector(
       X=&Add_Float(
        A=&Divide_Float(A=$User.FallWidth, B=2)
        B=&Multiply_Float(A=$User.SideSmokeOffset, B=-1)
       )
       Y=0
       Z=&Subtract_Float(
        A=&MakeFloatFromVector(Vector=Linked() {linkedVariable {name=User.BottomPosition, type=NiagaraPosition}}, Channel=::NewEnumerator2)
        B=&MakeFloatFromVector(Vector=Linked() {linkedVariable {name=Engine.Owner.Position, type=NiagaraPosition}}, Channel=::NewEnumerator2)
       )
      )
      Alpha=$Particles.ExeInd
     )
    )) {isSetParameters=true}
   /SetVariables(Particles.Position=&MakeCustomVectorFromBool(
     Bool=$Particles.OtherSide
     True Vector=&AddVectorToPosition(
      Position=$Particles.Position
      Vector=&MakeVector(
       X=&Multiply_Float(
        A=&Max_Float(
         A:NiagaraFloat=&Subtract_Float(A=&Divide_Float(A=$User.FallThickness, B=2), B=100)
         B:NiagaraFloat=0
        )
        B=-1
       )
       Y=&Add_Float(
        A=&FloatFromCurve(
         FloatCurve=DataInterface() {propertyValues {Curve {keys [{interpMode=RCIM_Cubic, leaveTangent=2020.085, tangentMode=RCTM_Break, time=0, value=-50}, <at 0.4 150>, <at 1 150>]}}}
         CurveIndex=$Particles.ExeInd
        )
        B=&Max_Float(
         A:NiagaraFloat=&Subtract_Float(A=&Divide_Float(A=$User.FallThickness, B=2), B=100)
         B:NiagaraFloat=0
        )
       )
       Z=0
      )
     )
     False Vector=&AddVectorToPosition(
      Position=$Particles.Position
      Vector=&MakeVector(
       X=&Max_Float(
        A:NiagaraFloat=&Subtract_Float(A=&Divide_Float(A=$User.FallThickness, B=2), B=100)
        B:NiagaraFloat=0
       )
       Y=&Add_Float(
        A=&FloatFromCurve(
         FloatCurve=DataInterface() {propertyValues {Curve {keys [{interpMode=RCIM_Cubic, leaveTangent=2020.085, tangentMode=RCTM_Break, time=0, value=-50}, <at 0.4 150>, <at 1 150>]}}}
         CurveIndex=$Particles.ExeInd
        )
        B=&Max_Float(
         A:NiagaraFloat=&Subtract_Float(A=&Divide_Float(A=$User.FallThickness, B=2), B=100)
         B:NiagaraFloat=0
        )
       )
       Z=0
      )
     )
    )) {isSetParameters=true}
   /SetVariables(Particles.RibbonLinkOrder:NiagaraFloat=$Particles.ExeInd) {isSetParameters=true}
   /SetVariables(Particles.RibbonID=/Niagara/DynamicInputs/NiagaraID/MakeNiagaraIDFromInt(Index:NiagaraInt32=$Particles.RID, Acquire Tag:NiagaraInt32=$Particles.RID)) {isSetParameters=true}
  ]
  particleUpdateScript [
   &ParticleState(Kill Particles When Lifetime Has Elapsed=false, Let Infinitely Lived Particles Die When Emitter Deactivates=true)
   &ScaleColor(
    Scale Alpha=&FloatFromCurve(
     FloatCurve=DataInterface() {propertyValues {Curve {keys [
         {interpMode=RCIM_Cubic, leaveTangent=1.7255, tangentMode=RCTM_Break, time=0, value=0}
         {arriveTangent=1.5583, interpMode=RCIM_Cubic, leaveTangent=1.5583, tangentMode=RCTM_User, time=0.2, value=0.8}
         {arriveTangent=-1.3337, interpMode=RCIM_Cubic, leaveTangent=-1.0169, tangentMode=RCTM_Break, time=0.7, value=0.7}
         {interpMode=RCIM_Cubic, tangentMode=RCTM_Break, time=1, value=0}
        ]}}}
     CurveIndex=$Particles.ExeInd
     Scale Curve=&FloatFromCurve(
      FloatCurve=DataInterface() {propertyValues {Curve {keys [<at 0.75 1>,<at 0.95 0>]}}}
      CurveIndex=&Abs_Float(Float:NiagaraFloat=&DotProduct(
        Vector A=&NormalizeVector(
         Vector To Normalize=&Subtract_Position(
          A=&TransformPosition(Position=$Particles.Position, Source Space=::Local, Destination Space=::World)
          B=$System.CamPos
         )
         Safe Normalize=true
         Safe Fallback Threshold=0
         Safe Fallback Vector=<0 0 0.001>
        )
        Vector B=<0 0 -1>
       ))
     )
    )
    ScaleA=true
    ScaleRGB=true
   )
   &ScaleRibbonWidth(Ribbon Width Scale=&FloatFromCurve(
     FloatCurve=DataInterface() {propertyValues {Curve {keys [
         {interpMode=RCIM_Cubic, leaveTangent=0.1385, tangentMode=RCTM_Break, time=0, value=0.4}
         {arriveTangent=0.0509, interpMode=RCIM_Cubic, tangentMode=RCTM_Break, time=1, value=1.6}
        ]}}}
     CurveIndex=$Particles.ExeInd
     Scale Curve=&Multiply_Float(A=$User.ScaleSideSmoke, B=0.5)
    ))
   &DynamicMaterialParameters(
    Index 0 Param 1=$Emitter.Random
    Module.Param0WriteEnabled=true
    Module.Param1WriteEnabled=true
    Module.Param2WriteEnabled=true
    Module.Param3WriteEnabled=true
   )
   &ScaleColor(Scale Alpha=$User.Opacity_SideSmoke, ScaleA=true, ScaleRGB=true) {module=ScaleColor003}
  ]
  rendererClasses [&NiagaraRibbonRendererProperties]
  renderers [
   {rendererClass=&NiagaraRibbonRendererProperties}
   {rendererClass=&NiagaraRibbonRendererProperties, rendererIndex=1}
  ]
 }
 TopSmoke {
  bEnabled=false
  emitterUpdateScript [
   &EmitterState(
    Enable Distance Culling=true
    Life Cycle Mode=::NewEnumerator1
    Loop Duration=1
    MaxDistance=10000
    Module.bUseMaxDistance=true
    Module.bUseMinDistance=false
    Scalability Mode=::NewEnumerator1
    UseLoopDelay=false
   )
   &SpawnRate(
    SpawnRate=&Multiply_Float(A=7, B=&Divide_Float(A=$User.FallWidth, B=1500))
    Use Spawn Probability=false
   )
   &SpawnRate(
    SpawnGroup=1
    SpawnRate=&Multiply_Float(
     A=&Multiply_Float(
      A=4
      B=&FloatFromCurve(
       FloatCurve=DataInterface() {propertyValues {Curve {keys [<at 0.9 1>,<at 1 0>]}}}
       CurveIndex=&Divide_Float(
        A=$Engine.Owner.LODDistance
        B=&MakeCustomFloatFromBool(
         Bool=&SetBoolByFloatComparison(A=$User.BPInput_BottomSmokeDisableDistance, B=0, Comparison Type=::NewEnumerator4)
         True Float=1
         False Float=$User.BPInput_BottomSmokeDisableDistance
        )
       )
      )
     )
     B=&Divide_Float(A=$User.FallWidth, B=1500)
    )
    Use Spawn Probability=false
   ) {module=SpawnRate001}
  ]
  particleSpawnScript [
   &InitializeParticle(
    Color=$User.BaseFXColor
    Lifetime Max=3
    Lifetime Min=2.5
    Lifetime Mode=::NewEnumerator1
    Sprite Size Mode=::NewEnumerator3
    Uniform Sprite Size=&MaskFloatBySpawnGroup(
     Masked Float=&Lerp_Float(
      A=450
      B=1200
      Alpha=&FloatFromCurve(
       FloatCurve=DataInterface() {propertyValues {Curve {keys [
           {interpMode=RCIM_Cubic, leaveTangent=0.1305, tangentMode=RCTM_Break, time=0, value=0}
           {arriveTangent=1, interpMode=RCIM_Cubic, leaveTangent=1, time=0.7, value=0.3}
           {arriveTangent=2.0211, interpMode=RCIM_Cubic, tangentMode=RCTM_Break, time=1, value=1}
          ]}}}
       CurveIndex=&RandomRangeFloat(Minimum=0, Maximum=1)
      )
     )
     Passthrough Float=&Lerp_Float(
      A=500
      B=1000
      Alpha=&FloatFromCurve(
       FloatCurve=DataInterface() {propertyValues {Curve {keys [
           {interpMode=RCIM_Cubic, leaveTangent=0.1305, tangentMode=RCTM_Break, time=0, value=0}
           {arriveTangent=1, interpMode=RCIM_Cubic, leaveTangent=1, time=0.7, value=0.3}
           {arriveTangent=2.0211, interpMode=RCIM_Cubic, tangentMode=RCTM_Break, time=1, value=1}
          ]}}}
       CurveIndex=&RandomRangeFloat(Minimum=0, Maximum=1)
      )
     )
    )
    UsePositionOffset=false
    Write Material Random=true
   )
   &ShapeLocation(
    Box Midpoint=<0.5 1 0.5>
    Box Size=<0 200 200>
    MaskSpawnGroup=false
    Offset=&MaskVectorBySpawnGroup(
     Masked Vector=&AddVectorToPosition(
      Position=<0 -600 100>
      Vector=&MakeVector(
       X=0
       Y=&Max_Float(
        A:NiagaraFloat=&Subtract_Float(A=&Divide_Float(A=$User.FallThickness, B=2), B=100)
        B:NiagaraFloat=0
       )
       Z=&Max_Float(
        A:NiagaraFloat=&Subtract_Float(A=&Divide_Float(A=$User.FallThickness, B=2), B=100)
        B:NiagaraFloat=0
       )
      )
     )
     Passthrough Vector=&AddVectorToPosition(
      Position=<0 0 200>
      Vector=&MakeVector(
       X=0
       Y=&Max_Float(
        A:NiagaraFloat=&Subtract_Float(A=&Divide_Float(A=$User.FallThickness, B=2), B=100)
        B:NiagaraFloat=0
       )
       Z=0
      )
     )
    )
    Offset Coordinate Space=::Local
    Offset Mode=::NewEnumerator0
    Override Seed=false
    Shape Origin=&MaskVectorBySpawnGroup(
     Masked Vector=Linked() {linkedVariable {name=User.TopPosition, type=NiagaraPosition}}
     Passthrough Vector=Linked() {linkedVariable {name=User.BottomPosition, type=NiagaraPosition}}
    )
    Shape Primitive=::NewEnumerator2
    Surface Only=false
    UseOwnerScale=true
   )
   /SetVariables(Particles.Position=&AddVectorToPosition(
     Position=$Particles.Position
     Vector=&Multiply_VectorByFloat(
      Vector=&TransformVector(Vector=<1 0 0>, Destination Space=::Simulation)
      Float=&Multiply_Float(
       A=&Lerp_Float(
        A=0
        B=&Divide_Float(A=$User.FallWidth, B=2)
        Alpha=&FloatFromCurve(
         FloatCurve=DataInterface() {propertyValues {Curve {keys [
             {interpMode=RCIM_Cubic, leaveTangent=1.4213, tangentMode=RCTM_Break, time=0, value=0}
             {arriveTangent=0.6232, interpMode=RCIM_Cubic, tangentMode=RCTM_Break, time=1, value=1}
            ]}}}
         CurveIndex=&RandomRangeFloat(Minimum=0, Maximum=1)
        )
       )
       B=&Lerp_Float(
        A=-1
        B=1
        Alpha=&MakeFloatFromInt(Int=&RandomRangeInt(Minimum=0, Maximum=1))
       )
      )
     )
    )) {isSetParameters=true}
   /SetVariables(Particles.Velocity=&Multiply_Vector(A=$Particles.Velocity, B=<1 1 0.2>)) {isSetParameters=true}
   &AddVelocity(Velocity=&RandomRangeVector(Minimum=<0 368.6783 -100>, Maximum=<1 445.846 14.8514>))
  ]
  particleUpdateScript [
   &ParticleState()
   &SpriteRotationRate(Rotation Rate=-21.5757)
   &ScaleColor(
    Scale Alpha=&FloatFromCurve(
     FloatCurve=DataInterface() {propertyValues {Curve {keys [
         {interpMode=RCIM_Cubic, tangentMode=RCTM_User, time=0, value=0}
         {interpMode=RCIM_Cubic, tangentMode=RCTM_User, time=0.5, value=1}
         {tangentMode=RCTM_User, time=1, value=0}
        ]}}}
     Scale Curve=$User.Opacity_BottomSmoke
    )
    ScaleA=true
    ScaleRGB=true
   )
   &GravityForce(Gravity=&RandomRangeVector(Minimum=<0 0 -50.8566>, Maximum=<0 0 -14.5621>))
   &Drag(Drag=0.1636, Use Linear Drag=true, Use Rotational Drag=true)
   &SolveForcesAndVelocity(Clamp Velocity=false, Limit Acceleration=false)
   &DynamicMaterialParameters(
    Index 0 Param 1=&RandomRangeFloat(Minimum=0, Maximum=1)
    Module.Param0WriteEnabled=true
    Module.Param1WriteEnabled=true
    Module.Param2WriteEnabled=true
    Module.Param3WriteEnabled=true
   )
   &KillParticles(Kill Particles=&SetBoolByFloatComparison(A=$Particles.NormalizedAge, B=&MaskFloatBySpawnGroup(Masked Float=0.9, Passthrough Float=1)))
  ]
  rendererClasses [&NiagaraSpriteRendererProperties]
  renderers [{rendererClass=&NiagaraSpriteRendererProperties}]
 }
 TopSplash {
  bEnabled=false
  emitterUpdateScript [
   &EmitterState(
    Enable Distance Culling=true
    Life Cycle Mode=::NewEnumerator1
    Loop Duration=1
    MaxDistance=12500
    Module.bUseMaxDistance=true
    Module.bUseMinDistance=false
    Scalability Mode=::NewEnumerator1
    UseLoopDelay=false
   )
   &SpawnRate(
    SpawnRate=&Multiply_Float(
     A=14
     B=&MakeFloatFromInt(Int=&Add_Integer(A=&MakeIntFromFloat(Float=$User.TopSplashCount), B=2))
    )
    Use Spawn Probability=false
   )
  ]
  particleSpawnScript [
   &FN_SetSpawnGroup(Number of Spawn Groups:NiagaraInt32=&Add_Integer(A=&MakeIntFromFloat(Float=$User.TopSplashCount), B=2))
   &InitializeParticle(
    Color=$User.BaseFXColor
    Lifetime Max=0.7
    Lifetime Min=0.4
    Lifetime Mode=::NewEnumerator1
    Sprite Rotation Angle=&RandomRangeFloat(Minimum=-40, Maximum=40)
    Sprite Rotation Mode=::NewEnumerator0
    Sprite Size Mode=::NewEnumerator1
    Uniform Sprite Size Max=549.9521
    Uniform Sprite Size Min=297.0012
    UsePositionOffset=false
    Write Material Random=true
   )
   &ShapeLocation(
    Box Size=<200 300 0>
    MaskSpawnGroup=false
    Offset=&AddVectorToPosition(
     Position=<0 -150 20>
     Vector=&MakeVector(
      X=0
      Y=&Max_Float(
       A:NiagaraFloat=&Subtract_Float(A=&Divide_Float(A=$User.FallThickness, B=2), B=100)
       B:NiagaraFloat=0
      )
      Z=&Max_Float(
       A:NiagaraFloat=&Subtract_Float(A=&Divide_Float(A=$User.FallThickness, B=2), B=100)
       B:NiagaraFloat=0
      )
     )
    )
    Offset Coordinate Space=::Local
    Offset Mode=::NewEnumerator0
    Override Seed=false
    Shape Origin=&self:TopSplashPosition(
     Offset:Vector3f=&MakeVectorFromLinearColorRGB(Color=$User.TopSplashOffset)
     Seed:NiagaraInt32=&MakeIntFromFloat(Float=$User.TopSplashSeed)
     SideOffset:NiagaraFloat=$User.TopSplashSideOffset
     SpawnGroup:NiagaraInt32=$Particles.SpawnGroup
     TopPosition:Vector3f=Linked() {linkedVariable {name=User.TopPosition, type=NiagaraPosition}}
     Width:NiagaraFloat=$User.FallWidth
    )
    Shape Primitive=::NewEnumerator2
    Surface Only=false
    UseOwnerScale=true
   )
   &AddVelocity(
    Cone Angle=60
    Cone Axis=<0 0.2 1>
    Module.UseConeFalloff=false
    Module.UseVelDistribution=false
    Override Seed=false
    Velocity Mode=::NewEnumerator2
    Velocity Speed=600
   )
  ]
  particleUpdateScript [
   &ParticleState()
   &ScaleSpriteSize(Uniform Curve Sprite Scale=DataInterface() {propertyValues {Curve {keys [
        {arriveTangent=4.1785, interpMode=RCIM_Cubic, leaveTangent=4.1785, tangentMode=RCTM_User, time=0, value=0}
        {arriveTangent=1.1179, interpMode=RCIM_Cubic, leaveTangent=1.1179, tangentMode=RCTM_User, time=0.2042, value=0.6228}
        {arriveTangent=0.0224, interpMode=RCIM_Cubic, leaveTangent=0.0224, tangentMode=RCTM_User, time=1, value=1}
       ]}}})
   &SubUVAnimation(
    "Sprite Renderer "=DataInterface() {propertyValues {SpriteRenderer=&self:TopSplash_0.NiagaraSpriteRendererProperties_0}}
    UseEndFrame=false
    UseStartFrame=false
   )
   &ScaleColor(
    Scale Alpha=&FloatFromCurve(
     FloatCurve=DataInterface() {propertyValues {Curve {keys [
         {interpMode=RCIM_Cubic, tangentMode=RCTM_Break, time=0, value=0}
         {interpMode=RCIM_Cubic, tangentMode=RCTM_Break, time=0.2, value=1}
         {arriveTangent=-0.0405, interpMode=RCIM_Cubic, leaveTangent=-0.0405, tangentMode=RCTM_User, time=0.5, value=0.95}
         {arriveTangent=-3.8219, interpMode=RCIM_Cubic, leaveTangent=-3.8219, tangentMode=RCTM_User, time=1, value=0}
        ]}}}
     Scale Curve=3
    )
    ScaleA=true
    ScaleRGB=true
   )
   &GravityForce(Gravity=<0 0 -2846.4424>)
   &SolveForcesAndVelocity(Clamp Velocity=false, Limit Acceleration=false)
  ]
  rendererClasses [&NiagaraSpriteRendererProperties]
  renderers [{rendererClass=&NiagaraSpriteRendererProperties}]
 }
}
summary {
 systemName=NS_WaterfallFX_Hades
 emitters [
  {bEnabled=false, emitterName=LargeSmoke, rendererClasses [&NiagaraSpriteRendererProperties]}
  {emitterName=CenterSmoke, rendererClasses [&NiagaraSpriteRendererProperties]}
  {emitterName=LeftSideSmoke, rendererClasses [&NiagaraRibbonRendererProperties]}
  {bEnabled=false, emitterName=TopSmoke, rendererClasses [&NiagaraSpriteRendererProperties]}
  {emitterName=BottomSplash, simTarget=GPUComputeSim, rendererClasses [&NiagaraSpriteRendererProperties]}
  {bEnabled=false, emitterName=TopSplash, rendererClasses [&NiagaraSpriteRendererProperties]}
  {emitterName=DropSmokeA, rendererClasses [&NiagaraSpriteRendererProperties]}
  {emitterName=Drops, simTarget=GPUComputeSim, rendererClasses [&NiagaraSpriteRendererProperties]}
 ]
 userVariables [
  User.BASECOLOR_OLD {type=LinearColor, description="", defaultValue {struct=LinearColor, value=<rgba 0.1781 0.5 0.0779 1>}}
  User.BaseFXColor {type=LinearColor, description="", defaultValue {struct=LinearColor, value=<rgba 0.4427 0.1337 0.1487 1>}}
  User.BottomPosition {type=NiagaraPosition, description="", defaultValue {}}
  User.BPInput_BottomSmokeDisableDistance {type=NiagaraFloat, defaultValue=2048, description=""}
  User.DropSmokeAgeOffset {type=NiagaraFloat, defaultValue=0, description=""}
  User.FallThickness {type=NiagaraFloat, defaultValue=0, description=""}
  User.FallWidth {type=NiagaraFloat, defaultValue=0, description=""}
  User.LargeSmokeOffset {type=NiagaraFloat, defaultValue=900, description=""}
  User.Opacity_BottomSmoke {type=NiagaraFloat, defaultValue=1, description=""}
  User.Opacity_DropSmoke {type=NiagaraFloat, defaultValue=1, description=""}
  User.Opacity_SideSmoke {type=NiagaraFloat, defaultValue=1, description=""}
  User.ScaleBottomLargeSmoke {type=NiagaraFloat, defaultValue=1, description=""}
  User.ScaleBottomMainSplash {type=NiagaraFloat, defaultValue=1, description=""}
  User.ScaleBottomMediumSmoke {type=NiagaraFloat, defaultValue=1, description=""}
  User.ScaleDropSmoke {type=NiagaraFloat, defaultValue=1, description=""}
  User.ScaleDropSmokeSpawnRate {type=NiagaraFloat, defaultValue=1, description=""}
  User.ScaleSideSmoke {type=NiagaraFloat, defaultValue=1, description=""}
  User.SideSmokeOffset {type=NiagaraFloat, defaultValue=65, description=""}
  User.TopPosition {type=NiagaraPosition, description="", defaultValue {}}
  User.TopSplashCount {type=NiagaraFloat, defaultValue=2, description=""}
  User.TopSplashOffset {type=LinearColor, description="", defaultValue {struct=LinearColor, value=<rgba 0 0 0 1>}}
  User.TopSplashSeed {type=NiagaraFloat, defaultValue=128, description=""}
  User.TopSplashSideOffset {type=NiagaraFloat, defaultValue=0, description=""}
 ]
}
