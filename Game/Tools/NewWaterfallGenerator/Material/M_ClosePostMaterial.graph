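# Post-process material for the waterfall generator that blends close-range fog over the scene near a waterfall.
# Two noise layers, sampled with panorama UVs built from view-vector yaw and height, combine with depth-based density and tint by a depth-blended fog color.
# Fog fades with camera distance from the waterfall's bottom point.
# Further masks limit it by vertical view angle, camera position against top, bottom and facing inputs, and screen height.
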Material /Game/Tools/NewWaterfallGenerator/Material/M_ClosePostMaterial
graph {
 properties {materialDomain=MD_PostProcess, pixelDepthOffsetMode=PDOM_Legacy, usedShadingModels=Default Lit}
 nodes [
  FeatureLevelSwitch()
  n1 = ComponentMask(None=SceneTexture(UVs=ScreenPosition().ViewportUV) {sceneTextureId=PPI_PostProcessInput0}.Color) {b=true, g=true, r=true}
  n2 = SceneDepth(UVs=ScreenPosition().ViewportUV)
  n3 = Constant() {r=4096}
  n4 = OneMinus(None=Saturate(None=Divide(A=@n2.R, B=@n3)))
  n5 = Constant() {r=4}
  n6 = LinearInterpolate(
   A=VectorParameter() {parameterName=Fog2, defaultValue=<rgba 0.525 0.5625 0.75 1>}.RGB
   B=VectorParameter() {parameterName=Fog, defaultValue=<rgba 0.6596 0.7178 0.776 1>}.RGB
   Alpha=OneMinus(None=Power(Base=@n4, Exp=@n5))
  )
  n7 = TextureObjectParameter() {
   group=FogSecTexture
   parameterName=FogSecTexture_Texture
   samplerType=SAMPLERTYPE_LinearGrayscale
   sortPriority=0
   texture=/Game/Effects/Fort_Effects/Emotes/UltraLight/Texture/T_Tiling_Noise_01_GreyScale
  }
  n8 = ScreenPosition()
  n9 = VectorParameter() {
   group=FogSecTexture
   parameterName=FogSecTexture_UVControl
   channelNames {a=VPan, b=UPan, g=VScale, r=UScale}
   defaultValue=<rgba 2 1 0.03 0.03>
  }
  n10 = NamedRerouteUsage()
  n11 = Add(
   A=MaterialFunctionCall!(
    Texture=@n7
    "[Override]DDXInput"=DDX(Value=@n8.ViewportUV)
    "[Override]DDYInput"=DDY(Value=@n8.ViewportUV)
    AdditionalUVOffset
    UVControl_RGB=@n9
    UVControl_A=@n9
    "[Override]TimeInput"
    "[Override]UVInput"=@n10
   ) {materialFunction=/Game/Effects/Niagara/MasterMaterials/Functions/Common/FXCommon_TextureSample_DDXDDY}.R
   B=ScalarParameter() {defaultValue=0.2, group=FogSecTexture, parameterName=FogSecTexture_GrayAdjust_Add}
  )
  n12 = ScalarParameter() {defaultValue=0.6, group=FogSecTexture, parameterName=FogSecTexture_GrayAdjust_Mul}
  n13 = Power(
   Base=Saturate(None=Multiply(A=@n11, B=@n12))
   Exp=ScalarParameter() {defaultValue=0.5, group=FogSecTexture, parameterName=FogSecTexture_GrayAdjust_Power}
  )
  n14 = Constant() {r=0.5}
  n15 = Saturate(None=Clamp(None=Multiply(A=@n13, B=@n14)) {minDefault=0.2})
  n16 = TextureObjectParameter() {
   group=FogBaseTexture
   parameterName=FogBaseTexture_Texture
   samplerType=SAMPLERTYPE_LinearGrayscale
   sortPriority=0
   texture=/Game/Effects/Fort_Effects/Emotes/UltraLight/Texture/T_Cloud_Greyscale
  }
  n17 = ScreenPosition()
  n18 = VectorParameter() {
   group=FogBaseTexture
   parameterName=FogBaseTexture_UVControl
   channelNames {a=VPan, b=UPan, g=VScale, r=UScale}
   defaultValue=<rgba 1 0.75 0.0063 0.0031>
  }
  n19 = MaterialFunctionCall(
   Texture=@n16
   "[Override]DDXInput"=DDX(Value=@n17.ViewportUV)
   "[Override]DDYInput"=DDY(Value=@n17.ViewportUV)
   UVControl_RGB=@n18
   UVControl_A=@n18
   "[Override]UVInput"=@n10
  ) {materialFunction=&FXCommon_TextureSample_DDXDDY}
  n20 = VectorParameter() {parameterName=DepthEffect, defaultValue=<rgba 0.5 1 1 0>}
  n21 = Multiply(
   A=MaterialFunctionCall(Input=Power(Base=@n19.R, Exp=@n20.B), Input Low=Constant(), Input High=Constant() {r=1}, Target Low=@n20.R, Target High=Constant() {r=1}) {materialFunction=&RemapValueRange}.Result
   B=@n20.G
  )
  n22 = VectorParameter() {parameterName=DepthFogBaseParam, defaultValue=<rgba 1024 0.65 2 0>}
  n23 = Subtract(A=@n2.R, B=@n22.A)
  n24 = OneMinus(None=Saturate(None=Divide(A=@n23, B=@n22.R) {constB=2048}))
  n25 = Saturate(None=OneMinus(None=Power(Base=@n24, Exp=@n22.B)))
  n26 = Multiply(A=Saturate(None=Clamp(None=@n21, Min=@n20.A) {minDefault=0.1}), B=Saturate(None=Multiply(A=@n25, B=@n22.G)))
  n27 = Subtract(A=CameraPositionWS(), B=VectorParameter() {parameterName=BPInput_ButtomPos, defaultValue=<rgba 0 0 0 1>}.RGB)
  n28 = Power(
   Base=Divide(
    A=Length(None=@n27) {is=class MaterialExpressionLength}
    B=ScalarParameter() {defaultValue=4096, parameterName=BPInput_DensityGradientRadius}
   )
   Exp=Constant() {r=1}
  )
  n29 = Multiply(A=Saturate(None=Add(A=@n15, B=@n26)), B=Saturate(None=OneMinus(None=@n28)))
  n30 = MaterialFunctionCall(Float3=CameraVectorWS()) {materialFunction=&BreakOutFloat3Components}
  n31 = Abs(None=@n30.B)
  n32 = ScalarParameter() {defaultValue=4, parameterName=Mask}
  n33 = VectorParameter() {parameterName=BPInput_ButtomPos, defaultValue=<rgba 0 0 0 1>}
  n34 = Constant3Vector()
  n35 = Normalize(VectorInput=Subtract(A=Add(A=@n33.RGB, B=@n34), B=CameraPositionWS()))
  n36 = DotProduct(A=@n35, B=Constant3Vector() {constant=<rgba 0 0 1 0>})
  n37 = LinearInterpolate(
   A=LinearInterpolate(A=LinearInterpolate(A=@n1, B=@n6, Alpha=@n29), B=@n1, Alpha=Power(Base=@n31, Exp=@n32))
   B=@n1
   Alpha=Saturate(None=Multiply(A=@n36) {constB=20})
  ) {constB=0}
  n38 = Subtract(A=DotProduct(A=@n35, B=VectorParameter() {parameterName=BPInput_PositiveDir, defaultValue=<rgba 0 0 0 1>}.RGB)) {constB=0.15}
  n39 = OneMinus(None=Saturate(None=Multiply(A=@n38) {constB=5}))
  n40 = Normalize(VectorInput=Subtract(A=VectorParameter() {parameterName=BPInput_TopPos, defaultValue=<rgba 0 0 0 1>}.RGB, B=CameraPositionWS()))
  n41 = Constant3Vector() {constant=<rgba 0 0 -1 0>}
  n42 = Multiply(A=Subtract(A=DotProduct(A=@n40, B=@n41)) {constB=-0.3})
  n43 = ScreenPosition()
  LinearInterpolate(
   A=@n1
   B=LinearInterpolate(A=LinearInterpolate(A=@n1, B=@n37, Alpha=@n39), B=@n1, Alpha=Saturate(None=@n42))
   Alpha=Power(Base=MaterialFunctionCall(Float2=@n43.ViewportUV) {materialFunction=&BreakOutFloat2Components}.G) {constExponent=1}
  )
  Multiply()
  n44 = Constant3Vector() {constant=<rgba 0 0 1 0>}
  n45 = Normalize(VectorInput=MaterialFunctionCall!(Plane Normal=@n44, Vector=CameraVectorWS()) {materialFunction=/Engine/Functions/Engine_MaterialFunctions02/Math/ProjectVectorOntoPlane}.Result)
  n46 = Constant3Vector() {constant=<rgba 1 0 0 0>}
  n47 = MaterialFunctionCall!(Normalized V1=@n45, Normalized V2=@n46) {materialFunction=/Engine/Functions/Engine_MaterialFunctions02/Math/AngleBetweenVectors}
  n48 = DotProduct(A=CrossProduct(A=@n45, B=@n46), B=@n44)
  n49 = LinearInterpolate(A=@n47.Angle In Degrees, B=Subtract(B=@n47.Angle In Degrees) {constA=360}, Alpha=Saturate(None=Ceil(None=@n48)))
  NamedRerouteDeclaration(None=AppendVector(A=Divide(A=@n49) {constB=360}, B=@n30.B)) {name=SphereUV, nodeColor=<rgba 0 0.0391 1 1>}
 ]
}
