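# Translucent waterfall foam material for low-end platforms, from the waterfall generator tool.
# Vertex colors encode flow angle, speed and white-water amount, feeding FortWaterRiverFlowmaps, whose foam mask is multiplied by cubed vertex alpha.
# Quality and shading-path switches keep that mask for Forward, Mobile and deferred Low/Medium, zeroing it otherwise.
# Foam color shifts through Color_Time_of_Day, pink-tinted at morning and evening, blue at night, into an emissive adjustment function.
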Material /Game/Tools/NewWaterfallGenerator/Material/M_LowPlatformFoam
graph {
 properties {blendMode=BLEND_Translucent, pixelDepthOffsetMode=PDOM_Legacy, usedShadingModels=Default Lit}
 nodes [
  Constant3Vector() {constant=<rgba 0 0 7 0>}
  n1 = VectorParameter() {parameterName=FoamColor, defaultValue=<rgba 0.765 0.8775 0.9 1.2>}
  n2 = Multiply(A=@n1.RGB) {constB=0.2}
  n3 = VectorParameter() {parameterName=PinkC, defaultValue=<rgba 0.2386 0.1384 0.3 0.3>}
  n4 = LinearInterpolate(A=@n2, B=@n3.RGB, Alpha=@n3.RGB) {constAlpha=0.2}
  n5 = Multiply(A=@n1.RGB) {constB=0.1}
  n6 = VectorParameter() {parameterName=BlueC, defaultValue=<rgba 0.2389 0.3929 0.6052 0.2939>}
  MaterialFunctionCall(Emissive=MaterialFunctionCall(Color_Morning=@n4, Color_Day=@n2, Color_Evening=@n4, Color_Night=LinearInterpolate(A=@n5, B=@n6.RGB, Alpha=@n6.RGB) {constAlpha=0.2}) {materialFunction=&Color_Time_of_Day}.Result) {materialFunction=&MF_FX_EmissiveAdjustments}
  n7 = Constant()
  n8 = VertexColor()
  n9 = AppendVector(A=Cosine(None=@n8.R), B=Sine(None=@n8.R))
  n10 = Multiply(A=@n8.G, B=ScalarParameter() {defaultValue=1500, parameterName=WaterfallMaxFlowVelocity})
  n11 = ScalarParameter() {defaultValue=1300, parameterName=WaterfallMaxWhiteWater}
  n12 = NamedRerouteDeclaration(None=AppendVector(A=Multiply(A=@n9, B=@n10), B=Multiply(A=@n8.B, B=@n11))) {name=WaterfallSupport_VelocityAndLength, nodeColor=<rgba 1 0.4688 0 1>}
  n13 = NamedRerouteDeclaration(None=ScalarParameter() {defaultValue=200, parameterName=WaterfallDepth}) {name=WaterfallSupport_WaterDepth, nodeColor=<rgba 0.7109 1 0 1>}
  n14 = Multiply(A=Multiply(
    A=MaterialFunctionCall!(Attributes, Transition, Sim Foam Erasure, Velocity=@n12, Far Mask Override, Depth=@n13) {materialFunction=/FortWater/Materials/Functions/FortWaterRiverFlowmaps}.Foam Mask
    B=Power(Base=@n8.A) {constExponent=3}
   ))
  n15 = QualitySwitch(Default=@n7, Low=@n14, High=@n7, Medium=@n14, Epic=@n7)
  ShadingPathSwitch(Default=@n15, Deferred=@n15, Forward=@n14, Mobile=@n14)
 ]
}
