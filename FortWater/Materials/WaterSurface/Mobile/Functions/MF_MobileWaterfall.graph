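# Material function that builds the surface attributes for a waterfall in FortWater's mobile water materials.
# Two river normal samples scroll at different tilings over square-root UVs, forming the normal and alpha-based whitewater and roughness.
# Color parameters blend by vertex-normal slope, whitewater and top fade, then get hue-shifted.
# A dithered opacity mask fades edges and varies by face side, while world position offset displaces along vertex normals using texture alpha.
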MaterialFunction /FortWater/Materials/WaterSurface/Mobile/Functions/MF_MobileWaterfall
graph {nodes [
  Constant3Vector() {constant=<rgba 0.376 0.443 0.5469 1>}
  Constant3Vector() {constant=<rgba 0.1755 0.3029 0.401 1>}
  Constant3Vector() {constant=<rgba 0.032 0.0776 0.1042 1>}
  n1 = SetMaterialAttributes!(
   MaterialAttributes
   BaseColor=NamedRerouteUsage()
   Specular=NamedRerouteUsage()
   Roughness=NamedRerouteUsage()
   EmissiveColor
   OpacityMask=NamedRerouteUsage()
   Normal=NamedRerouteUsage()
   WorldPositionOffset=NamedRerouteUsage()
  ) {attributeSetTypes [
    #1
    #2
    #3
    #4
    #5
    #6
    #7
   ]}
  FunctionOutput(None=@n1)
  n2 = SquareRoot(None=TextureCoordinate())
  n3 = Multiply(A=@n2) {constB=0.7}
  n4 = ComponentMask(None=@n3) {r=true}
  n5 = Multiply(A=Time()) {constB=0.06}
  n6 = Subtract(A=ComponentMask(None=@n3) {g=true}, B=@n5)
  NamedRerouteDeclaration(None=TextureSample(UVs=AppendVector(A=@n4, B=@n6)) {samplerType=SAMPLERTYPE_LinearColor, texture=/FortWater/Materials/WaterSurface/Mobile/Textures/T_MobileWater_RiverNormal}.RGBA) {name=Texture1, nodeColor=<rgba 0 0.0053 1 1>}
  n7 = Multiply(A=@n2) {constB=0.4}
  n8 = ComponentMask(None=@n7) {r=true}
  n9 = Subtract(A=ComponentMask(None=@n7) {g=true}, B=@n5)
  NamedRerouteDeclaration(None=TextureSample(UVs=AppendVector(A=@n8, B=@n9)) {samplerType=SAMPLERTYPE_LinearColor, texture=&T_MobileWater_RiverNormal}.RGBA) {name=Texture2, nodeColor=<rgba 1 0 0.0203 1>}
  n10 = NamedRerouteUsage()
  n11 = Multiply(A=ComponentMask(None=@n10) {g=true, r=true}) {constB=2}
  n12 = NamedRerouteUsage()
  n13 = Multiply(A=ComponentMask(None=@n12) {g=true, r=true}) {constB=2}
  n14 = AppendVector(
   A=Add(A=Add(A=@n11) {constB=-1}, B=Add(A=@n13) {constB=-1})
   B=Multiply(A=ComponentMask(None=@n10) {b=true}, B=ComponentMask(None=@n12) {b=true})
  )
  NamedRerouteDeclaration(None=Multiply(A=Transform(None=@n14), B=TwoSidedSign())) {name=Normal, nodeColor=<rgba 1 0 0.3047 1>}
  n15 = ScalarParameter() {parameterName=Mobile_OpacityMask_Contrast_Low}
  n16 = ScalarParameter() {defaultValue=1, parameterName=Mobile_OpacityMask_Contrast_high}
  n17 = TextureCoordinate()
  n18 = ComponentMask(None=@n17) {g=true}
  n19 = NamedRerouteUsage()
  n20 = NamedRerouteUsage()
  n21 = Add(A=Add(A=ComponentMask(None=@n19) {a=true}, B=ComponentMask(None=@n20) {a=true})) {constB=0.1}
  n22 = Multiply(
   A=Multiply(A=LinearInterpolate(A=@n15, B=@n16, Alpha=@n18), B=ScalarParameter() {defaultValue=7, parameterName=Mobile_OpacityMask_Mult}) {constB=7}
   B=@n21
  )
  n23 = LinearInterpolate(A=Saturate(None=Multiply(A=@n18) {constB=40}), B=Saturate(None=Multiply(A=@n22) {constB=5}), Alpha=Saturate(None=TwoSidedSign()))
  n24 = Add(A=Frac(None=ComponentMask(None=@n17) {r=true})) {constB=-0.5}
  n25 = OneMinus(None=Multiply(A=Abs(None=@n24)) {constB=2})
  n26 = Saturate(None=Multiply(A=@n21, B=Multiply(A=@n25) {constB=5}))
  NamedRerouteDeclaration(None=MaterialFunctionCall(Alpha Threshold=Multiply(A=@n23, B=@n26)) {materialFunction=&DitherTemporalAA}.Result) {name=Opacity Mask, nodeColor=<rgba 0 0.4141 1 1>}
  n27 = Abs(None=ComponentMask(None=VertexNormalWS()) {b=true})
  n28 = VectorParameter() {group=Mobile Color, parameterName=Mobile Color Light, sortPriority=40, defaultValue=<rgba 0.1755 0.3029 0.401 1>}
  n29 = VectorParameter() {group=Mobile Color, parameterName=Mobile Color Dark, sortPriority=40, defaultValue=<rgba 0.376 0.443 0.5469 1>}
  n30 = VectorParameter() {group=Mobile Color, parameterName=Mobile Color Whitewater, sortPriority=40, defaultValue=<rgba 0.376 0.443 0.5469 1>}
  n31 = NamedRerouteUsage()
  n32 = NamedRerouteUsage()
  n33 = Multiply(A=Add(A=ComponentMask(None=@n31) {a=true}, B=ComponentMask(None=@n32) {a=true})) {constB=0.7}
  n34 = LinearInterpolate(A=LinearInterpolate(A=@n28.RGB, B=@n29.RGB, Alpha=@n27), B=@n30.RGB, Alpha=@n33)
  n35 = OneMinus(None=@n27)
  n36 = Saturate(None=TwoSidedSign())
  n37 = TextureCoordinate()
  n38 = Saturate(None=Multiply(A=ComponentMask(None=@n37) {g=true}) {constB=6})
  n39 = Multiply(A=Saturate(None=Add(A=@n35, B=@n36)), B=Multiply(A=@n38, B=@n38))
  NamedRerouteDeclaration(None=MaterialFunctionCall(
    Hue Shift Percentage=ScalarParameter() {group=Mobile Color, parameterName=MobileHSVShift, sliderMax=1, sliderMin=-1, sortPriority=40}
    Texture=LinearInterpolate(A=@n29.RGB, B=@n34, Alpha=@n39)
   ) {materialFunction=&HueShift}.Result) {name=Base Color, nodeColor=<rgba 0.1719 0 1 1>}
  n40 = TwoSidedSign()
  NamedRerouteDeclaration(None=Multiply(A=Saturate(None=@n40)) {constB=0.6}) {name=Specular, nodeColor=<rgba 0.8047 0 1 1>}
  NamedRerouteDeclaration(None=Saturate(None=@n33)) {name=Roughness, nodeColor=<rgba 0.5234 1 0 1>}
  n41 = ComponentMask(None=NamedRerouteUsage()) {a=true}
  n42 = VertexNormalWS()
  n43 = ComponentMask(None=TextureCoordinate()) {g=true}
  n44 = Multiply(A=Multiply(A=Multiply(A=@n41, B=@n42)) {constB=40}, B=Saturate(None=Multiply(A=@n43) {constB=10}))
  n45 = Frac(None=ComponentMask(None=TextureCoordinate()) {r=true})
  n46 = Multiply(A=Abs(None=Add(A=@n45) {constB=-0.5})) {constB=2}
  n47 = Saturate(None=Multiply(A=OneMinus(None=@n46)) {constB=2})
  NamedRerouteDeclaration(None=Add(A=Multiply(A=@n44, B=@n47), B=Multiply(A=@n42) {constB=-20})) {name=WPO, nodeColor=<rgba 0.4766 1 0 1>}
 ]}
