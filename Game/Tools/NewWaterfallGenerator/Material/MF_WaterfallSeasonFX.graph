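# Waterfall material function that adds an optional seasonal, time-of-day glow to incoming material attributes, named for an S31 red Hades river effect.
# Color_Time_of_Day resolves four color parameters for morning, day, evening and night into one tint, which multiplies base color and adds to existing emissive.
# The S31RedHadesRiverTOD static switch selects the modified attributes or passes the input through unchanged.
# All four colors default to black, so the added glow is authored through parameter overrides.
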MaterialFunction /Game/Tools/NewWaterfallGenerator/Material/MF_WaterfallSeasonFX
graph {nodes [
  n1 = FunctionInput() {inputName=MAInput, inputType=FunctionInput_MaterialAttributes}
  n2 = GetMaterialAttributes(None=@n1) {attributeGetTypes [#1, #2]}
  n3 = VectorParameter() {parameterName=S31RedHadesRiverTOD_Morning, defaultValue=<rgba 0 0 0 1>}
  n4 = VectorParameter() {parameterName=S31RedHadesRiverTOD_Day, defaultValue=<rgba 0 0 0 1>}
  n5 = VectorParameter() {parameterName=S31RedHadesRiverTOD_Evening, defaultValue=<rgba 0 0 0 1>}
  n6 = VectorParameter() {parameterName=S31RedHadesRiverTOD_Night, defaultValue=<rgba 0 0 0 1>}
  n7 = Add(
   A=Multiply(
    A=@n2.BaseColor
    B=MaterialFunctionCall(Color_Morning=@n3.RGB, Color_Day=@n4.RGB, Color_Evening=@n5.RGB, Color_Night=@n6.RGB) {materialFunction=&Color_Time_of_Day}.Result
   )
   B=@n2.EmissiveColor
  )
  FunctionOutput(None=StaticSwitchParameter(True=SetMaterialAttributes!(MaterialAttributes=@n1, EmissiveColor=@n7) {attributeSetTypes [#1]}, False=@n1) {parameterName=S31RedHadesRiverTOD})
 ]}
