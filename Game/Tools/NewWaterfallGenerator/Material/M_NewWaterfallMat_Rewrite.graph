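# Masked, two-sided waterfall material for the waterfall generator, with Blueprint-fed dimensions driving tiled UVs and fall masks.
# High-end attributes layer noise-distorted foam, emissive areas, normal-direction WPO and dithered edge masks, blending into river flowmaps.
# Static switches add a golden variant with collection-driven transition and triplanar sparkles.
# Quality and shading-path switches select the high-end or MF_MobileWaterfall-based branch, feeding MF_WaterfallSeasonFX.
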Material /Game/Tools/NewWaterfallGenerator/Material/M_NewWaterfallMat_Rewrite
graph {
 properties {
  bTangentSpaceNormal=false
  bUseMaterialAttributes=true
  bUsedWithInstancedStaticMeshes=true
  bUsedWithNanite=true
  blendMode=BLEND_Masked
  pixelDepthOffsetMode=PDOM_Legacy
  translucencyLightingMode=TLM_SurfacePerPixelLighting
  twoSided=true
  usedShadingModels=Default Lit
 }
 nodes [
  n1 = Divide(B=Divide(A=NamedRerouteUsage(), B=NamedRerouteUsage()))
  n2 = Subtract(
   A=MaterialFunctionCall(Float2=NamedRerouteUsage()) {materialFunction=&BreakOutFloat2Components}.G
   B=Divide(A=NamedRerouteUsage(), B=NamedRerouteUsage())
  ) {constB=-1}
  n3 = OneMinus(None=Saturate(None=Multiply(A=@n1, B=@n2)))
  Add(A=@n3)
  n4 = VertexColor()
  Add(A=@n4.R, B=Multiply(A=@n4.R, B=NamedRerouteUsage()))
  Constant() {r=1}
  Constant() {r=1}
  n5 = NamedRerouteUsage()
  n6 = NamedRerouteUsage()
  MaterialFunctionCall!(MAInput=ShadingPathSwitch(Default=QualitySwitch(Default=@n5, Low=@n6), Mobile=@n6)) {materialFunction=&/MF_WaterfallSeasonFX}
  MaterialFunctionCall!(In, Velocity, Depth, Transition, IsRiverMaterial, IsTranslucentMat) {materialFunction=/FortWater/Materials/WaterSurface/Hermes/Gold/MF_LiquidGold}
  Multiply()
  NamedRerouteDeclaration(None=ScalarParameter() {group="[IgnoreThisGroup]BPInput", parameterName=BPInput_FallLength}) {name=FallLength, nodeColor=<rgba 0 1 0.6538 1>}
  n7 = NamedRerouteUsage()
  n8 = NamedRerouteUsage()
  NamedRerouteDeclaration(None=Subtract(A=Subtract(A=@n7, B=@n8), B=NamedRerouteUsage())) {name=ExtendLength, nodeColor=<rgba 1 0.3984 0 1>}
  n9 = StaticSwitchParameter(
   True=ComponentMask(None=CollectionParameter() {collection=/Water/Materials/MPC/MPC_GoldenWatefall, parameterName=FakeSubsurfaceColor}) {b=true, g=true, r=true}
   False=VectorParameter() {parameterName=FakeSubsurfaceColor}.RGB
  ) {desc=CH7S1 GoldenWaterfall Update, parameterName=UseMPCforWaterfallColors}
  n10 = VectorParameter() {parameterName=Gold Sub surface, defaultValue=<rgba 0.2346 0.5831 0.7604 1>}
  n11 = LinearInterpolate(A=@n9, B=@n10.RGB, Alpha=NamedRerouteUsage())
  n12 = Saturate(None=Length(None=@n11) {is=class MaterialExpressionLength})
  n13 = TextureObjectParameter() {
   group=MainWave
   parameterName=MainWave_Texture
   samplerType=SAMPLERTYPE_LinearColor
   sortPriority=0
   texture=/Game/Global/Textures/Default/Tileable_Patterns/T_Water_TilingNormal_With_Height_02_Softened
  }
  n14 = VectorParameter() {group=MainWave, parameterName=MainWave_UVControl, channelNames {a=VPan, b=UPan, g=VScale, r=UScale}, defaultValue=<rgba 1 1 0 0>}
  n15 = MaterialFunctionCall(Texture=@n13, UVControl_RGB=@n14, UVControl_A=@n14, "[Override]TimeInput"=NamedRerouteUsage(), "[Override]UVInput"=NamedRerouteUsage()) {materialFunction=&FXCommon_TextureSample}
  n16 = Constant3Vector() {constant=<rgba 0 0 1 0>}
  n17 = QualitySwitch(Default=@n15.RGB, Low=@n16, High=@n15.RGB, Medium=@n15.RGB, Epic=@n15.RGB)
  n18 = ShadingPathSwitch(Default=@n17, Deferred=@n17, Forward=@n16, Mobile=@n16)
  n19 = Constant3Vector() {constant=<rgba 0 0 1 1>}
  n20 = Abs(None=DotProduct(A=@n18, B=Normalize(VectorInput=@n19)))
  n21 = NamedRerouteUsage()
  n22 = Multiply(A=Power(Base=OneMinus(None=@n20), Exp=Constant() {r=10}), B=Power(Base=OneMinus(None=@n21)) {constExponent=3})
  n23 = DotProduct(A=CameraVectorWS(), B=Transform(None=Constant3Vector() {constant=<rgba 0 0 1 0>}))
  n24 = Constant() {r=1}
  n25 = MaterialFunctionCall(Input=OneMinus(None=Abs(None=@n23)), Input Low=Constant(), Input High=@n24, Target Low=Constant() {r=0.3}, Target High=@n24) {materialFunction=&RemapValueRange}
  n26 = NamedRerouteUsage()
  n27 = LinearInterpolate(
   A=@n26
   B=StaticSwitchParameter(True=@n11, False=@n9) {parameterName=Golden Waterfall?}
   Alpha=Multiply(A=Ceil(None=@n12), B=Multiply(A=@n22, B=@n25.Result))
  )
  n28 = QualitySwitch(Default=@n27, Low=@n26, High=@n27, Medium=@n27, Epic=@n27)
  n29 = ShadingPathSwitch(Default=@n28, Deferred=@n28, Forward=@n26, Mobile=@n26)
  n30 = NamedRerouteUsage()
  n31 = TextureObjectParameter() {group=PureFoam, parameterName=PureFoam_Texture, sortPriority=0}
  n32 = VectorParameter() {group=PureFoamNoise, parameterName=PureFoamNoise_UVControl, channelNames {a=VPan, b=UPan, g=VScale, r=UScale}, defaultValue=<rgba 1 1 0 0>}
  n33 = VectorParameter() {
   group=PureFoamNoise
   parameterName=PureFoamNoise_NoiseControl
   sortPriority=60
   channelNames {a=VStrength, b=UStrength, g=VOffset, r=UOffset}
   defaultValue=<rgba -0.5 -0.5 0.01 0.01>
  }
  n34 = MaterialFunctionCall!(
   NoiseValue=MaterialFunctionCall(
    Texture=TextureObjectParameter() {group=PureFoamNoise, parameterName=PureFoamNoise_Texture, sortPriority=0}
    UVControl_RGB=@n32
    UVControl_A=@n32
    "[Override]TimeInput"=NamedRerouteUsage()
    "[Override]UVInput"=NamedRerouteUsage()
   ) {materialFunction=&FXCommon_TextureSample}.R
   Control_RGB=@n33
   Control_A=@n33
  ) {materialFunction=/Game/Effects/Niagara/MasterMaterials/Functions/Common/FXCommon_NoiseControl}
  n35 = VectorParameter() {group=PureFoam, parameterName=PureFoam_UVControl, channelNames {a=VPan, b=UPan, g=VScale, r=UScale}, defaultValue=<rgba 1 1 0 0>}
  n36 = NamedRerouteUsage()
  n37 = NamedRerouteUsage()
  n38 = Multiply(
   A=Add(
    A=MaterialFunctionCall(Texture=@n31, AdditionalUVOffset=@n34.NoiseValue, UVControl_RGB=@n35, UVControl_A=@n35, "[Override]TimeInput"=@n36, "[Override]UVInput"=@n37) {materialFunction=&FXCommon_TextureSample}.R
    B=ScalarParameter() {group=PureFoam, parameterName=PureFoam_GrayAdjust_Add}
   )
   B=ScalarParameter() {defaultValue=1, group=PureFoam, parameterName=PureFoam_GrayAdjust_Mul}
  )
  n39 = Saturate(None=Power(Base=Saturate(None=@n38), Exp=ScalarParameter() {defaultValue=1, group=PureFoam, parameterName=PureFoam_GrayAdjust_Power}))
  n40 = Power(Base=NamedRerouteUsage(), Exp=ScalarParameter() {parameterName=whitecolor2c})
  NamedRerouteDeclaration(None=StaticSwitchParameter(True=LinearInterpolate(A=@n29, B=@n30, Alpha=@n39), False=LinearInterpolate(A=@n29, B=@n30, Alpha=@n40)) {parameterName=usePureMistMode}) {name=WhiteWater, nodeColor=<rgba 0.9688 1 0.981 1>}
  NamedRerouteDeclaration(None=MaterialFunctionCall(Normal=@n18, Flatness=ScalarParameter() {defaultValue=1, parameterName=FlatenNormal}) {materialFunction=&FlattenNormal}.Result) {name=FlatNormal, nodeColor=<rgba 0 0.0933 0.4948 1>}
  n41 = TextureObjectParameter() {group=WhiteWaterNoise, parameterName=WhiteWaterNoise_Texture, sortPriority=0}
  n42 = VectorParameter() {
   group=WhiteWaterNoise
   parameterName=WhiteWaterNoise_UVControl
   channelNames {a=VPan, b=UPan, g=VScale, r=UScale}
   defaultValue=<rgba 1 1 0 0>
  }
  n43 = NamedRerouteUsage()
  n44 = NamedRerouteUsage()
  n45 = VectorParameter() {
   group=WhiteWaterNoise
   parameterName=WhiteWaterNoise_NoiseControl
   sortPriority=60
   channelNames {a=VStrength, b=UStrength, g=VOffset, r=UOffset}
   defaultValue=<rgba -0.5 -0.5 0.01 0.01>
  }
  n46 = VectorParameter() {group=WhiteWaterMain, parameterName=WhiteWaterMain_UVControl, channelNames {a=VPan, b=UPan, g=VScale, r=UScale}, defaultValue=<rgba 1 1 0 0>}
  n47 = NamedRerouteUsage()
  n48 = MaterialFunctionCall(
   Texture=TextureObjectParameter() {group=WhiteWaterMain, parameterName=WhiteWaterMain_Texture, sortPriority=0}
   AdditionalUVOffset=MaterialFunctionCall(
    NoiseValue=MaterialFunctionCall(Texture=@n41, UVControl_RGB=@n42, UVControl_A=@n42, "[Override]TimeInput"=@n43, "[Override]UVInput"=@n44) {materialFunction=&FXCommon_TextureSample}.RG
    Control_RGB=@n45
    Control_A=@n45
   ) {materialFunction=&FXCommon_NoiseControl}.NoiseValue
   UVControl_RGB=@n46
   UVControl_A=@n46
   "[Override]TimeInput"=@n43
   "[Override]UVInput"=@n47
  ) {materialFunction=&FXCommon_TextureSample}
  n49 = VectorParameter() {parameterName=WhiteWaterMain, defaultValue=<rgba 1 1 1 0.5>}
  n50 = StaticSwitchParameter(True=LinearInterpolate(A=@n49.RGB, B=NamedRerouteUsage(), Alpha=NamedRerouteUsage()), False=@n49.A) {parameterName=Golden Waterfall?}
  n51 = VectorParameter() {group=WhiteWater2, parameterName=WhiteWater2_UVControl, channelNames {a=VPan, b=UPan, g=VScale, r=UScale}, defaultValue=<rgba 1 1 0 0>}
  n52 = MaterialFunctionCall(
   Texture=TextureObjectParameter() {group=WhiteWater2, parameterName=WhiteWater2_Texture, sortPriority=0}
   UVControl_RGB=@n51
   UVControl_A=@n51
   "[Override]TimeInput"=@n43
   "[Override]UVInput"=@n47
  ) {materialFunction=&FXCommon_TextureSample}
  n53 = VectorParameter() {parameterName=WhiteWater2Main, defaultValue=<rgba 1 1 1 0.5>}
  n54 = StaticSwitchParameter(True=LinearInterpolate(A=@n53.RGB, B=NamedRerouteUsage(), Alpha=NamedRerouteUsage()), False=@n53.RGB) {parameterName=Golden Waterfall?}
  n55 = Saturate(None=Add(
    A=MaterialFunctionCall(Texture=@n48.R, RGB=@n50, A=@n49.A) {materialFunction=&MF_FX_Refine}.Result
    B=MaterialFunctionCall(Texture=@n52.R, RGB=@n54, A=@n53.A) {materialFunction=&MF_FX_Refine}.Result
   ))
  n56 = Power(Base=@n55, Exp=ScalarParameter() {defaultValue=15, parameterName=WhiteWaterMaxPower})
  n57 = Add(
   A=MaterialFunctionCall(Float2=TextureCoordinate()) {materialFunction=&BreakOutFloat2Components}.G
   B=ScalarParameter() {parameterName=WhiteWaterMask_Add}
  )
  n58 = ScalarParameter() {defaultValue=1.5, parameterName=WhiteWaterMask_Mul}
  n59 = Power(Base=Saturate(None=Multiply(A=@n57, B=@n58)), Exp=ScalarParameter() {defaultValue=2, parameterName=WhiteWaterMask_Power})
  n60 = VectorParameter() {
   group=WhiteWaterMasksks
   parameterName=WhiteWaterMasksks_UVControl
   channelNames {a=VPan, b=UPan, g=VScale, r=UScale}
   defaultValue=<rgba 1 1 0 0>
  }
  n61 = Add(
   A=MaterialFunctionCall(
    Texture=TextureObjectParameter() {group=WhiteWaterMasksks, parameterName=WhiteWaterMasksks_Texture, sortPriority=0}
    UVControl_RGB=@n60
    UVControl_A=@n60
    "[Override]TimeInput"=NamedRerouteUsage()
    "[Override]UVInput"=NamedRerouteUsage()
   ) {materialFunction=&FXCommon_TextureSample}.R
   B=ScalarParameter() {group=WhiteWaterMasksks, parameterName=WhiteWaterMasksks_GrayAdjust_Add}
  )
  n62 = ScalarParameter() {defaultValue=1, group=WhiteWaterMasksks, parameterName=WhiteWaterMasksks_GrayAdjust_Mul}
  n63 = Power(
   Base=Saturate(None=Multiply(A=@n61, B=@n62))
   Exp=ScalarParameter() {defaultValue=1, group=WhiteWaterMasksks, parameterName=WhiteWaterMasksks_GrayAdjust_Power}
  )
  NamedRerouteDeclaration(None=LinearInterpolate(A=LinearInterpolate(A=@n56, B=@n55, Alpha=@n59), B=@n55, Alpha=Saturate(None=@n63))) {name=Whitewater_Mix, nodeColor=<rgba 0.7197 0.8039 1 1>}
  n64 = VectorParameter() {
   group=EmissiveAreaTop
   parameterName=Emissive_NoiseControl
   sortPriority=60
   channelNames {a=VStrength, b=UStrength, g=VOffset, r=UOffset}
   defaultValue=<rgba 0 -0.5 0 0.07>
  }
  n65 = Constant()
  n66 = QualitySwitch(
   Default=StaticSwitchParameter(
    True=MaterialFunctionCall(NoiseValue=@n18, Control_RGB=@n64, Control_A=@n64) {materialFunction=&FXCommon_NoiseControl}.NoiseValue
    False=@n65
   ) {parameterName=Update5/10/2024}
   Low=@n65
  )
  n67 = ShadingPathSwitch(Default=@n66, Deferred=@n66, Forward=@n65, Mobile=@n65)
  n68 = MaterialFunctionCall(Float2=Add(A=@n67, B=NamedRerouteUsage())) {materialFunction=&BreakOutFloat2Components}
  n69 = VectorParameter() {
   group=EmissiveArea2
   parameterName=EmissveSideMask_EdgeMask_Control
   channelNames {a="", b=Power, g=Multiply, r=Offset}
   defaultValue=<rgba 0 3 0.5 1>
  }
  n70 = NamedRerouteUsage()
  n71 = MaterialFunctionCall(Float2=@n67) {materialFunction=&BreakOutFloat2Components}
  n72 = ScalarParameter() {parameterName=EmissiveAreaOffset}
  n73 = NamedRerouteUsage()
  n74 = Divide(A=Add(A=Add(A=@n70, B=@n71.G), B=Divide(A=@n72, B=@n73)), B=Divide(A=ScalarParameter() {parameterName=EmissiveAreaLength}, B=@n73))
  n75 = Multiply(
   A=OneMinus(None=Saturate(None=@n74))
   B=ScalarParameter() {defaultValue=1, group=EmissiveAreaTop, parameterName=EmissiveAreaTop_GrayAdjust_Mul}
  )
  n76 = OneMinus(None=Power(Base=Saturate(None=@n75), Exp=ScalarParameter() {defaultValue=1, group=EmissiveAreaTop, parameterName=EmissiveAreaTop_GrayAdjust_Power}))
  n77 = OneMinus(None=@n68.G)
  n78 = ScalarParameter() {group=EmissiveArea2, parameterName=EmissiveArea2_GrayAdjust_Add}
  n79 = Saturate(None=Multiply(A=Add(A=@n77, B=@n78), B=ScalarParameter() {defaultValue=1, group=EmissiveArea2, parameterName=EmissiveArea2_GrayAdjust_Mul}))
  n80 = ScalarParameter() {defaultValue=1, group=EmissiveArea2, parameterName=EmissiveArea2_GrayAdjust_Power}
  n81 = LinearInterpolate(B=LinearInterpolate(A=NamedRerouteUsage(), Alpha=Saturate(None=@n76)) {constB=0}, Alpha=Saturate(None=Power(Base=@n79, Exp=@n80)))
  n82 = StaticSwitchParameter(
   True=Multiply(
    A=MaterialFunctionCall!(UorV=@n68.R, EdgeControl=@n69) {materialFunction=/Game/Effects/Niagara/MasterMaterials/Functions/Common/FXCommon_EdgeMask_Simple}.Mask
    B=@n81
   )
   False=@n81
  ) {parameterName=Update5/10/2024}
  n83 = Multiply(A=NamedRerouteUsage(), B=NamedRerouteUsage())
  n84 = Multiply(A=NamedRerouteUsage(), B=NamedRerouteUsage())
  n85 = Multiply(A=@n84, B=@n84)
  n86 = NamedRerouteUsage()
  n87 = Add(A=Add(A=Add(A=@n82, B=@n83), B=@n85), B=Multiply(A=@n85, B=Multiply(A=@n86) {constB=8}))
  n88 = RayTracingQualitySwitch(Normal=@n87, RayTraced=Multiply(A=@n87, B=ScalarParameter() {defaultValue=1, parameterName=LumenEmissiveMultiplier}))
  n89 = NamedRerouteUsage()
  NamedRerouteDeclaration(None=StaticSwitchParameter(True=LinearInterpolate(A=@n82, B=@n88, Alpha=@n89), False=@n82) {parameterName=Golden Waterfall?}) {name=Emissive, nodeColor=<rgba 0 1 0.9592 1>}
  n90 = CollectionParameter() {collection=&MPC_GoldenWatefall, parameterName=FallEmitColor}
  n91 = StaticSwitchParameter(
   True=VectorParameter() {parameterName=BPInput_EmissiveColor, defaultValue=<rgba 0 0 0 1>}.RGB
   False=StaticSwitchParameter(
    True=ComponentMask(None=@n90) {b=true, g=true, r=true}
    False=VectorParameter() {parameterName=FallEmitColor, defaultValue=<rgba 0 0 0 1>}.RGB
   ) {desc=CH7S1 GoldenWaterfall Update, parameterName=UseMPCforWaterfallColors}
  ) {parameterName=Update5/10/2024}
  n92 = VectorParameter() {parameterName=Golden Emissive Color, defaultValue=<rgba 1 0.2963 0 1>}
  n93 = NamedRerouteUsage()
  NamedRerouteDeclaration(None=StaticSwitchParameter(True=LinearInterpolate(A=@n91, B=@n92.RGB, Alpha=@n93), False=@n91) {parameterName=Golden Waterfall?}) {name=EmissiveColor, nodeColor=<rgba 0.9688 1 0 1>}
  n94 = CollectionParameter() {collection=&MPC_GoldenWatefall, parameterName=FallBaseColor}
  n95 = StaticSwitchParameter(
   True=VectorParameter() {parameterName=BPInput_BaseColor, defaultValue=<rgba 0 0 0 1>}.RGB
   False=StaticSwitchParameter(
    True=ComponentMask(None=@n94) {b=true, g=true, r=true}
    False=VectorParameter() {parameterName=FallBaseColor, defaultValue=<rgba 0 0 0 1>}.RGB
   ) {desc=CH7S1 GoldenWaterfall Update, parameterName=UseMPCforWaterfallColors}
  ) {parameterName=Update5/10/2024}
  n96 = VectorParameter() {parameterName=Golden Base Color, defaultValue=<rgba 0.25 0.2006 0 1>}
  n97 = NamedRerouteUsage()
  NamedRerouteDeclaration(None=StaticSwitchParameter(True=LinearInterpolate(A=@n95, B=@n96.RGB, Alpha=@n97), False=@n95) {parameterName=Golden Waterfall?}) {name=BaseColor, nodeColor=<rgba 0.9688 1 0 1>}
  n98 = StaticSwitchParameter(
   True=VectorParameter() {parameterName=BPInput_FoamColor, defaultValue=<rgba 0 0 0 1>}.RGB
   False=VectorParameter() {parameterName=BaseColor2, defaultValue=<rgba 0 0 0 1>}.RGB
  ) {parameterName=Update5/10/2024}
  n99 = VectorParameter() {parameterName="Golden Foam Color  2", defaultValue=<rgba 0 0 0 1>}
  n100 = NamedRerouteUsage()
  n101 = StaticSwitchParameter(
   True=NamedRerouteDeclaration(None=LinearInterpolate(A=@n98, B=@n99.RGB, Alpha=@n100)) {name=FoamColor, nodeColor=<rgba 0.9688 1 0 1>}
   False=VectorParameter() {parameterName=FoamC, defaultValue=<rgba 0.9062 1 1 1>}.RGB
  ) {parameterName=Update5/10/2024}
  n102 = VectorParameter() {parameterName=Golden Foam Color, defaultValue=<rgba 0.6354 0.5558 0.3078 1>}
  n103 = NamedRerouteUsage()
  NamedRerouteDeclaration(None=StaticSwitchParameter(True=LinearInterpolate(A=@n101, B=@n102.RGB, Alpha=@n103), False=@n101) {parameterName=Golden Waterfall?}) {name=FoamColor_ForTopFoam, nodeColor=<rgba 1 0.7031 0 1>}
  n104 = LocalPosition()
  n105 = Constant() {r=4000}
  n106 = Subtract(A=Divide(A=Add(A=@n104.Z, B=@n105), B=Constant() {r=2500}), B=Multiply(A=NamedRerouteUsage(), B=Constant() {r=6.0746}))
  n107 = Multiply(A=@n106, B=CollectionParameter() {collection=&MPC_GoldenWatefall, parameterName=TransitionAlpha})
  NamedRerouteDeclaration(None=StaticSwitchParameter(True=Saturate(None=@n107), False=Constant()) {parameterName=Golden Waterfall?}) {name=Golden Transition, nodeColor=<rgba 0.2891 1 0 1>}
  NamedRerouteDeclaration(None=Multiply(A=Time() {bOverride_Period=true, period=65555}, B=ScalarParameter() {defaultValue=0.5, parameterName=GlobalTimeScale})) {name=ScaledTime, nodeColor=<rgba 1 0.7734 0 1>}
  n108 = VectorParameter() {parameterName=Spark Color, defaultValue=<rgba 1 0.5373 0.0769 1>}
  n109 = Subtract(A=PixelDepth().R, B=ScalarParameter() {defaultValue=2000, parameterName=GoldEmissive DistFadeoffset})
  n110 = ScalarParameter() {defaultValue=15000, parameterName=GoldEmissive DistFade}
  n111 = OneMinus(None=Saturate(None=Divide(A=@n109, B=@n110)))
  n112 = ScalarParameter() {defaultValue=40, parameterName=GoldCellSize}
  n113 = MaterialFunctionCall!("Normal (WS)", Texture Scale=@n112, Triplanar Fade, Derivatives Scale) {materialFunction=/Game/Athena/Environments/MaterialFunctions/MF_TriplanarDithered}
  n114 = ComponentMask(None=@n113.UVs) {g=true}
  n115 = Multiply(A=Multiply(A=Time(), B=ScalarParameter() {defaultValue=8, parameterName=Spark Fall Speed})) {constB=-1}
  n116 = ComponentMask(None=AppendVector(A=ComponentMask(None=@n113.UVs) {r=true}, B=Add(A=@n114, B=@n115))) {g=true, r=true}
  n117 = Frac(None=@n116)
  n118 = Constant2Vector() {g=1, r=3}
  n119 = Divide(A=WorldPosition() {worldPositionShaderOffset=WPT_ExcludeAllShaderOffsets}.XYZ, B=@n112)
  n120 = ComponentMask(None=@n119) {r=true}
  n121 = ComponentMask(None=@n119) {g=true}
  n122 = ComponentMask(None=@n119) {b=true}
  n123 = Multiply(A=Time(), B=ScalarParameter() {defaultValue=8, parameterName=Spark Fall Speed})
  n124 = VectorNoise(World Position=AppendVector(A=AppendVector(A=@n120, B=@n121), B=Add(A=@n122, B=@n123)))
  n125 = ComponentMask(None=@n124) {r=true}
  n126 = TextureSample!(
   UVs=Rotator(Coordinate=Multiply(A=@n117, B=@n118), Time=Multiply(A=@n125) {constB=33.3})
   Tex
   "DDX(UVs)"=DDX(Value=@n113.DDX)
   "DDY(UVs)"=DDY(Value=@n113.DDY)
   Apply View MipBias
  ) {mipValueMode=TMVM_Derivative, samplerType=SAMPLERTYPE_LinearGrayscale, texture=/FortWater/Materials/WaterSurface/Hermes/Gold/Gold_Flare_M}
  n127 = Multiply(A=ScalarParameter() {defaultValue=0.1, parameterName=GoldSparkleSpeed}, B=Time())
  n128 = Multiply(A=ComponentMask(None=@n124) {b=true, g=true, r=true}, B=ScalarParameter() {defaultValue=13516.0879, parameterName=GoldEmissive Offset})
  n129 = Saturate(None=Sine(None=Add(A=@n127, B=@n128)))
  n130 = ComponentMask(None=@n129) {r=true}
  n131 = Multiply(A=Multiply(A=Multiply(A=@n126.R, B=@n130), B=ComponentMask(None=@n129) {g=true}), B=ComponentMask(None=@n129) {b=true})
  n132 = ScalarParameter() {defaultValue=1, parameterName=GoldEmissiveSparkle Radius}
  n133 = LinearInterpolate(
   A=ScalarParameter() {defaultValue=50, parameterName=GoldEmissive Sparkleh}
   B=ScalarParameter() {defaultValue=2, parameterName=GoldEmissive Sparkle}
   Alpha=Saturate(None=Multiply(A=@n132, B=@n130))
  )
  n134 = Multiply(A=ScalarParameter() {defaultValue=30, group=Gold Liquid, parameterName=Gold_Emissive_Sparkle}, B=Saturate(None=Power(Base=@n131, Exp=@n133)))
  n135 = LinearInterpolate(
   B=Multiply(A=NamedRerouteDeclaration(None=@n111) {name=PixelDepthFadeClose, nodeColor=<rgba 0.5938 0 1 1>}, B=Multiply(A=@n134, B=@n134))
   Alpha=NamedRerouteUsage()
  )
  NamedRerouteDeclaration(None=StaticSwitchParameter(True=Multiply(A=@n108.RGB, B=@n135), False=Constant()) {parameterName=Golden Waterfall?}) {name=Gold Spark, nodeColor=<rgba 0.6406 1 0 1>}
  NamedRerouteDeclaration(None=VectorParameter() {parameterName=Golden WW Color 1, defaultValue=<rgba 0.5573 0.5573 0.5573 1>}.RGB) {name=WhiteW Color 1, nodeColor=<rgba 1 0.4746 0 1>}
  NamedRerouteDeclaration(None=VectorParameter() {parameterName=Golden WW Color 2, defaultValue=<rgba 0.5457 0.5457 0.5457 1>}.RGB) {name=WhiteW Color 2, nodeColor=<rgba 1 0.4746 0 1>}
  NamedRerouteDeclaration(None=Multiply(A=@n106, B=@n106)) {name=Core Mask, nodeColor=<rgba 0 0.2266 1 1>}
  NamedRerouteDeclaration(None=@n10.RGB) {name=Sub surface color gold, nodeColor=<rgba 0 0.3906 1 1>}
  n136 = TextureCoordinate()
  n137 = MaterialFunctionCall(Float2=@n136) {materialFunction=&BreakOutFloat2Components}
  n138 = MaterialFunctionCall(X=OneMinus(None=Frac(None=@n137.R)), Y=@n137.G) {materialFunction=&MakeFloat2}
  NamedRerouteDeclaration(None=If(
    A=MaterialFunctionCall(Float2=@n136) {materialFunction=&BreakOutFloat2Components}.R
    "A > B"=@n138.Result
    "A == B"=@n138.Result
    "A < B"=@n136
   ) {constB=1}) {name=CorrectedUV, nodeColor=<rgba 0.9922 1 0 1>}
  n139 = NamedRerouteUsage()
  n140 = NamedRerouteUsage()
  n141 = Multiply(
   A=Divide(B=Divide(A=@n139, B=@n140))
   B=MaterialFunctionCall(Float2=NamedRerouteUsage()) {materialFunction=&BreakOutFloat2Components}.G
  )
  NamedRerouteDeclaration(None=OneMinus(None=Saturate(None=@n141))) {name=ExtendArea, nodeColor=<rgba 0.6961 1 0 1>}
  n142 = Constant2Vector() {r=0.5}
  n143 = Add(A=NamedRerouteUsage(), B=Multiply(A=@n142) {constB=-1})
  n144 = NamedRerouteUsage()
  n145 = NamedRerouteUsage()
  n146 = Multiply(
   A=AppendVector(A=Divide(A=@n144) {constA=500, constB=2500}, B=Divide(A=@n145) {constA=500, constB=3500})
   B=ScalarParameter() {defaultValue=1, parameterName=CustomParameter_TextureDensity}
  )
  NamedRerouteDeclaration(None=Add(A=Multiply(A=@n143, B=@n146), B=@n142)) {name=NoRemapedTiledUV, nodeColor=<rgba 0 1 0.0391 1>}
  NamedRerouteDeclaration(None=ScalarParameter() {defaultValue=500, group="[IgnoreThisGroup]BPInput", parameterName=BPInput_Width}) {name=Width, nodeColor=<rgba 1 0 0.6328 1>}
  NamedRerouteDeclaration(None=ScalarParameter() {defaultValue=500, group="[IgnoreThisGroup]BPInput", parameterName=BPInput_Length}) {name=Length, nodeColor=<rgba 1 0 0.1855 1>}
  NamedRerouteDeclaration(None=ScalarParameter() {group="[IgnoreThisGroup]BPInput", parameterName=BPInput_BevelLength}) {name=BevelLength, nodeColor=<rgba 1 0.9609 0 1>}
  NamedRerouteDeclaration(None=@n3) {name=AreaMask_Bevel, nodeColor=<rgba 0 1 0.7891 1>}
  n147 = Add(A=NamedRerouteUsage(), B=NamedRerouteUsage())
  n148 = NamedRerouteUsage()
  n149 = MaterialFunctionCall(Float2=NamedRerouteUsage()) {materialFunction=&BreakOutFloat2Components}
  n150 = Multiply(A=Divide(B=Divide(A=@n147, B=@n148)), B=@n149.G)
  NamedRerouteDeclaration(None=OneMinus(None=Saturate(None=@n150))) {name=ExtendAndBevelArea, nodeColor=<rgba 0 1 0.3633 1>}
  n151 = NamedRerouteUsage()
  n152 = Add(A=NamedRerouteUsage(), B=NamedRerouteUsage())
  n153 = Divide(B=Divide(A=Subtract(A=@n151, B=@n152), B=@n151))
  n154 = Add(A=Multiply(A=Divide(A=@n152, B=@n151)) {constB=-1}, B=@n149.G)
  NamedRerouteDeclaration(None=Saturate(None=Multiply(A=@n153, B=@n154))) {name=FallArea, nodeColor=<rgba 0.3594 1 0 1>}
  n155 = NamedRerouteUsage()
  n156 = ScalarParameter() {parameterName=WhiteWater}
  n157 = Power(Base=Saturate(None=NamedRerouteUsage())) {constExponent=5}
  n158 = ScalarParameter() {defaultValue=0.225, group=Water Shading Model, parameterName=Specular}
  n159 = ScalarParameter() {defaultValue=0.05, group=Water Shading Model, parameterName=Roughness}
  n160 = ScalarParameter() {group=Water Shading Model, parameterName=Surface Opacity}
  n161 = SetMaterialAttributes!(
   MaterialAttributes
   Base Color=VectorParameter() {parameterName=BaseColor, defaultValue=<rgba 0 0 0 1>}.RGB
   Specular=@n158
   Roughness=@n159
   Opacity=@n160
   Normal
   World Position Offset
  ) {attributeSetTypes [
    #1
    #2
    #3
    #4
    #5
    #6
   ]}
  n162 = Constant3Vector() {constant=<rgba -1 0 0 0>}
  n163 = VectorParameter() {parameterName=Test, defaultValue=<rgba 500 0 500 200>}
  n164 = MaterialFunctionCall(Float3=Transform(None=Multiply(A=@n162, B=@n163.R)) {transformSourceType=TRANSFORMSOURCE_Local}) {materialFunction=&BreakOutFloat3Components}
  n165 = MaterialFunctionCall!(
   Attributes=@n161
   Transition
   Sim Foam Erasure
   Velocity=MaterialFunctionCall(R=@n164.R, G=@n164.R, B=@n163.B) {materialFunction=&AppendMany}.RGB
   Far Mask Override
   Depth=@n163.A
  ) {materialFunction=/FortWater/Materials/Functions/FortWaterRiverFlowmaps}
  n166 = MakeMaterialAttributes!(
   BaseColor
   Metallic
   Specular=LinearInterpolate(A=@n158, Alpha=@n165.Foam Mask) {constB=0.5}
   Roughness=LinearInterpolate(A=@n159, Alpha=@n165.Foam Mask) {constB=0.5}
   Anisotropy
   EmissiveColor
   Opacity=LinearInterpolate(A=@n160, Alpha=@n165.Foam Mask)
   OpacityMask
   Normal=LinearInterpolate(B=@n165.Flowmap Normals, Alpha=@n165.Foam Mask)
   Tangent
   WorldPositionOffset=@n165.Flowmap WPO
   SubsurfaceColor
   ClearCoat
   ClearCoatRoughness
   AmbientOcclusion
   Refraction
   CustomizedUVs_0
   CustomizedUVs_1
   CustomizedUVs_2
   CustomizedUVs_3
   CustomizedUVs_4
   CustomizedUVs_5
   CustomizedUVs_6
   CustomizedUVs_7
   PixelDepthOffset
   ShadingModel
   Displacement
  )
  n167 = QualitySwitch(Default=@n165.Attributes, Low=@n166, High=@n165.Attributes, Medium=@n166, Epic=@n165.Attributes)
  n168 = BreakMaterialAttributes!(Attr=ShadingPathSwitch(Default=@n167, Deferred=@n167, Forward=@n166, Mobile=@n166))
  n169 = TextureCoordinate()
  n170 = Add(
   A=OneMinus(None=MaterialFunctionCall(Float2=@n169) {materialFunction=&BreakOutFloat2Components}.G)
   B=ScalarParameter() {group=whitewater3, parameterName=whitewater3_GrayAdjust_Add}
  )
  n171 = ScalarParameter() {defaultValue=1, group=whitewater3, parameterName=whitewater3_GrayAdjust_Mul}
  n172 = Power(
   Base=Saturate(None=Multiply(A=@n170, B=@n171))
   Exp=ScalarParameter() {defaultValue=1, group=whitewater3, parameterName=whitewater3_GrayAdjust_Power}
  )
  n173 = Saturate(None=Multiply(A=NamedRerouteUsage()) {constB=3})
  n174 = Saturate(None=Multiply(A=Saturate(None=@n172), B=Power(Base=@n173) {constExponent=1.5}))
  n175 = ScalarParameter() {defaultValue=5, parameterName=FresE}
  n176 = ScalarParameter() {parameterName=WhiteWaterRoughness}
  n177 = NamedRerouteUsage()
  n178 = MaterialFunctionCall(In=Saturate(None=Multiply(A=@n177) {constB=1.5}), Contrast=Constant() {r=0.41}) {materialFunction=&CheapContrast}
  n179 = NamedRerouteUsage()
  n180 = SetMaterialAttributes!(
   MaterialAttributes
   Base Color=LinearInterpolate(A=LinearInterpolate(A=@n155, B=@n156, Alpha=@n157), B=NamedRerouteUsage(), Alpha=LinearInterpolate(B=@n168.Opacity, Alpha=@n174))
   Specular=LinearInterpolate(A=Fresnel(ExponentIn=@n175), B=@n168.Specular, Alpha=@n174)
   Roughness=LinearInterpolate(A=LinearInterpolate(A=@n176, Alpha=@n178.Result), B=@n168.Roughness, Alpha=@n174)
   Emissive Color=NamedRerouteUsage()
   Opacity Mask=NamedRerouteUsage()
   Normal=Transform(None=LinearInterpolate(A=@n179, B=@n168.Normal, Alpha=@n174))
   World Position Offset=LinearInterpolate(A=NamedRerouteUsage(), Alpha=@n174) {constB=0}
  ) {attributeSetTypes [
    #1
    #2
    #3
    #7
    #8
    #5
    #6
   ]}
  NamedRerouteDeclaration(None=@n180) {name=HighEnd Material, nodeColor=<rgba 0.2656 1 0 1>}
  n181 = MaterialFunctionCall() {materialFunction=/FortWater/Materials/WaterSurface/Mobile/Functions/MF_MobileWaterfall}
  n182 = GetMaterialAttributes(None=@n181.Result) {attributeGetTypes [#1, #8, #7]}
  n183 = Multiply(A=Power(Base=Multiply(A=@n182.Base Color, B=@n182.Base Color)) {constExponent=5}, B=NamedRerouteUsage())
  n184 = Multiply(A=Desaturation(None=Saturate(None=@n183)), B=VectorParameter() {parameterName=Mobile Gold Color, defaultValue=<rgba 1 0.4685 0.1667 1>}.RGB)
  n185 = ScalarParameter() {defaultValue=50}
  n186 = LinearInterpolate(A=@n182.Emissive Color, B=Add(A=Multiply(A=@n184, B=@n185) {constB=15}, B=NamedRerouteUsage()), Alpha=NamedRerouteUsage())
  n187 = VectorParameter() {defaultValue=<rgba 0.4925 0.796 0.9271 1>}
  n188 = NamedRerouteUsage()
  n189 = LinearInterpolate(
   A=@n182.Base Color
   B=Multiply(A=Multiply(A=@n182.Base Color, B=@n187.RGB), B=LinearInterpolate(Alpha=@n188) {constA=1, constB=0.7})
   Alpha=NamedRerouteUsage()
  )
  n190 = NamedRerouteUsage()
  NamedRerouteDeclaration(None=StaticSwitchParameter(
    True=SetMaterialAttributes!(MaterialAttributes=@n182.MaterialAttributes, Emissive Color=@n186, Base Color=@n189, Opacity Mask=@n190) {attributeSetTypes [#7, #1, #8]}
    False=@n181.Result
   ) {parameterName=Golden Waterfall?}) {name=Mobile Material, nodeColor=<rgba 0 0.6016 1 1>}
  n191 = NamedRerouteUsage()
  n192 = MaterialFunctionCall(Float2=TextureCoordinate()) {materialFunction=&BreakOutFloat2Components}
  n193 = LinearInterpolate(
   B=Power(Base=OneMinus(None=@n191), Exp=ScalarParameter() {defaultValue=8, group=EdgeMaskTop, parameterName=BackfaceTopMask_Power})
   Alpha=Floor(None=@n192.R)
  ) {constA=1}
  n194 = VectorParameter() {
   group=EdgeMaskTopNoise
   parameterName=EdgeMaskTopNoise_UVControl
   channelNames {a=VPan, b=UPan, g=VScale, r=UScale}
   defaultValue=<rgba 1 1 0 0>
  }
  n195 = MaterialFunctionCall(
   Texture=TextureObjectParameter() {group=EdgeMaskTopNoise, parameterName=EdgeMaskTopNoise_Texture, sortPriority=0}
   UVControl_RGB=@n194
   UVControl_A=@n194
   "[Override]TimeInput"=NamedRerouteUsage()
   "[Override]UVInput"=NamedRerouteUsage()
  ) {materialFunction=&FXCommon_TextureSample}
  n196 = AppendVector(A=@n195.R, B=@n195.R)
  n197 = VectorParameter() {
   group=EdgeMaskTopNoise
   parameterName=EdgeMaskTopNoise_NoiseControl
   sortPriority=60
   channelNames {a=VStrength, b=UStrength, g=VOffset, r=UOffset}
   defaultValue=<rgba -0.5 -0.5 0.01 0.01>
  }
  n198 = VectorParameter() {
   group=EdgeMaskTop
   parameterName=EdgeMaskTop_EdgeMask_InvertControl
   channelNames {a="", b=Power, g=Multiply, r=Offset}
   defaultValue=<rgba 0 1 1 1>
  }
  n199 = VectorParameter() {
   group=EdgeMaskTop
   parameterName=EdgeMaskTop_EdgeMask_Control
   channelNames {a="", b=Power, g=Multiply, r=Offset}
   defaultValue=<rgba 0 1 1 1>
  }
  n200 = MaterialFunctionCall!(
   UorV=Add(
    A=@n192.G
    B=MaterialFunctionCall(NoiseValue=@n196, Control_RGB=@n197, Control_A=@n197) {materialFunction=&FXCommon_NoiseControl}.NoiseValue
   )
   Invert_EdgeControl=@n198
   NoInvert_EdgeControl=@n199
  ) {materialFunction=/Game/Effects/Niagara/MasterMaterials/Functions/Common/FXCommon_EdgeMask_Ex}
  n201 = VectorParameter() {
   group=EdgeMaskSide
   parameterName=EdgeMaskSide_EdgeMask_Control
   channelNames {a="", b=Power, g=Multiply, r=Offset}
   defaultValue=<rgba 0 1 1 1>
  }
  n202 = VectorParameter() {
   group=EdgeMaskSide
   parameterName=EdgeMaskSide_EdgeMask_ExtendAreaControl
   channelNames {a="", b=Power, g=Multiply, r=Offset}
   defaultValue=<rgba 0 1 1 1>
  }
  n203 = NamedRerouteUsage()
  n204 = LinearInterpolate(A=@n201, B=@n202, Alpha=Saturate(None=Multiply(A=@n202.A, B=@n203)))
  n205 = Multiply(
   A=Multiply(A=Multiply(A=@n193, B=@n200.Mask), B=@n4.R)
   B=LinearInterpolate(
    A=MaterialFunctionCall(UorV=@n192.R, Invert_EdgeControl=@n204, NoInvert_EdgeControl=@n204) {materialFunction=&FXCommon_EdgeMask_Ex}.Mask
    Alpha=ScalarParameter() {parameterName=DisableEdgeMaskSideForDoubleFaceWaterfall}
   )
  ) {constA=1}
  n206 = Constant() {r=0.99}
  n207 = Constant() {r=1}
  n208 = MaterialFunctionCall!(Alpha Threshold=@n205, Random) {materialFunction=/Game/TimeOfDay/CloudsAndStorm/Materials/MaterialFunctions/DitherTemporalAABlue}
  n209 = Constant()
  NamedRerouteDeclaration(None=If(A=@n205, "A > B"=If(A=@n205, B=@n206, "A > B"=@n207, "A == B"=@n208.Result, "A < B"=@n208.Result), "A == B"=@n209, "A < B"=@n209) {constB=0.1}) {name=OpacityMask, nodeColor=<rgba 1 0 0.9375 1>}
  n210 = Constant()
  n211 = TextureObjectParameter() {group=WPO, parameterName=WPO_Texture, sortPriority=0}
  n212 = VectorParameter() {group=WPO, parameterName=WPO_UVControl, channelNames {a=VPan, b=UPan, g=VScale, r=UScale}, defaultValue=<rgba 1 1 0 0>}
  n213 = NamedRerouteUsage()
  n214 = NamedRerouteUsage()
  n215 = AppendVector(A=Ceil(None=Divide(A=@n214) {constA=500, constB=2500}), B=Divide(A=NamedRerouteUsage()) {constA=500, constB=3500})
  n216 = Constant2Vector() {r=0.5}
  n217 = Add(A=NamedRerouteUsage(), B=Multiply(A=@n216) {constB=-1})
  n218 = LinearInterpolate(
   A=NamedRerouteUsage()
   B=Add(A=Multiply(A=@n215, B=@n217), B=@n216)
   Alpha=ScalarParameter() {parameterName=DisableEdgeMaskSideForDoubleFaceWaterfall}
  )
  n219 = Multiply(
   A=Add(
    A=MaterialFunctionCall(Texture=@n211, UVControl_RGB=@n212, UVControl_A=@n212, "[Override]TimeInput"=@n213, "[Override]UVInput"=@n218) {materialFunction=&FXCommon_TextureSample}.R
    B=ScalarParameter() {group=WPO, parameterName=WPO_GrayAdjust_Add}
   )
   B=ScalarParameter() {defaultValue=1, group=WPO, parameterName=WPO_GrayAdjust_Mul}
  )
  n220 = QualitySwitch(Default=@n219, Low=@n210, High=@n219, Medium=@n219, Epic=@n219)
  NamedRerouteDeclaration(None=Multiply(A=VertexNormalWS(), B=ShadingPathSwitch(Default=@n220, Deferred=@n220, Forward=@n210, Mobile=@n210)) {constA=1}) {name=WPO, nodeColor=<rgba 0 1 0.3438 1>}
  NamedRerouteUsage()
  NamedRerouteUsage()
  NamedRerouteUsage()
  NamedRerouteUsage()
  NamedRerouteUsage()
  Power()
  Reroute(None=@n116)
 ]
}
