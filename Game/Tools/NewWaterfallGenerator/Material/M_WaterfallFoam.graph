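# Translucent waterfall foam material for the waterfall generator, with emissive foam color tinted by time of day.
# Vertex color encodes flow direction, speed and whitewater for the FortWater river flowmap foam mask, with alpha cubed as fade.
# An eight-tap dithered shading-model-ID blur confines foam to single-layer water, bypassed on Switch, Android and PC-Perf.
# A box mask cuts foam out of a set volume, and a platform-switched Z offset lifts it.
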Material /Game/Tools/NewWaterfallGenerator/Material/M_WaterfallFoam
graph {
 properties {
  blendMode=BLEND_Translucent
  pixelDepthOffsetMode=PDOM_Legacy
  translucencyLightingMode=TLM_VolumetricPerVertexNonDirectional
  usedShadingModels=Default Lit
 }
 nodes [
  n1 = VectorParameter() {parameterName=FoamColor, defaultValue=<rgba 0.765 0.8775 0.9 1.2>}
  n2 = Multiply(A=@n1.RGB) {constB=0.2}
  n3 = VectorParameter() {parameterName=PinkC, defaultValue=<rgba 0.2386 0.1384 0.3 0.3>}
  n4 = LinearInterpolate(A=@n2, B=@n3.RGB, Alpha=@n3.RGB) {constAlpha=0.2}
  n5 = Multiply(A=@n1.RGB) {constB=0.1}
  n6 = VectorParameter() {parameterName=BlueC, defaultValue=<rgba 0.2389 0.3929 0.6052 0.2939>}
  MaterialFunctionCall(Emissive=MaterialFunctionCall(Color_Morning=@n4, Color_Day=@n2, Color_Evening=@n4, Color_Night=LinearInterpolate(A=@n5, B=@n6.RGB, Alpha=@n6.RGB) {constAlpha=0.2}) {materialFunction=&Color_Time_of_Day}.Result) {materialFunction=&MF_FX_EmissiveAdjustments}
  n7 = Constant()
  n8 = ScalarParameter() {defaultValue=25, parameterName=TopFoamZOffset}
  n9 = Constant() {r=7}
  MaterialFunctionCall(
   X=@n7
   Y=@n7
   Z=MaterialFunctionCall("Xb1,PS4"=@n8, "PC-High&&Epic,XSX,PS5"=@n8, "PC-Perf,Android"=@n9, Switch=@n9, "PC-Low(NotPerf)&&Medium"=@n8) {materialFunction=&PlatformSwitch_1D}.Result
  ) {materialFunction=&MakeFloat3}
  n10 = VertexColor()
  n11 = Cosine(None=@n10.R)
  n12 = Sine(None=@n10.R)
  n13 = ScalarParameter() {defaultValue=1500, parameterName=WaterfallMaxFlowVelocity}
  n14 = AppendVector(
   A=Multiply(A=AppendVector(A=@n11, B=@n12), B=Multiply(A=@n10.G, B=@n13))
   B=Multiply(A=@n10.B, B=ScalarParameter() {defaultValue=1300, parameterName=WaterfallMaxWhiteWater})
  )
  n15 = ScalarParameter() {defaultValue=200, parameterName=WaterfallDepth}
  n16 = Multiply(
   A=MaterialFunctionCall!(
    Attributes
    Transition
    Sim Foam Erasure
    Velocity=NamedRerouteDeclaration(None=@n14) {name=WaterfallSupport_VelocityAndLength, nodeColor=<rgba 1 0.4688 0 1>}
    Far Mask Override
    Depth=NamedRerouteDeclaration(None=@n15) {name=WaterfallSupport_WaterDepth, nodeColor=<rgba 0.7109 1 0 1>}
   ) {materialFunction=/FortWater/Materials/Functions/FortWaterRiverFlowmaps}.Foam Mask
   B=Power(Base=@n10.A) {constExponent=3}
  )
  n17 = Constant() {r=1}
  n18 = ScreenPosition()
  n19 = MaterialFunctionCall!(OffsetAmount=ScalarParameter() {defaultValue=0.07, parameterName=MaterialIDBlurScale}) {materialFunction=/Engine/Functions/Engine_MaterialFunctions01/ImageAdjustment/BlurSampleOffsets}
  MaterialExpressionScreenPosition_1 = ScreenPosition()
  n20 = Custom!(ScreenPos=@MaterialExpressionScreenPosition_1.PixelPosition) {
   code="float random = InterleavedGradientNoise(ScreenPos.xy, float(View.StateFrameIndexMod8));\r\nreturn random;"
   outputType=CMOT_Float1
   inputs [{
     inputName=ScreenPos
     input {expression=&self:MaterialExpressionScreenPosition_1, outputIndex=1}
    }]
  }
  n21 = Divide(A=@n20, B=ScalarParameter() {defaultValue=2, parameterName=DivideDither}) {constB=1.5}
  n22 = SceneTexture(UVs=Add(A=@n18.ViewportUV, B=Multiply(A=@n19."0,1", B=@n21))) {sceneTextureId=PPI_ShadingModelID}
  n23 = Constant()
  n24 = Constant() {r=1}
  n25 = SceneTexture(UVs=Add(A=@n18.ViewportUV, B=Multiply(A=@n19."1,0", B=@n21))) {sceneTextureId=PPI_ShadingModelID}
  n26 = Constant()
  n27 = Constant() {r=1}
  n28 = Add(A=@n18.ViewportUV, B=Multiply(A=@n19."0,-1", B=@n21))
  n29 = Constant()
  n30 = Add(
   A=Add(
    A=If(A=@n22.Color, "A > B"=@n23, "A == B"=@n24, "A < B"=@n23) {constB=10, equalsThreshold=0.01}
    B=If(A=@n25.Color, "A > B"=@n26, "A == B"=@n27, "A < B"=@n26) {constB=10, equalsThreshold=0.01}
   )
   B=If(A=SceneTexture(UVs=@n28) {sceneTextureId=PPI_ShadingModelID}.Color, "A > B"=@n29, "A == B"=Constant() {r=1}, "A < B"=@n29) {constB=10, equalsThreshold=0.01}
  )
  n31 = Multiply(A=@n19."-1,0", B=@n21)
  n32 = Constant()
  n33 = If(
   A=SceneTexture(UVs=Add(A=@n18.ViewportUV, B=@n31)) {sceneTextureId=PPI_ShadingModelID}.Color
   "A > B"=@n32
   "A == B"=Constant() {r=1}
   "A < B"=@n32
  ) {constB=10, equalsThreshold=0.01}
  n34 = SceneTexture(UVs=Add(A=@n18.ViewportUV, B=Multiply(A=@n19."-.5,.5", B=@n21))) {sceneTextureId=PPI_ShadingModelID}
  n35 = Constant()
  n36 = Constant() {r=1}
  n37 = Add(A=@n18.ViewportUV, B=Multiply(A=@n19.".5,.5", B=@n21))
  n38 = Constant()
  n39 = Add(
   A=Add(A=Add(A=@n30, B=@n33), B=If(A=@n34.Color, "A > B"=@n35, "A == B"=@n36, "A < B"=@n35) {constB=10, equalsThreshold=0.01})
   B=If(A=SceneTexture(UVs=@n37) {sceneTextureId=PPI_ShadingModelID}.Color, "A > B"=@n38, "A == B"=Constant() {r=1}, "A < B"=@n38) {constB=10, equalsThreshold=0.01}
  )
  n40 = Multiply(A=@n19.".5,-.5", B=@n21)
  n41 = Constant()
  n42 = If(
   A=SceneTexture(UVs=Add(A=@n18.ViewportUV, B=@n40)) {sceneTextureId=PPI_ShadingModelID}.Color
   "A > B"=@n41
   "A == B"=Constant() {r=1}
   "A < B"=@n41
  ) {constB=10, equalsThreshold=0.01}
  n43 = SceneTexture(UVs=Add(A=@n18.ViewportUV, B=Multiply(A=@n19."-.5,-.5", B=@n21))) {sceneTextureId=PPI_ShadingModelID}
  n44 = Constant()
  n45 = Constant() {r=1}
  n46 = Divide(A=Add(A=Add(A=@n39, B=@n42), B=If(A=@n43.Color, "A > B"=@n44, "A == B"=@n45, "A < B"=@n44) {constB=10, equalsThreshold=0.01})) {constB=8}
  n47 = MaterialFunctionCall("Xb1,PS4"=@n46, "PC-High&&Epic,XSX,PS5"=@n46, "PC-Perf,Android"=@n17, Switch=@n17, "PC-Low(NotPerf)&&Medium"=@n46) {materialFunction=&PlatformSwitch_1D}
  n48 = LocalPosition()
  n49 = TransformPosition(None=VectorParameter() {parameterName=BoxMaskCenterPos, defaultValue=<rgba 0 0 0 1>}.RGB) {transformSourceType=TRANSFORMPOSSOURCE_World}
  n50 = VectorParameter() {parameterName=BoxMaskExtendAndFalloff, defaultValue=<rgba 996.6113 1000 10000 100>}
  Multiply(
   A=Multiply(A=Multiply(A=@n16, B=@n47.Result), B=@n1.A)
   B=OneMinus(None=MaterialFunctionCall!(A=@n48.XYZ, B=@n49, Bounds=@n50.RGB, Edge Falloff=@n50.RGB) {materialFunction=/Engine/Functions/Engine_MaterialFunctions01/Gradient/BoxMask-3D}.Result)
  )
 ]
}
