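# Mobile water material function for waterfall splash foam, outputting unlit emissive color and opacity.
# Centered UVs become polar coordinates for two radially scrolling foam samples, time-offset by world-space foam noise, multiplied and faded from the center.
# A sharpened band along the U midline is added to shape opacity.
# Two foam colors blend and are tinted by a time-of-day sky color computed per vertex through an interpolator.
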MaterialFunction /FortWater/Materials/WaterSurface/Mobile/Functions/MF_MobileWaterfallSplash
graph {nodes [
  FunctionOutput(None=SetMaterialAttributes!(MaterialAttributes, EmissiveColor=NamedRerouteUsage(), Opacity=NamedRerouteUsage(), ShadingModel=ShadingModel() {shadingModel=MSM_Unlit}) {attributeSetTypes [#1, #2, #3]})
  n1 = Add(A=TextureCoordinate()) {constB=-0.5}
  n2 = ComponentMask(None=@n1) {r=true}
  n3 = Divide(A=Arctangent2Fast(Y=@n2, X=ComponentMask(None=@n1) {g=true})) {constB=6.28}
  n4 = Distance(A=@n1, B=Constant())
  n5 = TextureSample(UVs=Multiply(A=WorldPosition().XY) {constB=0.0002}) {samplerType=SAMPLERTYPE_LinearColor, texture=/FortWater/Materials/WaterSurface/Mobile/Textures/T_MobileWater_Foam}
  n6 = Multiply(A=Subtract(A=Time(), B=Multiply(A=@n5.R) {constB=0.25})) {constB=0.2}
  n7 = AppendVector(A=Add(A=@n3) {constB=0.5}, B=Subtract(A=@n4, B=@n6))
  n8 = AppendVector(A=@n3, B=Subtract(A=Multiply(A=@n4) {constB=0.745}, B=@n6))
  n9 = Multiply(A=@n4) {constB=2}
  n10 = Multiply(
   A=Multiply(
    A=TextureSample(UVs=@n7) {samplerType=SAMPLERTYPE_LinearColor, texture=&T_MobileWater_Foam}.R
    B=TextureSample(UVs=@n8) {samplerType=SAMPLERTYPE_LinearColor, texture=&T_MobileWater_Foam}.R
   )
   B=OneMinus(None=Saturate(None=@n9))
  )
  n11 = OneMinus(None=Multiply(A=Abs(None=@n2)) {constB=2})
  n12 = Multiply(A=@n11, B=@n11)
  n13 = Multiply(A=@n12, B=@n12)
  n14 = Subtract(A=Add(A=Multiply(A=@n10) {constB=6}, B=@n13)) {constB=0.02}
  n15 = Add(A=Multiply(A=@n10) {constB=2}, B=@n13)
  NamedRerouteDeclaration(None=Multiply(A=Saturate(None=@n14), B=Saturate(None=@n15))) {name=Opacity, nodeColor=<rgba 0 0.2969 1 1>}
  n16 = VectorParameter() {group=Mobile Water, parameterName=Mobile Foam Color 1, defaultValue=<rgba 0.3438 0.5326 0.6667 1>}
  n17 = VectorParameter() {group=Mobile Water, parameterName=Mobile Foam Color 2, defaultValue=<rgba 0.5763 0.5835 0.5885 1>}
  n18 = NamedRerouteUsage()
  n19 = Multiply(A=@n18, B=@n18)
  n20 = VertexInterpolator(VS=MaterialFunctionCall(
    Color_Morning=VectorParameter() {parameterName=Sky Color Morning, defaultValue=<rgba 1 0.6076 0.4972 1>}.RGB
    Color_Day=VectorParameter() {parameterName=Sky Color Day, defaultValue=<rgba 1 0.9153 0.832 1>}.RGB
    Color_Evening=VectorParameter() {parameterName=Sky Color Evening, defaultValue=<rgba 0.9635 0.6107 0.5258 1>}.RGB
    Color_Night=VectorParameter() {parameterName=Sky Color Night, defaultValue=<rgba 0.305 0.4452 0.7011 1>}.RGB
   ) {materialFunction=&Color_Time_of_Day}.Result)
  NamedRerouteDeclaration(None=Multiply(
    A=LinearInterpolate(A=@n16.RGB, B=@n17.RGB, Alpha=@n19)
    B=NamedRerouteDeclaration(None=@n20.PS) {name=Sky Color, nodeColor=<rgba 0.7578 1 0 1>}
   )) {name=Base Color, nodeColor=<rgba 0 1 0.1562 1>}
 ]}
