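# Translucent waterfall splash material that renders animated radial foam and ripples where falls meet water.
# Noise-distorted main foam, two sine-crossfaded fast foam layers and a center flow build opacity and color, remapped through a curve atlas and tinted.
# A dithered eight-tap shading-model-ID test limits deferred opacity to Single Layer Water pixels.
# Gerstner waves drive normal and offset to follow the water surface; mobile has a separate material function.
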Material /Game/Tools/NewWaterfallGenerator/Material/M_WaterfallSplash
graph {
 properties {
  bUseMaterialAttributes=true
  bUsedWithInstancedStaticMeshes=true
  bUsedWithWater=true
  blendMode=BLEND_Translucent
  pixelDepthOffsetMode=PDOM_Legacy
  translucencyLightingMode=TLM_VolumetricPerVertexDirectional
  usedShadingModels=Default Lit
 }
 nodes [
  n1 = Custom!(None) {
   code=View.TemporalAAParams
   desc="\n*** Original source before expression upgrade ***\nView.TemporalAAParams\n*** Original source before expression upgrade ***\nFrame.TemporalAAParams"
   description=TemporalAA SampleID
   outputType=CMOT_Float2
  }
  MaterialExpressionAdd_19 = Add(
   A=MaterialFunctionCall!(TextureResolution=Constant2Vector() {g=1, r=1}) {materialFunction=/Engine/Functions/Engine_MaterialFunctions02/Texturing/ScreenAlignedPixelToPixelUVs}.Result
   B=ComponentMask(None=@n1) {r=true}
  )
  n2 = Custom!(p=@MaterialExpressionAdd_19) {
   code="((uint)(p.x) + 2 * (uint)(p.y)) % 5"
   outputType=CMOT_Float1
   inputs [{inputName=p, input {expression=&self:MaterialExpressionAdd_19}}]
  }
  Divide(A=@n2) {constB=6}
  MaterialFunctionCall() {materialFunction=&DitherTemporalAA_Fortnite}
  n3 = ScalarParameter() {parameterName=SplashColorHueShift}
  n4 = TextureObjectParameter() {group=MainFoam, parameterName=MainFoam_Texture, sortPriority=0, texture=/Game/Global/Textures/Default/Tileable_Patterns/T_RiverFoam_new_M}
  n5 = TextureCoordinate()
  n6 = TextureObjectParameter() {
   group=MainFoamNoise
   parameterName=MainFoamNoise_Texture
   sortPriority=0
   texture=&T_RiverFoam_new_M
  }
  n7 = TextureCoordinate()
  n8 = VectorParameter() {
   group=MainFoamNoise
   parameterName=MainFoamNoise_UVControl
   channelNames {a=VPan, b=UPan, g=VScale, r=UScale}
   defaultValue=<rgba 1 1.5 0 -0.2>
  }
  n9 = MaterialFunctionCall() {materialFunction=&VectorToRadialValue}
  n10 = MaterialFunctionCall!(
   Texture=@n6
   "[Override]DDXInput"=DDX(Value=@n7)
   "[Override]DDYInput"=DDY(Value=@n7)
   AdditionalUVOffset
   UVControl_RGB=@n8
   UVControl_A=@n8
   "[Override]TimeInput"
   "[Override]UVInput"=@n9.Radial Coordinates
  ) {materialFunction=/Game/Effects/Niagara/MasterMaterials/Functions/Common/FXCommon_TextureSample_DDXDDY}
  n11 = ScalarParameter() {defaultValue=0.3, parameterName=Mainawf}
  n12 = VectorParameter() {
   group=MainFoamNoise
   parameterName=MainFoamNoise_NoiseControl
   sortPriority=60
   channelNames {a=VStrength, b=UStrength, g=VOffset, r=UOffset}
   defaultValue=<rgba -0.5 -0.5 0.1 0.05>
  }
  n13 = VectorParameter() {group=MainFoam, parameterName=MainFoam_UVControl, channelNames {a=VPan, b=UPan, g=VScale, r=UScale}, defaultValue=<rgba 1 2 0 -0.1>}
  n14 = MaterialFunctionCall(
   Texture=@n4
   "[Override]DDXInput"=DDX(Value=@n5)
   "[Override]DDYInput"=DDY(Value=@n5)
   AdditionalUVOffset=MaterialFunctionCall!(NoiseValue=Power(Base=@n10.B, Exp=@n11), Control_RGB=@n12, Control_A=@n12) {materialFunction=/Game/Effects/Niagara/MasterMaterials/Functions/Common/FXCommon_NoiseControl}.NoiseValue
   UVControl_RGB=@n13
   UVControl_A=@n13
   "[Override]UVInput"=@n9.Radial Coordinates
  ) {materialFunction=&FXCommon_TextureSample_DDXDDY}
  n15 = Multiply(
   A=Add(A=@n14.B, B=ScalarParameter() {group=MainFoamColor, parameterName=MainFoamColor_GrayAdjust_Add})
   B=ScalarParameter() {defaultValue=1, group=MainFoamColor, parameterName=MainFoamColor_GrayAdjust_Mul}
  )
  n16 = Saturate(None=Power(Base=Saturate(None=@n15), Exp=ScalarParameter() {defaultValue=0.4, group=MainFoamColor, parameterName=MainFoamColor_GrayAdjust_Power}))
  n17 = TextureObjectParameter() {group=FastFoam2, parameterName=FastFoam2_Texture, sortPriority=0, texture=&T_RiverFoam_new_M}
  n18 = TextureCoordinate()
  n19 = VectorParameter() {
   group=FastFoamNoise
   parameterName=FastFoamNoise_NoiseControl
   sortPriority=60
   channelNames {a=VStrength, b=UStrength, g=VOffset, r=UOffset}
   defaultValue=<rgba -0.5 -0.5 0 0.05>
  }
  n20 = ScalarParameter() {parameterName=LinkToWaterBodyIndex}
  n21 = QualitySwitch(Default=StaticBool(), Low=StaticBool() {value=true})
  n22 = MaterialFunctionCall!(Z Offset Only, Position, WaterBodyIndex=@n20, Normals use VS Interpolation=@n21, All Normals VS=@n21) {materialFunction=/FortWater/Materials/Functions/Fort_ComputeGerstnerWaves}
  n23 = Add(
   A=MaterialFunctionCall(NoiseValue=@n14.B, Control_RGB=@n19, Control_A=@n19) {materialFunction=&FXCommon_NoiseControl}.NoiseValue
   B=Multiply(A=ComponentMask(None=@n22.Normal) {g=true, r=true}, B=ScalarParameter() {defaultValue=0.02, parameterName=WAn})
  )
  n24 = VectorParameter() {group=FastFoam2, parameterName=FastFoam2_UVControl, channelNames {a=VPan, b=UPan, g=VScale, r=UScale}, defaultValue=<rgba 4 2 0 -0.13>}
  n25 = MaterialFunctionCall(Float2=@n9.Radial Coordinates) {materialFunction=&BreakOutFloat2Components}
  n26 = AppendVector(A=@n25.R, B=Power(Base=@n25.G, Exp=ScalarParameter() {defaultValue=1, parameterName=UVStretch}))
  n27 = MaterialFunctionCall(
   Texture=@n17
   "[Override]DDXInput"=DDX(Value=@n18)
   "[Override]DDYInput"=DDY(Value=@n18)
   AdditionalUVOffset=Add(A=@n23, B=ScalarParameter() {defaultValue=0.2, parameterName=awd})
   UVControl_RGB=@n24
   UVControl_A=@n24
   "[Override]UVInput"=@n26
  ) {materialFunction=&FXCommon_TextureSample_DDXDDY}
  n28 = LinearInterpolate(
   A=ScalarParameter() {defaultValue=0.2, group=FastFoam2, parameterName=FastFoam2_GrayAdjust_Power0}
   B=ScalarParameter() {defaultValue=3, group=FastFoam2, parameterName=FastFoam2_GrayAdjust_Power}
   Alpha=@n25.G
  )
  n29 = TextureObjectParameter() {group=FastFoam, parameterName=FastFoam_Texture, sortPriority=0, texture=&T_RiverFoam_new_M}
  n30 = TextureCoordinate()
  n31 = VectorParameter() {group=FastFoam, parameterName=FastFoam_UVControl, channelNames {a=VPan, b=UPan, g=VScale, r=UScale}, defaultValue=<rgba 4 2 0 -0.17>}
  n32 = MaterialFunctionCall(
   Texture=@n29
   "[Override]DDXInput"=DDX(Value=@n30)
   "[Override]DDYInput"=DDY(Value=@n30)
   AdditionalUVOffset=@n23
   UVControl_RGB=@n31
   UVControl_A=@n31
   "[Override]UVInput"=@n26
  ) {materialFunction=&FXCommon_TextureSample_DDXDDY}
  n33 = LinearInterpolate(
   A=ScalarParameter() {defaultValue=0.2, group=FastFoam, parameterName=FastFoam_GrayAdjust_Power0}
   B=ScalarParameter() {defaultValue=3, group=FastFoam, parameterName=FastFoam_GrayAdjust_Power}
   Alpha=@n25.G
  )
  n34 = Multiply(A=Time(), B=ScalarParameter() {defaultValue=0.7, parameterName=rh3rh3})
  n35 = LinearInterpolate(A=Saturate(None=Power(Base=@n27.A, Exp=@n28)), B=Saturate(None=Power(Base=@n32.A, Exp=@n33)), Alpha=ConstantBiasScale(None=Sine(None=@n34)))
  n36 = Multiply(
   A=Add(A=@n14.B, B=ScalarParameter() {group=MainFoamOpacity, parameterName=MainFoamOpacity_GrayAdjust_Add})
   B=ScalarParameter() {defaultValue=1.25, group=MainFoamOpacity, parameterName=MainFoamOpacity_GrayAdjust_Mul}
  )
  n37 = Saturate(None=Power(Base=Saturate(None=@n36), Exp=ScalarParameter() {defaultValue=0.3, group=MainFoamOpacity, parameterName=MainFoamOpacity_GrayAdjust_Power}))
  n38 = CurveAtlasRowParameter(CurveTime=LinearInterpolate(A=@n16, B=Add(A=@n35, B=@n16), Alpha=@n37)) {
   atlas=&/CLA_Waterfall
   curve=&/CL_SplashColorRemap
   parameterName=ColorCurve
  }
  n39 = VectorParameter() {parameterName=BPInput_RippleColor, defaultValue=<rgba 1 1 1 1>}
  n40 = MaterialFunctionCall(
   In=Power(Base=OneMinus(None=@n25.G), Exp=ScalarParameter() {defaultValue=2, parameterName=aadadad}) {constExponent=3}
   Contrast=ScalarParameter() {defaultValue=0.5, parameterName=wfwfwfw}
  ) {materialFunction=&CheapContrast}
  n41 = StaticSwitchParameter(
   True=LinearInterpolate(A=Multiply(A=@n38, B=@n39.RGB), B=@n39.RGB, Alpha=@n40.Result)
   False=LinearInterpolate(A=@n38, B=VectorParameter() {parameterName=PColor, defaultValue=<rgba 0.5525 0.6338 0.65 1.2>}.RGB, Alpha=@n40.Result)
  ) {parameterName=Update5/10/2024}
  n42 = Multiply(
   A=OneMinus(None=MaterialFunctionCall(Float2=@n9.Radial Coordinates) {materialFunction=&BreakOutFloat2Components}.G)
   B=ScalarParameter() {defaultValue=1.6, parameterName=awda}
  )
  n43 = Constant() {r=1}
  n44 = ScreenPosition()
  n45 = MaterialFunctionCall!(OffsetAmount=ScalarParameter() {defaultValue=0.03, parameterName=Scale}) {materialFunction=/Engine/Functions/Engine_MaterialFunctions01/ImageAdjustment/BlurSampleOffsets}
  MaterialExpressionScreenPosition_1 = ScreenPosition()
  n46 = Custom!(ScreenPos=@MaterialExpressionScreenPosition_1.PixelPosition) {
   code="float random = InterleavedGradientNoise(ScreenPos.xy, float(View.StateFrameIndexMod8));\r\nreturn random;"
   outputType=CMOT_Float1
   inputs [{
     inputName=ScreenPos
     input {expression=&self:MaterialExpressionScreenPosition_1, outputIndex=1}
    }]
  }
  n47 = Divide(A=@n46, B=ScalarParameter() {defaultValue=2, parameterName=DivideDither}) {constB=1.5}
  n48 = SceneTexture(UVs=Add(A=@n44.ViewportUV, B=Multiply(A=@n45."0,1", B=@n47))) {sceneTextureId=PPI_ShadingModelID}
  n49 = Constant()
  n50 = Constant() {r=1}
  n51 = SceneTexture(UVs=Add(A=@n44.ViewportUV, B=Multiply(A=@n45."1,0", B=@n47))) {sceneTextureId=PPI_ShadingModelID}
  n52 = Constant()
  n53 = Constant() {r=1}
  n54 = Add(A=@n44.ViewportUV, B=Multiply(A=@n45."0,-1", B=@n47))
  n55 = Constant()
  n56 = Add(
   A=Add(
    A=If(A=@n48.Color, "A > B"=@n49, "A == B"=@n50, "A < B"=@n49) {constB=10, equalsThreshold=0.01}
    B=If(A=@n51.Color, "A > B"=@n52, "A == B"=@n53, "A < B"=@n52) {constB=10, equalsThreshold=0.01}
   )
   B=If(A=SceneTexture(UVs=@n54) {sceneTextureId=PPI_ShadingModelID}.Color, "A > B"=@n55, "A == B"=Constant() {r=1}, "A < B"=@n55) {constB=10, equalsThreshold=0.01}
  )
  n57 = Multiply(A=@n45."-1,0", B=@n47)
  n58 = Constant()
  n59 = If(
   A=SceneTexture(UVs=Add(A=@n44.ViewportUV, B=@n57)) {sceneTextureId=PPI_ShadingModelID}.Color
   "A > B"=@n58
   "A == B"=Constant() {r=1}
   "A < B"=@n58
  ) {constB=10, equalsThreshold=0.01}
  n60 = SceneTexture(UVs=Add(A=@n44.ViewportUV, B=Multiply(A=@n45."-.5,.5", B=@n47))) {sceneTextureId=PPI_ShadingModelID}
  n61 = Constant()
  n62 = Constant() {r=1}
  n63 = Add(A=@n44.ViewportUV, B=Multiply(A=@n45.".5,.5", B=@n47))
  n64 = Constant()
  n65 = Add(
   A=Add(A=Add(A=@n56, B=@n59), B=If(A=@n60.Color, "A > B"=@n61, "A == B"=@n62, "A < B"=@n61) {constB=10, equalsThreshold=0.01})
   B=If(A=SceneTexture(UVs=@n63) {sceneTextureId=PPI_ShadingModelID}.Color, "A > B"=@n64, "A == B"=Constant() {r=1}, "A < B"=@n64) {constB=10, equalsThreshold=0.01}
  )
  n66 = Multiply(A=@n45.".5,-.5", B=@n47)
  n67 = Constant()
  n68 = If(
   A=SceneTexture(UVs=Add(A=@n44.ViewportUV, B=@n66)) {sceneTextureId=PPI_ShadingModelID}.Color
   "A > B"=@n67
   "A == B"=Constant() {r=1}
   "A < B"=@n67
  ) {constB=10, equalsThreshold=0.01}
  n69 = SceneTexture(UVs=Add(A=@n44.ViewportUV, B=Multiply(A=@n45."-.5,-.5", B=@n47))) {sceneTextureId=PPI_ShadingModelID}
  n70 = Constant()
  n71 = Constant() {r=1}
  n72 = Divide(A=Add(A=Add(A=@n65, B=@n68), B=If(A=@n69.Color, "A > B"=@n70, "A == B"=@n71, "A < B"=@n70) {constB=10, equalsThreshold=0.01})) {constB=8}
  n73 = LinearInterpolate(B=Add(A=@n35, B=@n37), Alpha=@n37)
  n74 = TextureObjectParameter() {
   group=CenterWaterflow
   parameterName=CenterWaterflow_Texture
   sortPriority=0
   texture=&T_RiverFoam_new_M
  }
  n75 = TextureCoordinate()
  n76 = DDX(Value=@n75)
  n77 = DDY(Value=@n75)
  n78 = VectorParameter() {
   group=CenterWaterflow
   parameterName=CenterWaterflow_UVControl
   channelNames {a=VPan, b=UPan, g=VScale, r=UScale}
   defaultValue=<rgba 2 2 0 -0.3>
  }
  n79 = Saturate(None=Power(
    Base=MaterialFunctionCall(Texture=@n74, "[Override]DDXInput"=@n76, "[Override]DDYInput"=@n77, UVControl_RGB=@n78, UVControl_A=@n78, "[Override]UVInput"=@n26) {materialFunction=&FXCommon_TextureSample_DDXDDY}.A
    Exp=ScalarParameter() {defaultValue=1, group=CenterWaterflow, parameterName=CenterWaterflow_GrayAdjust_Power}
   ))
  n80 = ScalarParameter() {defaultValue=0.4, group=CenterWaterflow, parameterName=CenterWaterflow_GrayAdjust_Mul}
  n81 = Saturate(None=Add(A=Multiply(A=@n79, B=@n80), B=ScalarParameter() {defaultValue=0.7, group=CenterWaterflow, parameterName=CenterWaterflow_GrayAdjust_Add}))
  n82 = Multiply(
   A=Power(Base=Saturate(None=@n42), Exp=ScalarParameter() {defaultValue=1.2, parameterName=awdaw})
   B=Multiply(A=ShadingPathSwitch(Default=@n72, Deferred=@n72, Forward=@n43, Mobile=@n43), B=LinearInterpolate(A=@n73, B=@n81, Alpha=@n40.Result))
  )
  n83 = TextureCoordinate()
  n84 = Multiply(A=OneMinus(None=MaterialFunctionCall(Float2=@n83) {materialFunction=&BreakOutFloat2Components}.R)) {constB=2}
  n85 = Multiply(A=Saturate(None=Subtract(A=@n84))) {constB=8}
  n86 = Multiply(A=@n82, B=Multiply(A=Saturate(None=@n85), B=Floor(None=@n84)))
  n87 = ShadingPathSwitch(Default=@n82, Deferred=@n82, Forward=@n86, Mobile=@n86)
  n88 = LinearInterpolate(
   A=VectorParameter() {parameterName=BPInput_BaseColor, defaultValue=<rgba 0.014 0.0607 0.07 1>}.RGB
   B=Desaturation(
    None=MaterialFunctionCall(Hue Shift Percentage=@n3, Texture=@n41) {materialFunction=&HueShift}.Result
    Fraction=ScalarParameter() {parameterName=SplashColorDesaturate}
   )
   Alpha=@n87
  )
  n89 = MaterialFunctionCall!(
   Normal=@n22.Normal
   WPO=@n22.Normal
   Mask=MaterialFunctionCall!(
    WaterDepth=MaterialFunctionCall!(Texture, Position) {materialFunction=/FortWater/Materials/Functions/WaterBodyData}.Water Depth
    Water Body Index=@n20
   ) {materialFunction=/Water/Materials/Functions/ComputeWavesAttenuationFactor}.WaveAttenuationFactor
  ) {materialFunction=/FortWater/Materials/Functions/MaskOutWaves}
  n90 = SetMaterialAttributes!(
   MaterialAttributes
   Base Color=@n88
   Opacity=Power(Base=@n87) {constExponent=0.5}
   Normal=@n89.Normal
   World Position Offset=Add(A=@n89.WPO, B=VectorParameter() {parameterName=ExtraWPO, defaultValue=<rgba 0 0 15 1>}.RGB)
   Shading Model=ShadingModel()
   Emissive Color=Multiply(A=@n88, B=ScalarParameter() {parameterName=RippleEmissiveStrength})
  ) {attributeSetTypes [
    #1
    #2
    #3
    #4
    #5
    #6
   ]}
  NamedRerouteDeclaration(None=@n90) {name=HighEnd Material, nodeColor=<rgba 0 1 0.5078 1>}
  NamedRerouteDeclaration(None=MaterialFunctionCall() {materialFunction=/FortWater/Materials/WaterSurface/Mobile/Functions/MF_MobileWaterfallSplash}.Result) {name=Mobile Material, nodeColor=<rgba 0.4062 1 0 1>}
  ShadingPathSwitch(Default=NamedRerouteUsage(), Mobile=NamedRerouteUsage())
  TextureObjectParameter() {
   group=FastFoamTexture
   parameterName=FastFoamTexture_Texture
   sortPriority=0
   texture=&T_RiverFoam_new_M
  }
 ]
}
