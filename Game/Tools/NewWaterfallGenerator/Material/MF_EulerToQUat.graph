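# Library-exposed material function in the waterfall generator tool that converts an Euler angle vector in degrees into a quaternion.
# The Euler input is wrapped by Fmod 360, then passed with the Pi function into a Custom HLSL node.
# That node reads X, Y, Z as roll, pitch, yaw, converts to radians and applies the half-angle sine/cosine formula in half precision, returning a float4 with W as scalar.
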MaterialFunction /Game/Tools/NewWaterfallGenerator/Material/MF_EulerToQUat
graph {
 properties {bExposeToLibrary=true}
 nodes [
  MaterialExpressionMaterialFunctionCall_1 = MaterialFunctionCall() {materialFunction=&Pi}
  MaterialExpressionFmod_0 = Fmod(A=FunctionInput() {bUsePreviewValueAsDefault=true, inputName=Euler}, B=Constant3Vector() {constant=<rgba 360 360 360 0>})
  n1 = Custom!(M_PI=@MaterialExpressionMaterialFunctionCall_1.Result, Euler=@MaterialExpressionFmod_0) {
   code="half roll = Euler.x;\r\nhalf pitch = Euler.y;\r\nhalf yaw = Euler.z;\r\n    roll *= M_PI / 180.0;\r\n    pitch *= M_PI / 180.0;\r\n    yaw *= M_PI / 180.0;\r\n\r\n    half c1 = cos(yaw / 2);\r\n    half s1 = sin(yaw / 2);\r\n    half c2 = cos(pitch / 2);\r\n    half s2 = sin(pitch / 2);\r\n    half c3 = cos(roll / 2);\r\n    half s3 = sin(roll / 2);\r\n\r\n    half4 q;\r\n    q.w = c1 * c2 * c3 + s1 * s2 * s3;\r\n    q.x = c1 * c2 * s3 - s1 * s2 * c3;\r\n    q.y = c1 * s2 * c3 + s1 * c2 * s3;\r\n    q.z = s1 * c2 * c3 - c1 * s2 * s3;\r\n\r\n    return q;"
   description=EulerToQuaternion
   outputType=CMOT_Float4
   inputs [
    {inputName=M_PI, input {expression=&self:MaterialExpressionMaterialFunctionCall_1}}
    {inputName=Euler, input {expression=&self:MaterialExpressionFmod_0}}
   ]
  }
  FunctionOutput(None=@n1)
 ]
}
