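# Library material function in the waterfall generator tools that outputs a hard-edged 0/1 mask for positions inside a box with adjustable center, extent and rotation.
# It rotates the world position about the box center, using a quaternion built by MF_EulerToQUat from the negated Euler angles.
# Six If comparisons test each axis against center plus or minus extent, so extent acts as half-size, and their product forms the mask.
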MaterialFunction /Game/Tools/NewWaterfallGenerator/Material/MF_RotatableBoxMask
graph {
 properties {bExposeToLibrary=true}
 nodes [
  n1 = FunctionInput() {bUsePreviewValueAsDefault=true, inputName=BoxCenterPosition}
  n2 = FunctionInput() {bUsePreviewValueAsDefault=true, inputName=BoxExtend, previewValue=<1000 1000 1000 1>}
  n3 = Add(A=@n1, B=@n2)
  n4 = FunctionInput(Preview=WorldPosition().XYZ) {bUsePreviewValueAsDefault=true, inputName=WorldPosition, previewValue=<1000 1000 1000 1>}
  n5 = Multiply(A=FunctionInput() {bUsePreviewValueAsDefault=true, inputName=BoxRotationEuler}) {constB=-1}
  n6 = Add(
   A=MaterialFunctionCall!(
    Vector=Subtract(A=@n4, B=@n1)
    Quat=MaterialFunctionCall!(Euler=@n5) {materialFunction=&/MF_EulerToQUat}.Result
    Invert Quaternion
   ) {materialFunction=/Engine/Functions/Engine_MaterialFunctions03/Math/MultiplyVectorWithQuaternion}.Result
   B=@n1
  )
  n7 = ComponentMask(None=@n6) {r=true}
  n8 = Constant() {r=1}
  n9 = Constant()
  n10 = ComponentMask(None=@n6) {g=true}
  n11 = Multiply(
   A=If(A=ComponentMask(None=@n3) {r=true}, B=@n7, "A > B"=@n8, "A < B"=@n9)
   B=If(A=ComponentMask(None=@n3) {g=true}, B=@n10, "A > B"=@n8, "A < B"=@n9)
  )
  n12 = ComponentMask(None=@n6) {b=true}
  n13 = If(A=ComponentMask(None=@n3) {b=true}, B=@n12, "A > B"=@n8, "A < B"=@n9)
  n14 = Subtract(A=@n1, B=@n2)
  n15 = Constant() {r=1}
  n16 = Constant()
  n17 = Multiply(
   A=If(A=@n7, B=ComponentMask(None=@n14) {r=true}, "A > B"=@n15, "A < B"=@n16)
   B=If(A=@n10, B=ComponentMask(None=@n14) {g=true}, "A > B"=@n15, "A < B"=@n16)
  )
  n18 = If(A=@n12, B=ComponentMask(None=@n14) {b=true}, "A > B"=@n15, "A < B"=@n16)
  FunctionOutput(None=Multiply(A=Multiply(A=@n11, B=@n13), B=Multiply(A=@n17, B=@n18)))
 ]
}
