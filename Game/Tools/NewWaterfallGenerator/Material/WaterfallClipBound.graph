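# Material function for the waterfall generator that clips a material's opacity mask using up to four rotatable box volumes.
# Position, rotation and extent vector parameters for boxes A to D feed MF_RotatableBoxMask calls evaluated at world position.
# The four masks are combined with Max and lerp the incoming OpacityMask toward zero, cutting out the boxed regions.
# The EnableWaterfallClipBound static switch otherwise passes the input material attributes through unchanged.
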MaterialFunction /Game/Tools/NewWaterfallGenerator/Material/WaterfallClipBound
graph {nodes [
  n1 = FunctionInput() {inputName=InMA, inputType=FunctionInput_MaterialAttributes}
  n2 = GetMaterialAttributes(None=@n1) {attributeGetTypes [#1]}
  n3 = WorldPosition()
  n4 = Max(
   A=MaterialFunctionCall!(
    BoxExtend=VectorParameter() {parameterName=WaterfallClipBound_A_Extend, defaultValue=<rgba 0 0 0 1>}.RGB
    BoxRotationEuler=VectorParameter() {parameterName=WaterfallClipBound_A_Rotation, defaultValue=<rgba 0 0 0 1>}.RGB
    WorldPosition=@n3.XYZ
    BoxCenterPosition=VectorParameter() {parameterName=WaterfallClipBound_A_Position, defaultValue=<rgba 0 0 0 1>}.RGB
   ) {materialFunction=&/MF_RotatableBoxMask}.Result
   B=MaterialFunctionCall(
    BoxExtend=VectorParameter() {parameterName=WaterfallClipBound_B_Extend, defaultValue=<rgba 0 0 0 1>}.RGB
    BoxRotationEuler=VectorParameter() {parameterName=WaterfallClipBound_B_Rotation, defaultValue=<rgba 0 0 0 1>}.RGB
    WorldPosition=@n3.XYZ
    BoxCenterPosition=VectorParameter() {parameterName=WaterfallClipBound_B_Position, defaultValue=<rgba 0 0 0 1>}.RGB
   ) {materialFunction=&MF_RotatableBoxMask}.Result
  )
  n5 = MaterialFunctionCall(
   BoxExtend=VectorParameter() {parameterName=WaterfallClipBound_C_Extend, defaultValue=<rgba 0 0 0 1>}.RGB
   BoxRotationEuler=VectorParameter() {parameterName=WaterfallClipBound_C_Rotation, defaultValue=<rgba 0 0 0 1>}.RGB
   WorldPosition=@n3.XYZ
   BoxCenterPosition=VectorParameter() {parameterName=WaterfallClipBound_C_Position, defaultValue=<rgba 0 0 0 1>}.RGB
  ) {materialFunction=&MF_RotatableBoxMask}
  n6 = VectorParameter() {parameterName=WaterfallClipBound_D_Extend, defaultValue=<rgba 0 0 0 1>}
  n7 = VectorParameter() {parameterName=WaterfallClipBound_D_Rotation, defaultValue=<rgba 0 0 0 1>}
  n8 = VectorParameter() {parameterName=WaterfallClipBound_D_Position, defaultValue=<rgba 0 0 0 1>}
  n9 = LinearInterpolate(
   A=@n2.OpacityMask
   Alpha=Max(
    A=Max(A=@n4, B=@n5.Result)
    B=MaterialFunctionCall(BoxExtend=@n6.RGB, BoxRotationEuler=@n7.RGB, WorldPosition=@n3.XYZ, BoxCenterPosition=@n8.RGB) {materialFunction=&MF_RotatableBoxMask}.Result
   )
  ) {constB=0}
  FunctionOutput(None=StaticSwitchParameter(
    True=SetMaterialAttributes!(MaterialAttributes=@n2.MaterialAttributes, OpacityMask=@n9) {attributeSetTypes [#1]}
    False=@n1
   ) {parameterName=EnableWaterfallClipBound})
 ]}
